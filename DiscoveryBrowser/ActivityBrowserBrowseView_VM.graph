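# Discovery browser screen that stacks Featured and All category tile panels, fed by view-model data through setter functions.
# Mouse wheel, navigation and the Back to Top button page between panels.
# The ShiftToNextPanel animation drives CategoryIndexAnimator, whose setter lerps the vertical box's translation between cached panel positions, while helpers swap slot sizing and wheel input.
# ActivateFeaturedCategory selects a tile by mnemonic and pushes a nested-surface or View All page.
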WidgetBlueprint /DiscoveryBrowser/ActivityBrowserBrowseView_VM
animations [
 ShiftToNextPanel {
  binding_count=2
  length_seconds=0.5008
  object=&self:ShiftToNextPanel
  sequence {bindings [
    ActivityBrowserBrowseViewV2_C_8 {
     id=#1
     possessed_class=ActivityBrowserBrowseViewV2_C
     tracks [{
       type=MovieSceneDoubleTrack
       display_name=Category Index Animator
       sections [{type=MovieSceneDoubleSection, end_frame=10, start_frame=0, channels [None {type=MovieSceneScriptingDoubleChannel, keys [[0,0],[10,1]]}]}]
      }]
    }
    Image_HeaderSoftEdge {
     id=#2
     possessed_class=Image
     tracks [{
       type=MovieSceneFloatTrack
       display_name=Render Opacity
       sections [{type=MovieSceneFloatSection, end_frame=10, start_frame=0, channels [None {type=MovieSceneScriptingFloatChannel, keys [[0,0],[5,1],[10,0]]}]}]
      }]
    }
   ]}
 }
 Intro {
  binding_count=2
  length_seconds=0.45
  object=&self:Intro
  sequence {bindings [
    Overlay_Content {
     id=#3
     possessed_class=Overlay
     tracks [{
       type=MovieSceneFloatTrack
       display_name=Render Opacity
       sections [{type=MovieSceneFloatSection, channels [None {type=MovieSceneScriptingFloatChannel, keys [[0,0],[2,0],[7,1]]}]}]
      }]
    }
    TilePanel_Featured_VMversion {
     id=#4
     possessed_class=ActivityCategoryTilePanelBase_VM_C
     tracks [
      {
       type=MovieScene2DTransformTrack
       sections [{
         type=MovieScene2DTransformSection
         channels [
          Translation.X {type=MovieSceneScriptingFloatChannel, const [0,2]}
          Translation.Y {type=MovieSceneScriptingFloatChannel, keys [[2.9997,24],[9,0]]}
          Angle {type=MovieSceneScriptingFloatChannel}
          Scale.X {type=MovieSceneScriptingFloatChannel}
          Scale.Y {type=MovieSceneScriptingFloatChannel}
          Shear.X {type=MovieSceneScriptingFloatChannel}
          Shear.Y {type=MovieSceneScriptingFloatChannel}
         ]
        }]
      }
      {
       type=MovieSceneFloatTrack
       display_name=Render Opacity
       sections [{type=MovieSceneFloatSection, channels [None {type=MovieSceneScriptingFloatChannel, keys [[2.9997,0],[9,1]]}]}]
      }
     ]
    }
   ]}
 }
 Anim_DisplayLoading {
  binding_count=1
  length_seconds=0.2
  object=&self:Anim_DisplayLoading
  sequence {
   bindings [WBP_UIKit_Throbber {
     id=#5
     possessed_class=WBP_UIKit_Throbber_C
     tracks [{
       type=MovieSceneFloatTrack
       display_name=Render Opacity
       sections [{type=MovieSceneFloatSection, channels [None {type=MovieSceneScriptingFloatChannel, keys [[0,0],[12,1]]}]}]
      }]
    }]
   display_rate {numerator=60}
  }
 }
]
graphs [
 Animate to Next Panel {nodes [
   Animate to Next Panel() {is=title K2Node_FunctionEntry}
   n1 = Get CurrentPanelIndex() -> (CurrentPanelIndex:int=0)
   n2 = GetChildrenCount!(self:Panel Widget Object Reference=Get VerticalBox_Panels() -> (VerticalBox_Panels:Vertical Box Object Reference)) -> (ReturnValue:int=0)
   > n3 = Set NextPanelIndex(NextPanelIndex:int="Min (Integer)"(A=int + int(A=@n1), B=int - int(A=@n2)) -> (ReturnValue=0)) -> (Output_Get:int=0)
   > n4 = Branch(Condition=Get IsClickedOnBottomPanel() -> (IsClickedOnBottomPanel:bool))
   > n5 = Set IsClickedOnBottomPanel(IsClickedOnBottomPanel:bool=false) -> (Output_Get:bool)
   n6 = Get VerticalBox_Panels()
   n7 = Branch(
    execute=[@n5.then, @n9.then]
    Condition=IsWidgetFullSizedAndOnScreen!(Widget:Widget Object Reference=GetChildAt!(self:Panel Widget Object Reference=@n6, Index:int=@n3) -> (ReturnValue:Widget Object Reference)) -> (IsFullSized:bool)
   )
   > n8 = Branch(Condition=IsWidgetFullSizedAndOnScreen(Widget=Get VerticalBox_Panels()))
   AnimateToDesiredPanel!(execute=@n8.else, self, NewDesiredIndex:int=Get NextPanelIndex() -> (NextPanelIndex:int=0)) -> (then)
   > EnableMouseWheelInput!(
    execute
    self:Activity Category Tile Panel Base VM Object Reference=Get TilePanel_All_VMversion() -> (TilePanel_All_VMversion:Activity Category Tile Panel Base VM Object Reference)
    bListenForInput:bool=true
   ) -> (then)
   AnimateToDesiredPanel(execute=@n7.else, NewDesiredIndex=Get NextPanelIndex() -> (NextPanelIndex=0))
   n9 = SetSelectedIndex!(
    execute=@n4.else
    self:List View Object Reference=Get TileView_Categories!(self:Activity Category Tile Panel Base VM Object Reference=Get TilePanel_All_VMversion()) -> (TileView_Categories:Fort Activity Tile View Object Reference)
    Index:int=0
   ) -> (then)
  ]}
 Animate to Previous Panel {nodes [
   Animate to Previous Panel() {is=title K2Node_FunctionEntry}
   n1 = "Max (Integer)"(A=int - int(A=Get CurrentPanelIndex() -> (CurrentPanelIndex=0)), B=0)
   > n2 = Set PreviousPanelIndex(PreviousPanelIndex:int=@n1) -> (Output_Get:int=0)
   > SetSelectedIndex(
    self=Get TileView_Categories(self=Get TilePanel_Featured_VMversion() -> (TilePanel_Featured_VMversion:Activity Category Tile Panel Base VM Object Reference))
    Index=0
   )
   n3 = Get VerticalBox_Panels()
   > n4 = Branch(Condition=IsWidgetFullSizedAndOnScreen(Widget=GetChildAt(self=@n3, Index=@n2)))
   AnimateToDesiredPanel(execute=@n4.else, NewDesiredIndex=@n1)
   > EnableMouseWheelInput(self=Get TilePanel_All_VMversion(), bListenForInput=false)
  ]}
 AnimatePanelOpacity {nodes [
   n1 = AnimatePanelOpacity!() -> (then, AnimateAwayFromIndex:int, AnimateToIndex:int) {is=title K2Node_FunctionEntry}
   > n2 = Branch(Condition="integer < integer"(A=@n1, B=@n1.AnimateToIndex))
   > MoveUpOutOfView(self:Activity Category Tile Panel Base Object Reference=GetTilePanelFromIndex!(self, PanelIndex:int=Get AnimateAwayFromIndex() -> (AnimateAwayFromIndex:int=0)) -> (TilePanel:Activity Category Tile Panel Base Object Reference))
   MoveDownIntoView(
    execute=@n2.else
    self:Activity Category Tile Panel Base Object Reference=GetTilePanelFromIndex(PanelIndex=Get AnimateToIndex() -> (AnimateToIndex:int=0))
   )
  ]}
 AnimateToDesiredPanel {nodes [
   n1 = AnimateToDesiredPanel!() -> (then, NewDesiredIndex:int) {is=title K2Node_FunctionEntry}
   n2 = GetChildrenCount(self=Get VerticalBox_Panels()) -> (ReturnValue=0)
   > n3 = Set NewClampedPanel(NewClampedPanel:int="Clamp (Integer)"(Value=@n1, Min=0, Max=int - int(A=@n2)) -> (ReturnValue=0)) -> (Output_Get:int=0)
   > n4 = Branch(Condition="Equal (Integer)"(A=@n3, B=Get CurrentPanelIndex() -> (CurrentPanelIndex=0)))
   n5 = Get ShiftToNextPanel() -> (ShiftToNextPanel:Widget Animation Object Reference)
   n6 = Get DesiredPanelIndex() -> (DesiredPanelIndex:int=0)
   n7 = Get NewClampedPanel() -> (NewClampedPanel:int=0)
   > Branch(Condition=AND Boolean(A=IsAnimationPlaying(InAnimation=@n5), B="Not Equal (Integer)"(A=@n7, B=@n6)))
   > Set CurrentPlayDirection(CurrentPlayDirection:EUMGSequencePlayMode Enum=Reverse) -> (Output_Get:EUMGSequencePlayMode Enum=Forward)
   > PlayAnimationReverse(InAnimation=Get ShiftToNextPanel())
   > AnimatePanelOpacity!(
    execute
    self
    AnimateAwayFromIndex:int=Get DesiredPanelIndex() -> (DesiredPanelIndex=0)
    AnimateToIndex:int=Get CurrentPanelIndex() -> (CurrentPanelIndex=0)
   ) -> (then) {is=title K2Node_CallFunction}
   Set DesiredPanelIndex(execute=@n4.else, DesiredPanelIndex:int=Get NewClampedPanel() -> (NewClampedPanel=0)) -> (Output_Get:int=0)
   > PreUpdatePanelSize()
   > Set CurrentPlayDirection() -> (Output_Get=Forward)
   > PlayAnimationForward(InAnimation=Get ShiftToNextPanel())
   > AnimatePanelOpacity(AnimateAwayFromIndex=Get CurrentPanelIndex() -> (CurrentPanelIndex=0), AnimateToIndex=Get DesiredPanelIndex() -> (DesiredPanelIndex=0))
   > EnableMouseWheelInput(self=Get TilePanel_All_VMversion(), bListenForInput=false)
  ]}
 EventGraph {nodes [
   "On Clicked (WBP_BottomBarHack_BackToTop)"!() -> (OutputDelegate, then, Button:Common Button Base Object Reference) {is=title K2Node_ComponentBoundEvent}
   > AnimateToDesiredPanel(NewDesiredIndex=0) {is=title K2Node_CallFunction}
   > Call On Back to Top Clicked!(execute:Exec, self:Activity Browser Browse View VM Object Reference) -> (then:Exec) {is=title K2Node_CallDelegate}
   n1 = OnInputChanged!() -> (OutputDelegate, then, NewInputType:ECommonInputType Enum) {is=title K2Node_CustomEvent}
   > n2 = Set CurrentInputType(CurrentInputType:ECommonInputType Enum=@n1.NewInputType) -> (Output_Get:ECommonInputType Enum)
   > n3 = Switch on ECommonInputType!(execute, Selection:ECommonInputType Enum=@n2) -> (MouseAndKeyboard:Exec, Gamepad:Exec, Touch:Exec, Count:Exec) {is=title K2Node_SwitchEnum}
   SwitchToGamepad(execute=@n3.Gamepad)
   OnBottomPanelItemClicked() {is=title K2Node_CustomEvent}
   > Set IsClickedOnBottomPanel(IsClickedOnBottomPanel=true)
   OnTopPanelItemClicked() {is=title K2Node_CustomEvent}
   > Set IsClickedOnBottomPanel(IsClickedOnBottomPanel=false)
   n4 = ActivateFeaturedCategory!() -> (OutputDelegate, then, DesiredMnemonic:String) {is=title K2Node_CustomEvent}
   > n5 = For Each Loop with Break!(
    Exec:Exec
    Array:Array of Discover Tile Item View Model Object References=Get FeaturedCategories!(self:Discover Browse View Model Object Reference=Get FortDiscoverBrowseViewModel() -> (FortDiscoverBrowseViewModel:Discover Browse View Model Object Reference)) -> (FeaturedCategories:Array of Discover Tile Item View Model Object References)
    Break:Exec=@n8.then_1
   ) -> (LoopBody:Exec, Array Element:Discover Tile Item View Model Object Reference, Array Index:int, Completed:Exec)
   > Branch(Condition="Equal, Case Insensitive (String)"(A=GetMnemonic!(self:Discover Tile Item View Model Object Reference=@n5) -> (ReturnValue:String), B=@n4.DesiredMnemonic))
   n6 = Get TileView_Categories(self=Get TilePanel_Featured_VMversion())
   > SetSelectedIndex(self=@n6, Index=@n5.Array Index)
   > n7 = Cast To FortDiscoverSurfaceTileItemVM!(execute, Object=GetItemAt!(self:List View Object Reference=@n6, Index:int=@n5.Array Index) -> (ReturnValue:Object Reference)) -> (then, CastFailed, AsFort Discover Surface Tile Item VM:Fort Discover Surface Tile Item VM Object Reference) {is=title K2Node_DynamicCast}
   > n8 = Sequence()
   > n9 = Cast To FortPlayerController(Object=GetOwningPlayer())
   > HandleCategoryTileClicked!(
    execute
    self:Fort Activity Selector Object Reference=Get FortActivitySelector() -> (ReturnValue:Fort Activity Selector Object Reference) {is=title K2Node_GetSubsystem}
    FortPC:Fort Player Controller Object Reference=@n9
    InTileItemVM:Fort Discover Surface Tile Item VM Object Reference=@n7
   ) -> (then)
   > n10 = Branch(Condition="Equal (Name)"(A=GetReferencePanelName!(self:Fort Discover Surface Tile Item VM Object Reference=@n7) -> (ReturnValue:Name=None)))
   > n11 = PushPrimaryContentClassAsync!(
    execute:Exec
    WidgetClass:Common Activatable Widget Class Reference=&/WBP_Discover_NestedSurface.WBP_Discover_NestedSurface_C
   ) -> (then:Exec, OnPrimaryContentPushed:Exec, ActivatableWidget:Common Activatable Widget Object Reference) {is=title K2Node_AsyncAction}
   n12 = Cast To WBP_Discover_NestedSurface!(execute=@n11.OnPrimaryContentPushed, Object=@n11) -> (then, CastFailed, AsWBP Discover Nested Surface:WBP Discover Nested Surface Object Reference) {is=title K2Node_DynamicCast}
   > SetFortDiscoverSurfaceTileItemVM!(execute, self:WBP Discover Nested Surface Object Reference=@n12, Viewmodel:Fort Discover Surface Tile Item VM Object Reference=@n7) -> (then)
   > On Disover Surface Tile Item VM Set(self:WBP Discover Nested Surface Object Reference=@n12)
   n13 = GetReferenceSurfaceName!(self:Fort Discover Surface Tile Item VM Object Reference=@n7) -> (ReturnValue:Name=None)
   > SetSurfaceName!(
    execute
    self:Discover Surface View Model Object Reference=Get FortDiscoverSurfaceViewModel!(self:WBP Discover Nested Surface Object Reference=@n12) -> (FortDiscoverSurfaceViewModel:Discover Surface View Model Object Reference)
    InSurfaceName:Name=@n13
   ) -> (then)
   n14 = PushPrimaryContentClassAsync(execute=@n10.else, WidgetClass=&/WBP_ViewAllPage.WBP_ViewAllPage_C)
   n15 = Cast To WBP_ViewAllPage!(execute=@n14.OnPrimaryContentPushed, Object=@n14) -> (then, CastFailed, AsWBP View All Page:WBP View All Page Object Reference) {is=title K2Node_DynamicCast}
   > SetupSurface!(
    execute
    self:WBP View All Page Object Reference=@n15
    Surface Name:Name=@n13
    Panel Name:Name=GetReferencePanelName(self=@n7) -> (ReturnValue=None)
   ) -> (then)
   > ActivateWidget(self=@n14)
   Event Construct()
   > Parent: Construct()
   n16 = Create Event()
   n17 = Get CommonInputSubsystem()
   > Bind Event to On Input Method Changed(self=@n17, Delegate=@n16)
   > Bind Event to On Panel Item Clicked(self:Activity Category Tile Panel Base VM Object Reference=Get TilePanel_All_VMversion(), Delegate=Create Event()) {is=title K2Node_AddDelegate}
   > Bind Event to On Panel Item Clicked(self=Get TilePanel_Featured_VMversion(), Delegate=Create Event())
   n18 = Get Intro()
   > PlayAnimation(InAnimation=@n18, StartAtTime=GetEndTime(self=@n18), PlayMode=Reverse)
   > PostUpdatePanelSize()
   Event On Activated()
   n19 = Get CurrentInputType() -> (CurrentInputType:ECommonInputType Enum)
   > Branch(Condition=AND Boolean(A=Get IsClickedOnBottomPanel(), B="Equal (Enum)"(A=@n19, B=Gamepad)))
   > Animate to Next Panel() {is=title K2Node_CallFunction}
   K2Node_Event_2 = OnCategoryTilePanelSelected_1!() -> (OutputDelegate, then, SelectedPanel:Fort Activity Category Tile Panel Object Reference) {is=title K2Node_CustomEvent}
   n20 = Get LastSelectedPanel() -> (LastSelectedPanel:Activity Category Tile Panel Base Object Reference)
   > n21 = Is Valid(InputObject=@n20)
   > n22 = SetActive!(execute, self:Activity Category Tile Panel Base Object Reference=@n20, IsActive:bool=false) -> (then)
   n23 = Cast To ActivityCategoryTilePanelBase!(execute=[@n22.then, @n21.Is Not Valid], Object=@K2Node_Event_2.SelectedPanel) -> (then, CastFailed, AsActivity Category Tile Panel Base:Activity Category Tile Panel Base Object Reference) {is=title K2Node_DynamicCast}
   > SetActive(self=@n23, IsActive=true)
   > Set LastSelectedPanel(LastSelectedPanel:Activity Category Tile Panel Base Object Reference=@n23) -> (Output_Get:Activity Category Tile Panel Base Object Reference)
   Set LastSelectedPanel(execute=@n23.CastFailed)
   Event PreConstruct()
   > SetRenderTranslation(self=Get VerticalBox_Panels())
   > SetCurrentPanelIndex!(execute, self, CurrentPanelIndex:int=0) -> (then)
   > EnableMouseWheelInput(self=[Get TilePanel_All_VMversion(), Get TilePanel_Featured_VMversion()], bListenForInput=false)
   Event Destruct()
   > Parent: Destruct()
   > Unbind Event from On Input Method Changed(self=@n17, Delegate=@n16)
   SetVisibility(self=Get WBP_UIKit_Throbber() -> (WBP_UIKit_Throbber:WBP UIKit Throbber Object Reference), InVisibility=HitTestInvisible)
  ]}
 GetIndexOfPanel {nodes [
   GetIndexOfPanel!() -> (then, Widget:Widget Object Reference) {is=title K2Node_FunctionEntry}
   > n1 = For Each Loop with Break(Array:Array of Widget Object References=GetAllChildren(self=Get VerticalBox_Panels())) -> (Array Element:Widget Object Reference)
   Return Node!(execute=@n1.Completed, Index:int=-1)
   > n2 = Branch(Condition="Equal (Object)"(A=Get Widget() -> (Widget:Widget Object Reference), B=@n1))
   Return Node(execute=[@n2.then, @n4.then], Index=@n1.Array Index)
   n3 = Cast To SizeBox!(execute=@n2.else, Object=@n1) -> (then, CastFailed, AsSize Box:Size Box Object Reference) {is=title K2Node_DynamicCast}
   > n4 = Branch(Condition="Equal (Object)"(A=GetChildAt(self=@n3, Index=0), B=Get Widget()))
  ]}
 GetLocalPanelPosition {nodes [
   n1 = GetLocalPanelPosition!() -> (then, Index:int) {is=title K2Node_FunctionEntry}
   n2 = GetCachedGeometry!(self:Widget Object Reference=GetChildAt(self=Get VerticalBox_Panels(), Index=@n1)) -> (ReturnValue:"Geometry Structure (by ref)")
   > Return Node!(
    execute
    Position:double=BreakVector2D(InVec=GetLocalTopLeft!(Geometry:"Geometry Structure (by ref)"=@n2) -> (ReturnValue:Vector 2D Structure)).Y
   )
  ]}
 GetTilePanelFromIndex {nodes [
   n1 = GetTilePanelFromIndex!() -> (then, PanelIndex:int) {is=title K2Node_FunctionEntry}
   > n2 = Set ChildAtIndex(ChildAtIndex:Widget Object Reference=GetChildAt(self=Get VerticalBox_Panels(), Index=@n1)) -> (Output_Get:Widget Object Reference)
   > n3 = Is Valid(InputObject=@n2)
   Return Node!(execute=@n3.Is Not Valid, TilePanel:Activity Category Tile Panel Base Object Reference)
   > n4 = Cast To SizeBox(Object=Get ChildAtIndex() -> (ChildAtIndex:Widget Object Reference))
   > n5 = Cast To ActivityCategoryTilePanelBase(Object=GetChildAt(self=@n4, Index=0))
   > Return Node(TilePanel=@n5)
   Return Node(execute=[@n5.CastFailed, @n6.CastFailed])
   n6 = Cast To ActivityCategoryTilePanelBase(execute=@n4.CastFailed, Object=Get ChildAtIndex())
   > Return Node(TilePanel=@n6)
  ]}
 GetTopMostVisiblePanel {nodes [
   GetTopMostVisiblePanel() {is=title K2Node_FunctionEntry}
   > Return Node!(
    execute
    ReturnValue:Fort Activity Category Tile Panel Object Reference=GetTilePanelFromIndex(PanelIndex=Get CurrentPanelIndex() -> (CurrentPanelIndex=0)) {is=title K2Node_CallFunction}
   )
  ]}
 IsWidgetFullSizedAndOnScreen {nodes [
   n1 = IsWidgetFullSizedAndOnScreen!() -> (then, Widget:Widget Object Reference) {is=title K2Node_FunctionEntry}
   > Branch(Condition=IsValid(Object=@n1))
   n2 = Get VerticalBox_Panels()
   n3 = vector2d + vector2d!(
    A:Vector 2D Structure=GetLocalTopLeft(Geometry=GetCachedGeometry(self=@n1))
    B:Vector 2D Structure=Break Widget Transform!(WidgetTransform:"Widget Transform Structure (by ref)"=Get RenderTransform!(self:Widget Object Reference=@n2) -> (RenderTransform:Widget Transform Structure)) -> (Translation:Vector 2D Structure, Scale:Vector 2D Structure, Shear:Vector 2D Structure, Angle:float=~) {is=title K2Node_BreakStruct}
   ) -> (ReturnValue_X:double, ReturnValue_Y:double) {is=title K2Node_PromotableOperator}
   n4 = BreakVector2D(InVec=GetLocalSize!(Geometry:"Geometry Structure (by ref)"=GetCachedGeometry(self=@n1)) -> (ReturnValue:Vector 2D Structure))
   n5 = BreakVector2D(InVec=GetDesiredSize!(self:Widget Object Reference=@n1) -> (ReturnValue:Vector 2D Structure))
   > Return Node!(execute, IsFullSized:bool=AND Boolean(A="float <= float"(A=@n5.Y, B=@n4.Y) -> (ReturnValue=~), B="float >= float"(A=@n3.ReturnValue_Y)))
  ]}
 NavAllToFeatured {nodes [
   NavAllToFeatured!() -> (then, Navigation:EUINavigation Enum) {is=title K2Node_FunctionEntry}
   > Animate to Previous Panel() {is=title K2Node_CallFunction}
   > Return Node!(execute, ReturnValue:Widget Object Reference=Get TileView_Categories(self=Get TilePanel_Featured_VMversion()))
  ]}
 NavFeaturedToAll {nodes [
   NavFeaturedToAll!() -> (then, Navigation:EUINavigation Enum) {is=title K2Node_FunctionEntry}
   > Animate to Next Panel()
   > Return Node(ReturnValue=Get TileView_Categories(self=Get TilePanel_All_VMversion()))
  ]}
 OnBackToTopClicked {nodes [OnBackToTopClicked() {is=title K2Node_FunctionEntry}]}
 OnMouseWheel {nodes [
   n1 = OnMouseWheel!() -> (then, MyGeometry:Geometry Structure, MouseEvent:"Pointer Event Structure (by ref)") {is=title K2Node_FunctionEntry}
   > n2 = Branch(Condition="float > float"(A=Get Wheel Delta!(Input:"Pointer Event Structure (by ref)"=@n1.MouseEvent) -> (ReturnValue:float) {is=title K2Node_CallFunction}))
   > n3 = Animate to Previous Panel()
   Return Node(execute=[@n5.then, @n3.then], ReturnValue:Event Reply Structure=Handled())
   n4 = Branch(execute=@n2.else, Condition="float < float"(A=Get Wheel Delta(Input=@n1.MouseEvent)))
   > n5 = Animate to Next Panel()
   n6 = Parent: OnMouseWheel!(
    execute=@n4.else
    MyGeometry:Geometry Structure=Get MyGeometry() -> (MyGeometry:Geometry Structure)
    MouseEvent:"Pointer Event Structure (by ref)"=Get MouseEvent() -> (MouseEvent:"Pointer Event Structure (by ref)")
   ) -> (then, ReturnValue:Event Reply Structure) {is=title K2Node_CallParentFunction}
   > Return Node(ReturnValue=@n6)
  ]}
 OnPlayIntro {nodes [OnPlayIntro() {is=title K2Node_FunctionEntry}]}
 OnRowUpdated {nodes [OnRowUpdated!() -> (then, IsPromoted:bool=~, IsFirstRow:bool=~) {is=title K2Node_FunctionEntry}]}
 PostUpdatePanelSize {nodes [
   PostUpdatePanelSize() {is=title K2Node_FunctionEntry}
   n1 = Get VerticalBox_Panels()
   > n2 = For Loop with Break!(
    Execute:Exec
    FirstIndex:int=int + int(A=Get CurrentPanelIndex() -> (CurrentPanelIndex=0))
    LastIndex:int=int - int(A=GetChildrenCount(self=@n1) -> (ReturnValue=0))
    Break:Exec=@n5.then
   ) -> (LoopBody:Exec, Index:int, Completed:Exec) {is=title K2Node_MacroInstance}
   > n3 = Set CurrentWidget(CurrentWidget:Widget Object Reference=GetChildAt(self=Get VerticalBox_Panels(), Index=@n2)) -> (Output_Get:Widget Object Reference)
   > n4 = Cast To SizeBox(Object=@n3)
   > n5 = SetSize!(
    execute
    self:Vertical Box Slot Object Reference=SlotAsVerticalBoxSlot(Widget=@n4)
    InSize:Slate Child Size Structure=Make Slate Child Size!(Value:float=1, SizeRule:ESlateSizeRule Enum=Fill) -> (SlateChildSize:Slate Child Size Structure) {is=title K2Node_MakeStruct}
   ) -> (then)
  ]}
 PreUpdatePanelSize {nodes [
   PreUpdatePanelSize() {is=title K2Node_FunctionEntry}
   > n1 = Cast To SizeBox(Object=GetChildAt(self=Get VerticalBox_Panels(), Index=Get DesiredPanelIndex() -> (DesiredPanelIndex=0)))
   > n2 = Set SizeBox(SizeBox:Size Box Object Reference=@n1) -> (Output_Get:Size Box Object Reference)
   > n3 = Set VerticalBoxSlot(VerticalBoxSlot:Vertical Box Slot Object Reference=SlotAsVerticalBoxSlot(Widget=@n2)) -> (Output_Get:Vertical Box Slot Object Reference)
   n4 = Get Size!(self:Vertical Box Slot Object Reference=@n3) -> (Size:Slate Child Size Structure)
   > Branch(Condition="Not Equal (Enum)"!(
     A:ESlateSizeRule Enum=Break Slate Child Size!(SlateChildSize:"Slate Child Size Structure (by ref)"=@n4) -> (Value:float=~, SizeRule:ESlateSizeRule Enum) {is=title K2Node_BreakStruct}.SizeRule
     B:ESlateSizeRule Enum=Automatic
    ) -> (ReturnValue:bool))
   n5 = Get VerticalBox_Panels()
   > SetMaxDesiredHeight!(
    execute
    self:Size Box Object Reference=Get SizeBox() -> (SizeBox:Size Box Object Reference)
    InMaxDesiredHeight:float=GetLocalSize!(Geometry=GetCachedGeometry(self=@n5)) -> (ReturnValue_X:double, ReturnValue_Y:double).ReturnValue_Y
   ) -> (then)
   > SetSize(
    self=Get VerticalBoxSlot() -> (VerticalBoxSlot:Vertical Box Slot Object Reference)
    InSize=Make Slate Child Size(Value=1, SizeRule=Automatic)
   )
  ]}
 Set All Categories {nodes [
   n1 = Set All Categories!() -> (then, In Category List:Array of Discover Tile Item View Model Object References) {is=title K2Node_FunctionEntry}
   > Set Categories!(
    execute
    self:Activity Category Tile Panel Base VM Object Reference=Get TilePanel_All_VMversion()
    In Categories:Array of Discover Tile Item View Model Object References=@n1
   ) -> (then) {is=title K2Node_CallFunction}
  ]}
 Set All Categories Panel Display Name {nodes [
   n1 = Set All Categories Panel Display Name!() -> (then, In Panel Name:Text) {is=title K2Node_FunctionEntry}
   > Set Panel Display Name!(execute, self:Activity Category Tile Panel Base VM Object Reference=Get TilePanel_All_VMversion(), In Panel Name:Text=@n1) -> (then) {is=title K2Node_CallFunction}
  ]}
 Set Featured Categories {nodes [
   n1 = Set Featured Categories!() -> (then, In Category List:Array of Discover Tile Item View Model Object References) {is=title K2Node_FunctionEntry}
   > Set Categories(self=Get TilePanel_Featured_VMversion(), In Categories=@n1)
   > PlayAnimation(InAnimation=Get Intro())
   > PlayAnimationReverse(InAnimation=Get Anim_DisplayLoading() -> (Anim_DisplayLoading:Widget Animation Object Reference))
  ]}
 Set Featured Categories Panel Display Name {nodes [
   n1 = Set Featured Categories Panel Display Name!() -> (then, In Panel Name:Text) {is=title K2Node_FunctionEntry}
   > Set Panel Display Name(self=Get TilePanel_Featured_VMversion(), In Panel Name=@n1)
  ]}
 Set Panel Data {nodes [Set Panel Data() {is=title K2Node_FunctionEntry}]}
 Set ViewModel {nodes [
   n1 = Set ViewModel!() -> (then, In Discover Browse VM:Discover Browse View Model Object Reference) {is=title K2Node_FunctionEntry}
   > Set FortDiscoverBrowseViewModel(FortDiscoverBrowseViewModel:Discover Browse View Model Object Reference=@n1) -> (Output_Get:Discover Browse View Model Object Reference)
  ]}
 SetCategoryIndexAnimator {nodes [
   n1 = SetCategoryIndexAnimator!() -> (then, NewValue:double=~) {is=title K2Node_FunctionEntry}
   n2 = Get CurrentPlayDirection() -> (CurrentPlayDirection:EUMGSequencePlayMode Enum=Forward)
   > n3 = Branch(Condition="Nearly Equal (Float)"(
     A=@n1
     B=Select!(Forward:double=1, Reverse:double=0, PingPong:double=0, Index:EUMGSequencePlayMode Enum=@n2) -> (ReturnValue:double)
     ErrorTolerance=0.001
    ))
   > n4 = Switch on EUMGSequencePlayMode!(execute, Selection:EUMGSequencePlayMode Enum=Get CurrentPlayDirection() -> (CurrentPlayDirection=Forward)) -> (Forward:Exec, Reverse:Exec, PingPong:Exec) {is=title K2Node_SwitchEnum}
   > UpdateMouseWheelForIndex!(execute, self, PanelIndex:int=Get CurrentPanelIndex() -> (CurrentPanelIndex=0), bListenForInput:bool=false) -> (then)
   > UpdateMouseWheelForIndex(PanelIndex=Get DesiredPanelIndex() -> (DesiredPanelIndex=0), bListenForInput=true)
   > n5 = SetCurrentPanelIndex(CurrentPanelIndex=Get DesiredPanelIndex() -> (DesiredPanelIndex=0))
   n6 = PostUpdatePanelSize(execute=[@n5.then, @n8.then]) {is=title K2Node_CallFunction}
   n7 = Lerp(
    A=GetLocalPanelPosition!(Index:int=Get CurrentPanelIndex() -> (CurrentPanelIndex=0)) -> (Position:double) {is=title K2Node_CallFunction}
    B=GetLocalPanelPosition(Index=Get DesiredPanelIndex() -> (DesiredPanelIndex=0))
    Alpha=@n1
   )
   SetRenderTranslation(
    execute=[@n6.then, @n3.else]
    self=Get VerticalBox_Panels()
    Translation=MakeVector2D(X=0, Y=float * float(A=@n7, B=-1) {is=title K2Node_PromotableOperator})
   )
   UpdateMouseWheelForIndex(execute=[@n4.PingPong, @n4.Reverse], PanelIndex=Get CurrentPanelIndex() -> (CurrentPanelIndex=0), bListenForInput=true)
   > UpdateMouseWheelForIndex(PanelIndex=Get DesiredPanelIndex() -> (DesiredPanelIndex=0), bListenForInput=false)
   > n8 = Set DesiredPanelIndex(DesiredPanelIndex=Get CurrentPanelIndex() -> (CurrentPanelIndex=0)) -> (Output_Get=0)
  ]}
 SetCurrentPanelIndex {nodes [
   n1 = SetCurrentPanelIndex!() -> (then, CurrentPanelIndex:int) {is=title K2Node_FunctionEntry}
   > n2 = Set CurrentPanelIndex(CurrentPanelIndex:int=@n1) -> (Output_Get:int=0)
   > SetBottomButtonVisibility!(
    execute
    self:WBP Bottom Bar Decoy Button Object Reference=Get WBP_BottomBarHack_BackToTop() -> (WBP_BottomBarHack_BackToTop:WBP Bottom Bar Decoy Button Object Reference)
    ShouldButtonBeVisible:bool="integer > integer"(A=@n2)
   ) -> (then)
  ]}
 SnapToEndOfAnimation {nodes [
   n1 = SnapToEndOfAnimation!() -> (then, InAnimation:Widget Animation Object Reference, PlayMode:EUMGSequencePlayMode Enum=Forward) {is=title K2Node_FunctionEntry}
   > PlayAnimation(InAnimation=@n1, StartAtTime=GetEndTime(self=@n1), PlayMode=@n1.PlayMode)
  ]}
 SwitchToGamepad {nodes [
   SwitchToGamepad() {is=title K2Node_FunctionEntry}
   > n1 = Set CurrentPanel(CurrentPanel:Fort Activity Category Tile Panel Object Reference) -> (Output_Get:Fort Activity Category Tile Panel Object Reference)
   > n2 = Set PanelIndex(PanelIndex:int=GetIndexOfPanel!(Widget:Widget Object Reference=@n1) -> (Index:int=0) {is=title K2Node_CallFunction}) -> (Output_Get:int=0)
   n3 = Get CurrentPanel() -> (CurrentPanel:Fort Activity Category Tile Panel Object Reference)
   n4 = Get CurrentPanelIndex() -> (CurrentPanelIndex=0)
   > n5 = Branch(Condition=OR Boolean(A="Equal (Integer)"(A=@n2, B=@n4), B=IsWidgetFullSizedAndOnScreen(Widget=@n3) {is=title K2Node_CallFunction}))
   AnimateToDesiredPanel(execute=@n5.else, NewDesiredIndex=Get PanelIndex() -> (PanelIndex:int=0))
  ]}
 UpdateMouseWheelForIndex {nodes [
   n1 = UpdateMouseWheelForIndex!() -> (then, PanelIndex:int, bListenForInput:bool=~) {is=title K2Node_FunctionEntry}
   > n2 = Set TilePanel(TilePanel:Activity Category Tile Panel Base Object Reference=GetTilePanelFromIndex(PanelIndex=@n1)) -> (Output_Get:Activity Category Tile Panel Base Object Reference)
   n3 = Get TilePanel_Featured_VMversion()
   > EnableMouseWheelInput(
    self:Activity Category Tile Panel Base Object Reference=@n2
    bListenForInput=AND Boolean(
     A="Not Equal (Object)"(A:Object Reference=@n2, B:Object Reference=@n3) -> (ReturnValue:bool=~) {is=title K2Node_PromotableOperator}
     B=@n1.bListenForInput
    )
   )
  ]}
]
variables [
 CategoryIndexAnimator {type=number}
 CurrentInputType {type=ECommonInputType, default=MouseAndKeyboard}
 CurrentPanelIndex {type=integer}
 CurrentPlayDirection {type=EUMGSequencePlayMode, default=Forward}
 DesiredPanelIndex {type=integer, default=1}
 IsClickedOnBottomPanel {type=boolean}
 LastDesiredPanelIndex {type=integer, default=1}
 LastSelectedPanel {type=ActivityCategoryTilePanelBase_C, default=null}
 OnBackToTopClicked {}
 OnPlayIntro {}
 OnRowUpdated {}
]
widgets {
 info {inheritedWidgetCount=0, parentClass=&CommonActivatableWidget, rootWidgetClass=&SafeZone, widgetCount=13}
 widgets [
  SafeZone_0 {widgetClassPath=&SafeZone, widget=SafeZone_0, properties {padLeft=false, padRight=false}}
  Overlay {widgetClassPath=&Overlay, parent=SafeZone_0, widget=Overlay, slot {class=SafeZoneSlot}}
  Overlay_Content {
   widgetClassPath=&Overlay
   parent=Overlay
   widget=Overlay_Content
   slot {class=OverlaySlot, properties {horizontalAlignment=HAlign_Fill, verticalAlignment=VAlign_Fill}}
  }
  SizeBox_0 {
   widgetClassPath=&SizeBox
   parent=Overlay_Content
   widget=SizeBox_0
   properties {bOverride_MinDesiredWidth=true, minDesiredWidth=1760}
   slot {class=OverlaySlot, properties {horizontalAlignment=HAlign_Center, verticalAlignment=VAlign_Fill}}
  }
  VerticalBox_Panels {widgetClassPath=&VerticalBox, bIsVariable=true, parent=SizeBox_0, widget=VerticalBox_Panels, slot {class=SizeBoxSlot}}
  TilePanel_Featured_VMversion {
   widgetClassPath=&/ActivityCategoryTilePanelBase_VM
   bIsVariable=true
   parent=VerticalBox_Panels
   widget=TilePanel_Featured_VMversion
   properties {
    entryHeight=144
    navigation=&self:WidgetTree.TilePanel_Featured_VMversion.WidgetNavigation_1
    visibility=Visible
    desiredFocusWidget {widgetName=None}
   }
   slot {class=VerticalBoxSlot}
  }
  SizeBox_DynamicBounding {
   widgetClassPath=&SizeBox
   parent=VerticalBox_Panels
   widget=SizeBox_DynamicBounding
   properties {maxDesiredHeight=500}
   slot {class=VerticalBoxSlot, properties {padding=<ltrb 0 32 0 0>, size {sizeRule=Fill, value=1}}}
  }
  TilePanel_All_VMversion {
   widgetClassPath=&ActivityCategoryTilePanelBase_VM
   bIsVariable=true
   parent=SizeBox_DynamicBounding
   widget=TilePanel_All_VMversion
   properties {
    bIsFocusable=false
    entryHeight=120
    entryWidth=288
    navigation=&self:WidgetTree.TilePanel_All_VMversion.WidgetNavigation_0
    visibility=Visible
    desiredFocusWidget {widgetName=None}
   }
   slot {class=SizeBoxSlot}
  }
  Image_HeaderSoftEdge {
   widgetClassPath=&Image
   bIsVariable=true
   parent=Overlay_Content
   widget=Image_HeaderSoftEdge
   properties {
    renderOpacity=0
    visibility=SelfHitTestInvisible
    brush {
     bIsDynamicallyLoaded=false
     drawAs=Image
     imageType=NoImage
     mirroring=NoMirror
     resourceName=None
     resourceObject=/DiscoveryBrowser/TabScreens/DiscoverHome/Materials/MI_UI_Discovery_NebulousBackgroundWithTopMask
     tiling=NoTile
     imageSize=<32 50>
     margin=<ltrb 0 0 0 0>
     outlineSettings {
      bUseBrushTransparency=false
      roundingType=HalfHeightRadius
      width=0
      color {colorUseRule=UseColor_Specified, specifiedColor=<rgba 0 0 0 0>}
      cornerRadii=<0 0 0 0>
     }
     tintColor {colorUseRule=UseColor_Specified, specifiedColor=<rgba 1 1 1 1>}
     uVRegion {bIsValid=false, max=<0 0>, min=<0 0>}
    }
   }
   slot {class=OverlaySlot, properties {horizontalAlignment=HAlign_Fill}}
  }
  WBP_UIKit_Throbber {
   widgetClassPath=/FortUILibrary/UIKit/Throbber/WBP_UIKit_Throbber
   bIsVariable=true
   parent=Overlay
   widget=WBP_UIKit_Throbber
   properties {displayBackground=false, throbberSize=Large}
   slot {class=OverlaySlot, properties {horizontalAlignment=HAlign_Center, verticalAlignment=VAlign_Center}}
  }
  SZBottomButtons {
   widgetClassPath=&SafeZone
   parent=Overlay
   widget=SZBottomButtons
   properties {padTop=false}
   slot {class=OverlaySlot, properties {horizontalAlignment=HAlign_Right, verticalAlignment=VAlign_Bottom, padding=<ltrb 0 0 80 32>}}
  }
  HBBottomButtons {widgetClassPath=&HorizontalBox, parent=SZBottomButtons, widget=HBBottomButtons, slot {class=SafeZoneSlot}}
  WBP_BottomBarHack_BackToTop {
   widgetClassPath=/Game/UI/Frontend/BottomBar/WBP_BottomBarDecoyButton
   bIsVariable=true
   parent=HBBottomButtons
   widget=WBP_BottomBarHack_BackToTop
   properties {minWidth=120, triggeringInputAction {dataTable=&InputActionDataTable, rowName=BackToTop}}
   slot {class=HorizontalBoxSlot}
  }
 ]
}
