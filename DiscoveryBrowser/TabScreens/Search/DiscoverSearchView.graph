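# Activatable Discovery browser search screen that switches between category browsing and island name or code search.
# Text input feeds the search view model; query state selects a spinner, results tile view, no-results or blocked-island message.
# Tab changes, activation and view switches drive browse and search analytics.
# Scrolling controls an edge fade and decoy Back/Back to Top buttons, while custom navigation routes focus between field and tiles.
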WidgetBlueprint /DiscoveryBrowser/TabScreens/Search/DiscoverSearchView
animations [Intro {
  binding_count=2
  length_seconds=0.6
  object=&self:Intro
  sequence {
   bindings [
    Switcher_CategoriesOrSearch {
     id=#1
     possessed_class=CommonWidgetSwitcherLegacy
     tracks [
      {
       type=MovieSceneFloatTrack
       display_name=Render Opacity
       sections [{type=MovieSceneFloatSection, channels [None {type=MovieSceneScriptingFloatChannel, keys [[0,0],[6,0],[24,1]]}]}]
      }
      {
       type=MovieScene2DTransformTrack
       sections [{type=MovieScene2DTransformSection, channels [Translation.Y {type=MovieSceneScriptingFloatChannel, keys [[0,40],[6,40],[24,0]]}]}]
      }
     ]
    }
    FilteredInputField {
     id=#2
     possessed_class=WBP_UIKit_FilteredInputField_C
     tracks [
      {
       type=MovieSceneFloatTrack
       display_name=Render Opacity
       sections [{type=MovieSceneFloatSection, channels [None {type=MovieSceneScriptingFloatChannel, keys [[0,0],[18,1]]}]}]
      }
      {
       type=MovieScene2DTransformTrack
       sections [{type=MovieScene2DTransformSection, channels [Translation.Y {type=MovieSceneScriptingFloatChannel, keys [[0,40],[18,0]]}]}]
      }
     ]
    }
   ]
   display_rate {numerator=60}
  }
 }]
graphs [
 BP_GetDesiredFocusTarget {nodes [
   Get Desired Focus Target()
   > Return Node!(
    execute
    ReturnValue:Widget Object Reference=Get FilteredInputField() -> (FilteredInputField:WBP UIKit Filtered Input Field Object Reference)
   )
  ]}
 EventGraph {nodes [
   n1 = "On Entry Generated (TileView_IslandsResults)"!() -> (OutputDelegate, then, Widget:User Widget Object Reference) {is=title K2Node_ComponentBoundEvent}
   > n2 = Cast To WBP_IslandTile_Minimal!(execute, Object=@n1.Widget) -> (then, CastFailed, AsWBP Island Tile Minimal:WBP Island Tile Minimal Object Reference) {is=title K2Node_DynamicCast}
   n3 = Get WBP_Discover_Tile_Base!(self:WBP Island Tile Minimal Object Reference=@n2) -> (WBP_Discover_Tile_Base:WBP Discover Tile Base Object Reference)
   > Set Tile Size!(
    execute
    self:WBP Discover Tile Base Object Reference=@n3
    InSize:Vector 2D Structure=ColumnSizeToTileSize!(self:WBP Discover Tile Base Object Reference=@n3, NewColumnSize:int=5, IsSquareTiles:bool=false) -> (NewTileSize:Vector 2D Structure)
   ) -> (then) {is=title K2Node_CallFunction}
   n4 = BreakVector2D(InVec=Get Dynamic Tile Size!(self:WBP Discover Tile Base Object Reference=@n3) -> (DynamicTileSize:Vector 2D Structure) {is=title K2Node_CallFunction})
   n5 = Get TileView_IslandsResults() -> (TileView_IslandsResults:Fort Activity Tile View Object Reference)
   > SetEntryWidth!(execute, self:Tile View Object Reference=@n5, NewWidth:float=@n4) -> (then)
   > SetEntryHeight!(execute, self:Tile View Object Reference=@n5, NewHeight:float=@n4.Y) -> (then)
   n6 = "On Clicked (WBP_BottomBarHack_Back)"!() -> (OutputDelegate, then, Button:Common Button Base Object Reference) {is=title K2Node_ComponentBoundEvent}
   ClearSearch(
    execute=[@n6.then, @n11.else]
    self:Discover Search Model Object Reference=Get FortDiscoverSearchViewModel() -> (FortDiscoverSearchViewModel:Discover Search Model Object Reference)
   )
   > SetInputFieldText!(execute, self:WBP UIKit Filtered Input Field Object Reference=Get FilteredInputField(), Text:Text) -> (then)
   > SetTextFocus(self:WBP UIKit Filtered Input Field Object Reference=Get FilteredInputField())
   n7 = "On Text Changed (FilteredInputField)"!() -> (OutputDelegate, then, Text:Text) {is=title K2Node_ComponentBoundEvent}
   > Set SearchTerm!(
    execute
    SearchTerm:String="To String (Text)"(InText=@n7.Text)
    self:Discover Search Model Object Reference=Get FortDiscoverSearchViewModel()
   ) -> (then, Output_Get:String)
   n8 = "On Text Committed (FilteredInputField)"!() -> (OutputDelegate, then, CommitMethod:ETextCommit Enum=Default) {is=title K2Node_ComponentBoundEvent}
   > n9 = Switch on ETextCommit!(execute, Selection:ETextCommit Enum=@n8.CommitMethod) -> (Default:Exec, OnEnter:Exec, OnUserMovedFocus:Exec, OnCleared:Exec) {is=title K2Node_SwitchEnum}
   n10 = GetCurrentText!(execute=@n9.OnEnter, self:WBP UIKit Filtered Input Field Object Reference=Get FilteredInputField()) -> (then, Text:Text)
   > n11 = Branch(Condition=NOT Boolean(A=TextIsEmpty(InText=@n10)))
   > ActionSearch(self:Discover Search Model Object Reference=Get FortDiscoverSearchViewModel())
   "On Clicked (WBP_BottomBarHack_BackToTop)"!() -> (OutputDelegate, then, Button:Common Button Base Object Reference) {is=title K2Node_ComponentBoundEvent}
   > n12 = ScrollToTop(self:List View Base Object Reference=Get TileView_IslandsResults())
   SetTextFocus(execute=[@n12.then, @n22.then], self=Get FilteredInputField())
   n13 = OnTileViewScrolled!() -> (OutputDelegate, then, ItemOffset:float=~, DistanceRemaining:float=~) {is=title K2Node_CustomEvent}
   n14 = "Absolute (Float)"(A=float - float(A=@n13.DistanceRemaining, B=1))
   > SetRenderOpacity(
    self=Get Image_SoftEdgeGradient_Islands() -> (Image_SoftEdgeGradient_Islands:Image Object Reference)
    InOpacity="To Float (Integer)"(InInt=Ceil(A=@n14) {is=title K2Node_CallFunction})
   )
   n15 = Get TileView_IslandsResults()
   > n16 = Branch(Condition=AND Boolean(A="float > float"(A=@n13.ItemOffset), B=IsFullsizedOnScreen!(Widget:Widget Object Reference=@n15) -> (IsFullsized:bool)))
   > SetBottomButtonVisibility!(
    execute
    self:WBP Bottom Bar Decoy Button Object Reference=Get WBP_BottomBarHack_BackToTop() -> (WBP_BottomBarHack_BackToTop:WBP Bottom Bar Decoy Button Object Reference)
    ShouldButtonBeVisible:bool=true
   ) -> (then)
   > SetBottomButtonVisibility(self=Get WBP_BottomBarHack_Back() -> (WBP_BottomBarHack_Back:WBP Bottom Bar Decoy Button Object Reference), ShouldButtonBeVisible=false)
   SetBottomButtonVisibility(execute=@n16.else, self=Get WBP_BottomBarHack_BackToTop(), ShouldButtonBeVisible=false)
   > SetBottomButtonVisibility(self=Get WBP_BottomBarHack_Back(), ShouldButtonBeVisible=true)
   n17 = TopLevelTabChanged!() -> (OutputDelegate, then, LeavingFeature:EFortUIFeature Enum=ShowHome, EnteringFeature:EFortUIFeature Enum=ShowHome) {is=title K2Node_CustomEvent}
   > n18 = Branch(Condition="Equal (Enum)"(A:EFortUIFeature Enum=@n17.LeavingFeature, B:EFortUIFeature Enum=EnableSearchTabInTopNavigationBar))
   n19 = Get FortDiscoverSearchViewModel()
   > FireAnalyticsForBrowseAndSearch!(
    execute
    self:Discover Search Model Object Reference=@n19
    BrowseViewModel:Discover Browse View Model Object Reference=Get FortDiscoverBrowseViewModel() -> (FortDiscoverBrowseViewModel:Discover Browse View Model Object Reference)
   ) -> (then)
   Branch(execute=@n18.else, Condition="Equal (Enum)"(A=@n17.EnteringFeature, B=EnableSearchTabInTopNavigationBar))
   > StartAnalyticsSession(self:Discover Search Model Object Reference=@n19)
   > n20 = Branch(Condition=IsAnySearchActiveOrInProgress!(self:Discover Search Model Object Reference=Get FortDiscoverSearchViewModel()) -> (ReturnValue:bool))
   > StartAnalyticsImpressions(self:Discover Search Model Object Reference=@n19)
   > RespondToVisibilityChange!(execute, self:Fort Activity Tile View Object Reference=Get TileView_IslandsResults(), bIsVisible:bool=true) -> (then)
   StartAnalyticsImpressions(execute=@n20.else, self:Discover Surface View Model Object Reference=Get FortDiscoverBrowseViewModel())
   OnBackToTopClickedCategory() {is=title K2Node_CustomEvent}
   n21 = Get TileView_Categories!(self:Activity Category Tile Panel Base VM Object Reference=Get TilePanel_All_VMversion!(self:Activity Browser Browse View VM Object Reference=Get ActivityBrowserBrowseView_VM() -> (ActivityBrowserBrowseView_VM:Activity Browser Browse View VM Object Reference)) -> (TilePanel_All_VMversion:Activity Category Tile Panel Base VM Object Reference)) -> (TileView_Categories:Fort Activity Tile View Object Reference)
   > SetSelectedIndex!(execute, self:List View Object Reference=@n21, Index:int=0) -> (then)
   > n22 = SetFocus(self=@n21)
   n23 = OnIntroFinished() {is=title K2Node_CustomEvent}
   > Branch(Condition=NOR Boolean!(A:bool=IsMobilePlatform!() -> (ReturnValue:bool), B:bool=IsConsolePlatform!() -> (ReturnValue:bool)) -> (ReturnValue:bool))
   > SetTextFocus(self=Get FilteredInputField())
   Event PreConstruct()
   > SetMobileStyle() {is=title K2Node_CallFunction}
   > n24 = GetDynamicMaterial(self=Get Image_SoftEdgeGradient_Islands())
   > SetScalarParameterValue(self=@n24, ParameterName=Overlay Rotation, Value=0.5)
   > SetResultsTerm!(execute, self, ResultsTerm:String=$Search Result) -> (then)
   > BindToAthenaFrontendTabChange!(execute, Context:Widget Object Reference=Self-Reference(), EventToBind:"Delegate (by ref)"=@n17) -> (then)
   Event Construct()
   n25 = Get FortDiscoverBrowseViewModel()
   > n26 = Branch(Condition=ShouldQueryOnConstruct!(self:Discover Surface View Model Object Reference=@n25) -> (ReturnValue:bool))
   > n27 = Set SurfaceName!(execute, SurfaceName:Name=CreativeDiscoverySurface_Browse, self:Discover Surface View Model Object Reference=@n25) -> (then, Output_Get:Name=None)
   SetBottomButtonVisibility(execute=[@n27.then, @n26.else], self=Get WBP_BottomBarHack_BackToTop(), ShouldButtonBeVisible=false)
   > SetBottomButtonVisibility(self=Get WBP_BottomBarHack_Back(), ShouldButtonBeVisible=false)
   > Bind Event to On List View Scrolled(self:List View Object Reference=Get TileView_IslandsResults(), Delegate=@n13) {is=title K2Node_AddDelegate}
   > Bind Event to On Back to Top Clicked(self:Activity Browser Browse View VM Object Reference=Get ActivityBrowserBrowseView_VM(), Delegate=Create Event()) {is=title K2Node_AddDelegate}
   > Set SearchMode!(execute, SearchMode:EDiscoverSearchMode Enum=Island, self:Discover Search Model Object Reference=Get FortDiscoverSearchViewModel()) -> (then, Output_Get:EDiscoverSearchMode Enum=Island)
   Event On Activated()
   n28 = Get FortDiscoverBrowseViewModel()
   > n29 = Branch(Condition=ShouldQueryOnConstruct(self=@n28))
   Fire Event Main View Opened!(
    execute=[@n34.then, @n29.then, @n33.else]
    View:Widget Object Reference=Self-Reference()
    AdditionalInfo:Flow Analytics Additional Info Structure
    bIsOverlay:bool=false
   ) -> (then)
   > BindToAnimationFinished(Animation=Get Intro(), Delegate=@n23)
   > PlayAnimationForward(InAnimation=Get Intro())
   n30 = Get ActivityBrowserBrowseView_VM()
   > ActivateWidget(self=@n30)
   > SpawnSound2D!(
    execute
    Sound:Sound Base Object Reference=Get IntroWhooshSound() -> (IntroWhooshSound:Sound Base Object Reference)
    VolumeMultiplier:float=1
    PitchMultiplier:float=1
    StartTime:float=0
    ConcurrencySettings:Sound Concurrency Object Reference
    bPersistAcrossLevelTransition:bool=false
    bAutoDestroy:bool=true
   ) -> (then, ReturnValue:Audio Component Object Reference)
   n31 = IsAnySearchActiveOrInProgress(self=Get FortDiscoverSearchViewModel())
   > n32 = Branch(Condition=@n31)
   > StartAnalyticsVisibilityImpressions(self:Discover Search Model Object Reference=Get FortDiscoverSearchViewModel())
   StartVisibleImpressions(execute=@n32.else, self:Discover Surface View Model Object Reference=Get FortDiscoverBrowseViewModel())
   n33 = Branch(
    execute=@n29.else
    Condition="Equal (Name)"(A=Get SurfaceName!(self:Discover Surface View Model Object Reference=@n28) -> (SurfaceName:Name=None))
   )
   > n34 = Set SurfaceName(SurfaceName=CreativeDiscoverySurface_Browse, self=@n28) -> (Output_Get=None)
   Event On Deactivated()
   > UnbindFromAnimationFinished!(execute, self:User Widget Object Reference, Animation:Widget Animation Object Reference=Get Intro(), Delegate:Delegate=@n23) -> (then)
   > Check and Stop Animation!(execute:Exec, Animation:Widget Animation Object Reference=Get Intro()) -> (then:Exec) {is=title K2Node_MacroInstance}
   > DeactivateWidget(self=@n30)
   > n35 = Branch(Condition=@n31)
   > StopAnalyticsVisibilityImpressions(self:Discover Search Model Object Reference=Get FortDiscoverSearchViewModel())
   StopVisibleImpressions(execute=@n35.else, self:Discover Surface View Model Object Reference=Get FortDiscoverBrowseViewModel())
   Get TileView_IslandsResults()
  ]}
 IsFullsizedOnScreen {nodes [
   n1 = IsFullsizedOnScreen!() -> (then, Widget:Widget Object Reference) {is=title K2Node_FunctionEntry}
   n2 = GetPaintSpaceGeometry!(self:Widget Object Reference=@n1) -> (ReturnValue:"Geometry Structure (by ref)")
   > Return Node!(
    execute
    IsFullsized:bool="float < float"(
     A=GetLocalSize!(Geometry:"Geometry Structure (by ref)"=@n2) -> (ReturnValue_X:double, ReturnValue_Y:double).ReturnValue_Y
     B=GetDesiredSize!(self:Widget Object Reference=@n1) -> (ReturnValue_X:double, ReturnValue_Y:double).ReturnValue_Y
    )
   )
  ]}
 NavigateDownToBrowseOrSearch {nodes [
   NavigateDownToBrowseOrSearch!() -> (then, Navigation:EUINavigation Enum) {is=title K2Node_FunctionEntry}
   n1 = Get FortDiscoverSearchViewModel()
   n2 = IsAnySearchActiveOrInProgress(self=@n1)
   > n3 = Branch(Condition=@n2)
   n4 = Length(TargetArray:Array of Discover Tile Item View Model Object References=Get SearchResults!(self:Discover Search Model Object Reference=@n1) -> (SearchResults:Array of Discover Tile Item View Model Object References))
   n5 = Get TileView_Categories(self=Get TilePanel_Featured_VMversion!(self:Activity Browser Browse View VM Object Reference=Get ActivityBrowserBrowseView_VM()) -> (TilePanel_Featured_VMversion:Activity Category Tile Panel Base VM Object Reference))
   n6 = Select(
    Option 0:Widget Object Reference=@n5
    Option 1:Widget Object Reference=Select(Option 0=Get FilteredInputField(), Option 1=Get TileView_IslandsResults(), Index="integer > integer"(A=@n4))
    Index=@n2
   ) -> (ReturnValue:Widget Object Reference)
   PrintString!(
    execute=[@n3.then, @n7.then]
    InString=Append(A="Navigated DOWN to: ", B=GetDisplayName(Object=@n6))
    bPrintToScreen
    bPrintToLog
    TextColor
    Duration
    Key
   ) -> (then)
   > Return Node!(execute, ReturnValue:Widget Object Reference=@n6)
   n7 = SetSelectedIndex(execute=@n3.else, self=@n5, Index=0)
   Get FilteredInputField()
  ]}
 NavigateUpToTextEntry {nodes [
   NavigateUpToTextEntry!() -> (then, Navigation:EUINavigation Enum) {is=title K2Node_FunctionEntry}
   > Return Node(ReturnValue=Get InputField!(self:WBP UIKit Filtered Input Field Object Reference=Get FilteredInputField()) -> (InputField:WBP UIKit Input Field Object Reference))
  ]}
 On Update Providers {nodes [
   On Update Providers!() -> (then, ProvidersVM:Array of Discover Provider View Model Object References) {is=title K2Node_FunctionEntry}
   > SetDataFromProviders(self:Discover Browse View Model Object Reference=Get FortDiscoverBrowseViewModel())
   > Set Browse View VM() {is=title K2Node_CallFunction}
  ]}
 OnQueryingMorePages {nodes [
   n1 = OnQueryingMorePages!() -> (then, LoadingState:EDiscoverLoadingState Enum=Unset) {is=title K2Node_FunctionEntry}
   > SetVisibility(
    self=Get WBP_LoadingMorePages() -> (WBP_LoadingMorePages:WBP Loading More Pages Object Reference)
    InVisibility=Select(
     Option 0:ESlateVisibility Enum=Collapsed
     Option 1:ESlateVisibility Enum=HitTestInvisible
     Index="Equal (Enum)"(A:EDiscoverLoadingState Enum=@n1, B:EDiscoverLoadingState Enum=Loading)
    ) -> (ReturnValue:ESlateVisibility Enum)
   )
  ]}
 Set Browse View VM {nodes [
   Set Browse View VM() {is=title K2Node_FunctionEntry}
   > Set ViewModel!(
    execute
    self:Activity Browser Browse View VM Object Reference=Get ActivityBrowserBrowseView_VM()
    In Discover Browse VM:Discover Browse View Model Object Reference=Get FortDiscoverBrowseViewModel()
   ) -> (then) {is=title K2Node_CallFunction}
  ]}
 SetMobileStyle {nodes [
   SetMobileStyle() {is=title K2Node_FunctionEntry}
   > Branch(Condition=IsMobileGame())
   > SetSidesToPad!(
    execute
    self:Safe Zone Object Reference=Get SafeZone() -> (SafeZone:Safe Zone Object Reference)
    InPadLeft:bool=true
    InPadRight:bool=true
    InPadTop:bool=true
    InPadBottom:bool=false
   ) -> (then)
   > SetPadding!(
    execute
    self:Overlay Slot Object Reference=SlotAsOverlaySlot(Widget=Get TileView_IslandsResults())
    InPadding_Left:float=48
    InPadding_Top:float=0
    InPadding_Right:float=48
    InPadding_Bottom:float=0
   ) -> (then)
  ]}
 SetResultsTerm {nodes [
   n1 = SetResultsTerm!() -> (then, ResultsTerm:String) {is=title K2Node_FunctionEntry}
   > SetText(
    self=Get RichResultsText() -> (RichResultsText:Common Rich Text Block Object Reference)
    InText=Format Text!(Format="LOCTEXT(\"Results for <text color=\\\"FFFFFF\\\" fontface=\\\"black\\\">{ResultsTerm}</> \")", ResultsTerm:String=@n1) -> (Result)
   )
  ]}
 SetSearchInProgress {nodes [
   n1 = SetSearchInProgress!() -> (then, SearchInProgress:bool=~) {is=title K2Node_FunctionEntry}
   > n2 = Branch(Condition=@n1)
   > SetActiveWidget(
    self:Widget Switcher Object Reference=Get SearchResult_Switcher() -> (SearchResult_Switcher:Common Widget Switcher Legacy Object Reference)
    Widget=Get SearchSpinner() -> (SearchSpinner:Grid Panel Object Reference)
   )
   n3 = Branch(execute=@n2.else, Condition=IsAnySearchActiveOrInProgress(self=Get FortDiscoverSearchViewModel()))
   > n4 = Switch on ELinkEntryQueryResult!(
    execute
    Selection:ELinkEntryQueryResult Enum=Get LastLinkEntryQueryResult!(self:Discover Search Model Object Reference=Get FortDiscoverSearchViewModel()) -> (LastLinkEntryQueryResult:ELinkEntryQueryResult Enum=Success)
   ) -> (
    Success:Exec
    PartialSuccess:Exec
    FailedQuery:Exec
    FailedPlayerDataQuery:Exec
    FailedParse:Exec
    NotAllowedOnPlatform:Exec
    CabinModeBlocked:Exec
    UnableToFind:Exec
    NotVisible:Exec
    TournamentDataUnavailable:Exec
    FailedStringSanitization:Exec
    FailedSubModePopulation:Exec
    FailedSubModeSurfacePopulation:Exec
    FailedSubModeParsing:Exec
    FilteredOut:Exec
    RatingUnrated:Exec
    RatingRestricted:Exec
    ModeratorOnly:Exec
    FailedPlayerOrientedDataHydration:Exec
    InvalidQuery:Exec
    FailedReferenceLinkPopulation:Exec
   ) {is=title K2Node_SwitchEnum}
   n5 = Get FortDiscoverSearchViewModel()
   n6 = Branch(
    execute=[@n4.PartialSuccess, @n4.Success]
    Condition="Not Equal (Integer)"(A=GetIslandResultCount!(self:Discover Search Model Object Reference=@n5) -> (ReturnValue:int=0))
   )
   > SetActiveWidget(self=Get SearchResult_Switcher(), Widget=Get ResultsGrid() -> (ResultsGrid:Grid Panel Object Reference))
   SetActiveWidget(
    execute=[
     @n6.else
     @n4.CabinModeBlocked
     @n4.FailedParse
     @n4.FailedQuery
     @n4.FailedStringSanitization
     @n4.FailedSubModeParsing
     @n4.FailedSubModePopulation
     @n4.FailedSubModeSurfacePopulation
     @n4.FilteredOut
     @n4.NotAllowedOnPlatform
     @n4.NotVisible
     @n4.UnableToFind
    ]
    self=Get SearchResult_Switcher()
    Widget=Get WBP_NoResults() -> (WBP_NoResults:WBP Discovery No Results Message Object Reference)
   )
   SetActiveWidget(
    execute=[@n4.RatingRestricted, @n4.RatingUnrated]
    self=Get SearchResult_Switcher()
    Widget=Get WBP_BlockedIslandResult() -> (WBP_BlockedIslandResult:WBP Discovery No Results Message Object Reference)
   )
   SetActiveWidget(execute=@n3.else, self=Get SearchResult_Switcher(), Widget=Get WBP_NoResults())
  ]}
 SetSearchResults {nodes [
   n1 = SetSearchResults!() -> (then, SearchResultsArray:Array of Discover Tile Item View Model Object References) {is=title K2Node_FunctionEntry}
   n2 = Get TileView_IslandsResults()
   > ClearListItems(self=@n2)
   > Branch(Condition=Is Not Empty!(TargetArray:Array of Discover Tile Item View Model Object References=@n1) -> (ReturnValue:bool))
   > Set List Items!(execute, self:List View Object Reference=@n2, InListItems:Array of Object References=@n1) -> (then) {is=title K2Node_CallFunction}
  ]}
 Switch Between Browse And Search Views {nodes [
   n1 = Switch Between Browse And Search Views!() -> (then, IsSearchActive:bool=~) {is=title K2Node_FunctionEntry}
   > SetActiveWidget(
    self=Get Switcher_CategoriesOrSearch() -> (Switcher_CategoriesOrSearch:Common Widget Switcher Legacy Object Reference)
    Widget=Select(Option 0:Widget Object Reference=Get ActivityBrowserBrowseView_VM(), Option 1:Widget Object Reference=Get SearchResult_Switcher(), Index=@n1) -> (ReturnValue:Widget Object Reference)
   )
   > n2 = Branch(Condition=@n1)
   > n3 = StartAnalyticsImpressions(self:Discover Search Model Object Reference=Get FortDiscoverSearchViewModel())
   n4 = GetInputAction!(self:Common Button Base Object Reference=Get WBP_BottomBarHack_BackToTop()) -> (InputActionRow:Data Table Row Handle Structure, ReturnValue:bool)
   SetBottomButtonVisibility(
    execute=[@n3.then, @n5.then]
    self=Get WBP_BottomBarHack_Back()
    ShouldButtonBeVisible=AND Boolean(A="Equal (Boolean)"(A:bool=@n4.ReturnValue, B:bool) -> (ReturnValue:bool=~) {is=title K2Node_PromotableOperator}, B=@n1)
   )
   n5 = StartAnalyticsImpressions(execute=@n2.else, self:Discover Surface View Model Object Reference=Get FortDiscoverBrowseViewModel())
  ]}
]
variables [
 IntroWhooshSound {type=SoundBase, default=/Game/Sounds/Fort_UI_Sounds/FrontEnd/Tabs/UI_Whoosh_TopBar_Tab_Search_Cue}
 IsSearchActive {type=boolean}
]
widgets {
 info {inheritedWidgetCount=0, parentClass=&CommonActivatableWidget, rootWidgetClass=&GridPanel, widgetCount=28}
 widgets [
  RootGrid_Search {widgetClassPath=&GridPanel, widget=RootGrid_Search, properties {columnFill [1], rowFill [1]}}
  NebulousBackground {
   widgetClassPath=&Image
   bIsVariable=true
   parent=RootGrid_Search
   widget=NebulousBackground
   properties {brush {
     bIsDynamicallyLoaded=false
     drawAs=Image
     imageType=NoImage
     mirroring=NoMirror
     resourceName=None
     resourceObject=/FortUILibrary/MaterialInstances/Background/Nebulous/MI_UI_Background_Nebulous
     tiling=NoTile
     imageSize=<32 32>
     margin=<ltrb 0 0 0 0>
     outlineSettings {
      bUseBrushTransparency=false
      roundingType=HalfHeightRadius
      width=0
      color {colorUseRule=UseColor_Specified, specifiedColor=<rgba 0 0 0 0>}
      cornerRadii=<0 0 0 0>
     }
     tintColor {colorUseRule=UseColor_Specified, specifiedColor=<rgba 1 1 1 1>}
     uVRegion {bIsValid=false, max=<0 0>, min=<0 0>}
    }}
   slot {class=GridSlot}
  }
  CaptureForPostBufferUpdate {
   widgetClassPath=&WBP_CaptureForPostBufferUpdate
   bIsVariable=true
   parent=RootGrid_Search
   widget=CaptureForPostBufferUpdate
   slot {class=GridSlot, properties {layer=1}}
  }
  SafeZone {
   widgetClassPath=&SafeZone
   bIsVariable=true
   parent=RootGrid_Search
   widget=SafeZone
   properties {padBottom=false, padLeft=false, padRight=false, padTop=false}
   slot {class=GridSlot, properties {layer=10, padding=<ltrb 0 160 0 0>}}
  }
  VerticalBox {widgetClassPath=&VerticalBox, parent=SafeZone, widget=VerticalBox, slot {class=SafeZoneSlot}}
  Wrapper_CustomNavEvents_UMGBug {
   widgetClassPath=&GridPanel
   parent=VerticalBox
   widget=Wrapper_CustomNavEvents_UMGBug
   properties {columnFill [1], rowFill [1]}
   slot {class=VerticalBoxSlot, properties {horizontalAlignment=HAlign_Center, verticalAlignment=VAlign_Top}}
  }
  FilteredInputField {
   widgetClassPath=/FortUILibrary/UIKit/InputField/WBP_UIKit_FilteredInputField
   bIsVariable=true
   parent=Wrapper_CustomNavEvents_UMGBug
   widget=FilteredInputField
   properties {
    bIsFocusable=true
    hideValidationInfo=true
    hintText=Search for Island Names or Codes...
    isSearchField=true
    navigation=&self:WidgetTree.FilteredInputField.WidgetNavigation_0
    useHintText=true
    clearButtonPadding=<ltrb 0 0 12 0>
    desiredImageSize=<56 56>
    height Override {enable=true, value=80}
    scaleOverride {enable=true, value=1.2}
    width Override {enable=true, value=800}
   }
   slot {class=GridSlot}
  }
  Switcher_CategoriesOrSearch {
   widgetClassPath=&CommonWidgetSwitcherLegacy
   bIsVariable=true
   parent=VerticalBox
   widget=Switcher_CategoriesOrSearch
   properties {navigation=&self:WidgetTree.Switcher_CategoriesOrSearch.WidgetNavigation_0}
   slot {class=VerticalBoxSlot, properties {padding=<ltrb 0 80 0 0>, size {sizeRule=Fill, value=1}}}
  }
  ActivityBrowserBrowseView_VM {
   widgetClassPath=/DiscoveryBrowser/ActivityBrowserBrowseView_VM
   bIsVariable=true
   parent=Switcher_CategoriesOrSearch
   widget=ActivityBrowserBrowseView_VM
   properties {
    bAutoRestoreFocus=false
    bSupportsActivationFocus=false
    clipping=ClipToBounds
    navigation=&self:WidgetTree.ActivityBrowserBrowseView_VM.WidgetNavigation_0
    desiredFocusWidget {widgetName=None}
   }
   slot {class=WidgetSwitcherSlot}
  }
  SearchResult_Switcher {
   widgetClassPath=&CommonWidgetSwitcherLegacy
   bIsVariable=true
   parent=Switcher_CategoriesOrSearch
   widget=SearchResult_Switcher
   properties {activeWidgetIndex=2, navigation=&self:WidgetTree.SearchResult_Switcher.WidgetNavigation_0}
   slot {class=WidgetSwitcherSlot}
  }
  SearchSpinner {
   widgetClassPath=&GridPanel
   bIsVariable=true
   parent=SearchResult_Switcher
   widget=SearchSpinner
   properties {columnFill [1], rowFill [1]}
   slot {class=WidgetSwitcherSlot, properties {padding=<ltrb 0 0 0 80>}}
  }
  WBP_UIKit_Throbber {
   widgetClassPath=/FortUILibrary/UIKit/Throbber/WBP_UIKit_Throbber
   bIsVariable=true
   parent=SearchSpinner
   widget=WBP_UIKit_Throbber
   properties {displayBackground=false, throbberSize=Large}
   slot {class=GridSlot, properties {horizontalAlignment=HAlign_Center, verticalAlignment=VAlign_Center}}
  }
  WBP_NoResults {
   widgetClassPath=/DiscoveryBrowser/WBP_Discovery_MessageNoResults.WBP_Discovery_NoResultsMessage
   bIsVariable=true
   parent=SearchResult_Switcher
   widget=WBP_NoResults
   properties {flavorMessageBase="Please search for a different word or phrase ", flavorTitleBase=No Results Found}
   slot {class=WidgetSwitcherSlot, properties {horizontalAlignment=HAlign_Center, verticalAlignment=VAlign_Center, padding=<ltrb 0 0 0 80>}}
  }
  WBP_BlockedIslandResult {
   widgetClassPath=&WBP_Discovery_NoResultsMessage
   bIsVariable=true
   parent=SearchResult_Switcher
   widget=WBP_BlockedIslandResult
   properties {flavorMessageBase=Please search for a different word or phrase, flavorTitleBase=This island is not available}
   slot {class=WidgetSwitcherSlot, properties {horizontalAlignment=HAlign_Center, verticalAlignment=VAlign_Center, padding=<ltrb 0 0 0 80>}}
  }
  ResultsGrid {
   widgetClassPath=&GridPanel
   bIsVariable=true
   parent=SearchResult_Switcher
   widget=ResultsGrid
   properties {columnFill [1], rowFill [0,1]}
   slot {class=WidgetSwitcherSlot}
  }
  VerticalBox_IslandBox {
   widgetClassPath=&VerticalBox
   parent=ResultsGrid
   widget=VerticalBox_IslandBox
   properties {clipping=ClipToBoundsWithoutIntersecting}
   slot {class=GridSlot, properties {layer=1, row=1, rowSpan=1}}
  }
  Overlay_0 {
   widgetClassPath=&Overlay
   parent=VerticalBox_IslandBox
   widget=Overlay_0
   slot {class=VerticalBoxSlot, properties {size {sizeRule=Fill, value=1}}}
  }
  TileView_IslandsResults {
   widgetClassPath=/Script/DiscoveryBrowserUI.FortActivityTileView
   bIsVariable=true
   parent=Overlay_0
   widget=TileView_IslandsResults
   properties {
    bEnableFixedLineOffset=true
    bEntrySizeIncludesEntrySpacing=false
    clipping=Inherit
    entryHeight=230
    entryWidgetClass=/DiscoveryBrowser/Tiles/Islands/WBP_IslandTile_Full.WBP_IslandTile_Full_C
    entryWidth=408
    horizontalEntrySpacing=24
    numDesignerPreviewEntries=12
    tileAlignment=LeftAligned
    verticalEntrySpacing=24
    scrollBarStyle {
     thickness=0
     draggedThumbImage {
      bIsDynamicallyLoaded=false
      drawAs=NoDrawType
      imageType=FullColor
      mirroring=NoMirror
      resourceName=../../../Engine/Content/Slate/Common/Scrollbar_Thumb.png
      resourceObject=null
      tiling=NoTile
      imageSize=<0 0>
      margin=<ltrb 0.25 0.25 0.25 0.25>
      outlineSettings {
       bUseBrushTransparency=false
       roundingType=HalfHeightRadius
       width=0
       color {colorUseRule=UseColor_Specified, specifiedColor=<rgba 0 0 0 0>}
       cornerRadii=<0 0 0 0>
      }
      tintColor {colorUseRule=UseColor_Specified, specifiedColor=<rgba 1 1 1 1>}
      uVRegion {bIsValid=false, max=<0 0>, min=<0 0>}
     }
     horizontalBackgroundImage {
      bIsDynamicallyLoaded=false
      drawAs=NoDrawType
      imageType=NoImage
      mirroring=NoMirror
      resourceName=None
      resourceObject=null
      tiling=NoTile
      imageSize=<0 0>
      margin=<ltrb 0 0 0 0>
      outlineSettings {
       bUseBrushTransparency=false
       roundingType=HalfHeightRadius
       width=0
       color {colorUseRule=UseColor_Specified, specifiedColor=<rgba 0 0 0 0>}
       cornerRadii=<0 0 0 0>
      }
      tintColor {colorUseRule=UseColor_Specified, specifiedColor=<rgba 1 1 1 1>}
      uVRegion {bIsValid=false, max=<0 0>, min=<0 0>}
     }
     horizontalBottomSlotImage {
      bIsDynamicallyLoaded=false
      drawAs=NoDrawType
      imageType=FullColor
      mirroring=NoMirror
      resourceName=../../../Engine/Content/Slate/Common/Scrollbar_Background_Horizontal.png
      resourceObject=null
      tiling=NoTile
      imageSize=<8 8>
      margin=<ltrb 0 0 0 0>
      outlineSettings {
       bUseBrushTransparency=false
       roundingType=HalfHeightRadius
       width=0
       color {colorUseRule=UseColor_Specified, specifiedColor=<rgba 0 0 0 0>}
       cornerRadii=<0 0 0 0>
      }
      tintColor {colorUseRule=UseColor_Specified, specifiedColor=<rgba 1 1 1 1>}
      uVRegion {bIsValid=false, max=<0 0>, min=<0 0>}
     }
     horizontalTopSlotImage {
      bIsDynamicallyLoaded=false
      drawAs=NoDrawType
      imageType=FullColor
      mirroring=NoMirror
      resourceName=../../../Engine/Content/Slate/Common/Scrollbar_Background_Horizontal.png
      resourceObject=null
      tiling=NoTile
      imageSize=<8 8>
      margin=<ltrb 0 0 0 0>
      outlineSettings {
       bUseBrushTransparency=false
       roundingType=HalfHeightRadius
       width=0
       color {colorUseRule=UseColor_Specified, specifiedColor=<rgba 0 0 0 0>}
       cornerRadii=<0 0 0 0>
      }
      tintColor {colorUseRule=UseColor_Specified, specifiedColor=<rgba 1 1 1 1>}
      uVRegion {bIsValid=false, max=<0 0>, min=<0 0>}
     }
     hoveredThumbImage {
      bIsDynamicallyLoaded=false
      drawAs=NoDrawType
      imageType=FullColor
      mirroring=NoMirror
      resourceName=../../../Engine/Content/Slate/Common/Scrollbar_Thumb.png
      resourceObject=null
      tiling=NoTile
      imageSize=<0 0>
      margin=<ltrb 0.25 0.25 0.25 0.25>
      outlineSettings {
       bUseBrushTransparency=false
       roundingType=HalfHeightRadius
       width=0
       color {colorUseRule=UseColor_Specified, specifiedColor=<rgba 0 0 0 0>}
       cornerRadii=<0 0 0 0>
      }
      tintColor {colorUseRule=UseColor_Specified, specifiedColor=<rgba 1 1 1 1>}
      uVRegion {bIsValid=false, max=<0 0>, min=<0 0>}
     }
     normalThumbImage {
      bIsDynamicallyLoaded=false
      drawAs=NoDrawType
      imageType=FullColor
      mirroring=NoMirror
      resourceName=../../../Engine/Content/Slate/Common/Scrollbar_Thumb.png
      resourceObject=null
      tiling=NoTile
      imageSize=<0 0>
      margin=<ltrb 0.25 0.25 0.25 0.25>
      outlineSettings {
       bUseBrushTransparency=false
       roundingType=HalfHeightRadius
       width=0
       color {colorUseRule=UseColor_Specified, specifiedColor=<rgba 0 0 0 0>}
       cornerRadii=<0 0 0 0>
      }
      tintColor {colorUseRule=UseColor_Specified, specifiedColor=<rgba 1 1 1 1>}
      uVRegion {bIsValid=false, max=<0 0>, min=<0 0>}
     }
     verticalBackgroundImage {
      bIsDynamicallyLoaded=false
      drawAs=NoDrawType
      imageType=NoImage
      mirroring=NoMirror
      resourceName=None
      resourceObject=null
      tiling=NoTile
      imageSize=<0 0>
      margin=<ltrb 0 0 0 0>
      outlineSettings {
       bUseBrushTransparency=false
       roundingType=HalfHeightRadius
       width=0
       color {colorUseRule=UseColor_Specified, specifiedColor=<rgba 0 0 0 0>}
       cornerRadii=<0 0 0 0>
      }
      tintColor {colorUseRule=UseColor_Specified, specifiedColor=<rgba 1 1 1 0>}
      uVRegion {bIsValid=false, max=<0 0>, min=<0 0>}
     }
     verticalBottomSlotImage {
      bIsDynamicallyLoaded=false
      drawAs=NoDrawType
      imageType=FullColor
      mirroring=NoMirror
      resourceName=../../../Engine/Content/Slate/Common/Scrollbar_Background_Vertical.png
      resourceObject=null
      tiling=NoTile
      imageSize=<8 8>
      margin=<ltrb 0 0 0 0>
      outlineSettings {
       bUseBrushTransparency=false
       roundingType=HalfHeightRadius
       width=0
       color {colorUseRule=UseColor_Specified, specifiedColor=<rgba 0 0 0 0>}
       cornerRadii=<0 0 0 0>
      }
      tintColor {colorUseRule=UseColor_Specified, specifiedColor=<rgba 1 1 1 1>}
      uVRegion {bIsValid=false, max=<0 0>, min=<0 0>}
     }
     verticalTopSlotImage {
      bIsDynamicallyLoaded=false
      drawAs=NoDrawType
      imageType=FullColor
      mirroring=NoMirror
      resourceName=../../../Engine/Content/Slate/Common/Scrollbar_Background_Vertical.png
      resourceObject=null
      tiling=NoTile
      imageSize=<8 8>
      margin=<ltrb 0 0 0 0>
      outlineSettings {
       bUseBrushTransparency=false
       roundingType=HalfHeightRadius
       width=0
       color {colorUseRule=UseColor_Specified, specifiedColor=<rgba 0 0 0 0>}
       cornerRadii=<0 0 0 0>
      }
      tintColor {colorUseRule=UseColor_Specified, specifiedColor=<rgba 1 1 1 1>}
      uVRegion {bIsValid=false, max=<0 0>, min=<0 0>}
     }
    }
   }
   slot {class=OverlaySlot, properties {horizontalAlignment=HAlign_Fill, verticalAlignment=VAlign_Fill, padding=<ltrb 108 0 80 0>}}
  }
  Image_SoftEdgeGradient_Islands {
   widgetClassPath=&Image
   bIsVariable=true
   parent=Overlay_0
   widget=Image_SoftEdgeGradient_Islands
   properties {
    renderOpacity=0
    visibility=SelfHitTestInvisible
    brush {
     bIsDynamicallyLoaded=false
     drawAs=Image
     imageType=NoImage
     mirroring=NoMirror
     resourceName=None
     resourceObject=/DiscoveryBrowser/TabScreens/DiscoverHome/Materials/MI_UI_Discovery_NebulousBackgroundWithBottomMask
     tiling=NoTile
     imageSize=<32 50>
     margin=<ltrb 0 0 0 0>
     outlineSettings {
      bUseBrushTransparency=false
      roundingType=HalfHeightRadius
      width=0
      color {colorUseRule=UseColor_Specified, specifiedColor=<rgba 0 0 0 0>}
      cornerRadii=<0 0 0 0>
     }
     tintColor {colorUseRule=UseColor_Specified, specifiedColor=<rgba 1 1 1 1>}
     uVRegion {bIsValid=false, max=<0 0>, min=<0 0>}
    }
   }
   slot {class=OverlaySlot, properties {horizontalAlignment=HAlign_Fill}}
  }
  Spacer_99 {widgetClassPath=&Spacer, parent=VerticalBox_IslandBox, widget=Spacer_99, properties {size=<1 50>}, slot {class=VerticalBoxSlot}}
  RichResultsText {
   widgetClassPath=&CommonRichTextBlock
   bIsVariable=true
   parent=ResultsGrid
   widget=RichResultsText
   properties {
    bOverrideDefaultStyle=true
    lineHeightPercentage=1.25
    mobileTextBlockScale=1.2
    text="Results for <text color=\"FFFFFF\" fontface=\"medium\">SearchTerm</> "
    wrapTextAt=640
    defaultTextStyleOverride {
     overflowPolicy=Clip
     transformPolicy=None
     colorAndOpacity {colorUseRule=UseColor_Specified, specifiedColor=<rgba 1 1 1 1>}
     font {
      bForceMonospaced=false
      bMaterialIsStencil=false
      fontMaterial=null
      fontObject=&HeadingNow
      letterSpacing=0
      monospacedWidth=1
      size=18
      skewAmount=0
      typefaceFontName=74 Regular
      outlineSettings {
       bApplyOutlineToDropShadows=false
       bMiteredCorners=false
       bSeparateFillAlpha=false
       outlineMaterial=null
       outlineSize=0
       outlineColor=<rgba 0 0 0 1>
      }
     }
     highlightColor {colorUseRule=UseColor_Specified, specifiedColor=<rgba 0 0 0 1>}
     highlightShape {
      bIsDynamicallyLoaded=false
      drawAs=Image
      imageType=NoImage
      mirroring=NoMirror
      resourceName=None
      resourceObject=null
      tiling=NoTile
      imageSize=<32 32>
      margin=<ltrb 0 0 0 0>
      outlineSettings {
       bUseBrushTransparency=false
       roundingType=HalfHeightRadius
       width=0
       color {colorUseRule=UseColor_Specified, specifiedColor=<rgba 0 0 0 0>}
       cornerRadii=<0 0 0 0>
      }
      tintColor {colorUseRule=UseColor_Specified, specifiedColor=<rgba 1 1 1 1>}
      uVRegion {bIsValid=false, max=<0 0>, min=<0 0>}
     }
     selectedBackgroundColor {colorUseRule=UseColor_Foreground, specifiedColor=<rgba 1 0 1 1>}
     shadowColorAndOpacity=<rgba 0 0 0 1>
     shadowOffset=<0 0>
     strikeBrush {
      bIsDynamicallyLoaded=false
      drawAs=NoDrawType
      imageType=NoImage
      mirroring=NoMirror
      resourceName=None
      resourceObject=null
      tiling=NoTile
      imageSize=<32 32>
      margin=<ltrb 0 0 0 0>
      outlineSettings {
       bUseBrushTransparency=false
       roundingType=HalfHeightRadius
       width=0
       color {colorUseRule=UseColor_Specified, specifiedColor=<rgba 0 0 0 0>}
       cornerRadii=<0 0 0 0>
      }
      tintColor {colorUseRule=UseColor_Specified, specifiedColor=<rgba 1 1 1 1>}
      uVRegion {bIsValid=false, max=<0 0>, min=<0 0>}
     }
     underlineBrush {
      bIsDynamicallyLoaded=false
      drawAs=Image
      imageType=NoImage
      mirroring=NoMirror
      resourceName=None
      resourceObject=null
      tiling=NoTile
      imageSize=<32 32>
      margin=<ltrb 0 0 0 0>
      outlineSettings {
       bUseBrushTransparency=false
       roundingType=HalfHeightRadius
       width=0
       color {colorUseRule=UseColor_Specified, specifiedColor=<rgba 0 0 0 0>}
       cornerRadii=<0 0 0 0>
      }
      tintColor {colorUseRule=UseColor_Specified, specifiedColor=<rgba 1 1 1 1>}
      uVRegion {bIsValid=false, max=<0 0>, min=<0 0>}
     }
    }
   }
   slot {class=GridSlot, properties {horizontalAlignment=HAlign_Left, layer=2, verticalAlignment=VAlign_Top, padding=<ltrb 120 0 0 0>}}
  }
  SZBottomButtons {
   widgetClassPath=&SafeZone
   parent=RootGrid_Search
   widget=SZBottomButtons
   properties {padTop=false}
   slot {class=GridSlot, properties {horizontalAlignment=HAlign_Right, layer=20, verticalAlignment=VAlign_Bottom, padding=<ltrb 0 0 80 32>}}
  }
  Overlay_1 {widgetClassPath=&Overlay, parent=SZBottomButtons, widget=Overlay_1, slot {class=SafeZoneSlot}}
  WBP_CaptureForPostBufferUpdate {
   widgetClassPath=&WBP_CaptureForPostBufferUpdate
   bIsVariable=true
   parent=Overlay_1
   slot {class=OverlaySlot, properties {horizontalAlignment=HAlign_Fill, verticalAlignment=VAlign_Fill}}
  }
  HBBottomButtons {
   widgetClassPath=&HorizontalBox
   parent=Overlay_1
   widget=HBBottomButtons
   slot {class=OverlaySlot, properties {horizontalAlignment=HAlign_Fill, verticalAlignment=VAlign_Fill}}
  }
  WBP_BottomBarHack_Back {
   widgetClassPath=/Game/UI/Frontend/BottomBar/WBP_BottomBarDecoyButton
   bIsVariable=true
   parent=HBBottomButtons
   widget=WBP_BottomBarHack_Back
   properties {minWidth=120, triggeringInputAction {dataTable=&InputActionDataTable, rowName=Back_SearchResult}}
   slot {class=HorizontalBoxSlot}
  }
  WBP_BottomBarHack_BackToTop {
   widgetClassPath=&WBP_BottomBarDecoyButton
   bIsVariable=true
   parent=HBBottomButtons
   widget=WBP_BottomBarHack_BackToTop
   properties {minWidth=120, triggeringInputAction {dataTable=&InputActionDataTable, rowName=BackToTop}}
   slot {class=HorizontalBoxSlot}
  }
  WBP_LoadingMorePages {
   widgetClassPath=/DiscoveryBrowser/TabScreens/WBP_LoadingMorePages
   bIsVariable=true
   parent=RootGrid_Search
   widget=WBP_LoadingMorePages
   properties {visibility=Collapsed}
   slot {class=GridSlot, properties {horizontalAlignment=HAlign_Center, layer=30, verticalAlignment=VAlign_Bottom, padding=<ltrb 8 8 8 8>}}
  }
 ]
}
