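# Activatable search screen for Discover that switches between category browsing and search results split into Islands and Creators tabs.
# The filtered input drives the search view model, whose results fill one poly tile view, resized by search mode, with spinner and no-result states.
# The browse view model feeds the embedded category view, which can open on a session's starting activity.
# It also handles intro animation, scroll fade, back-to-top, gamepad focus and impression analytics.
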WidgetBlueprint /DiscoveryBrowser/TabScreens/Search/DiscoverSearchView_Creator
animations [
 Intro {
  binding_count=2
  length_seconds=0.6
  object=&self:Intro
  sequence {
   bindings [
    Switcher_CategoriesOrSearch {
     id=#1
     possessed_class=CommonWidgetSwitcherLegacy
     tracks [
      {
       type=MovieSceneFloatTrack
       display_name=Render Opacity
       sections [{type=MovieSceneFloatSection, channels [None {type=MovieSceneScriptingFloatChannel, keys [[0,0],[6,0],[24,1]]}]}]
      }
      {
       type=MovieScene2DTransformTrack
       sections [{type=MovieScene2DTransformSection, channels [Translation.Y {type=MovieSceneScriptingFloatChannel, keys [[0,40],[6,40],[24,0]]}]}]
      }
     ]
    }
    FilteredInputField {
     id=#2
     possessed_class=WBP_UIKit_FilteredInputField_C
     tracks [
      {
       type=MovieSceneFloatTrack
       display_name=Render Opacity
       sections [{type=MovieSceneFloatSection, channels [None {type=MovieSceneScriptingFloatChannel, keys [[0,0],[18,1]]}]}]
      }
      {
       type=MovieScene2DTransformTrack
       sections [{type=MovieScene2DTransformSection, channels [Translation.Y {type=MovieSceneScriptingFloatChannel, keys [[0,40],[18,0]]}]}]
      }
     ]
    }
   ]
   display_rate {numerator=60}
  }
 }
 Anim_ListScrollFade {
  binding_count=1
  length_seconds=0.3
  object=&self:Anim_ListScrollFade
  sequence {bindings [Image_SoftEdgeGradient_Islands {
     id=#3
     possessed_class=Image
     tracks [{
       type=MovieSceneFloatTrack
       display_name=Render Opacity
       sections [{type=MovieSceneFloatSection, end_frame=6, start_frame=0, channels [None {type=MovieSceneScriptingFloatChannel, keys [[0,0],[6,1]]}]}]
      }]
    }]}
 }
]
graphs [
 BP_GetDesiredFocusTarget {nodes [
   Get Desired Focus Target()
   n1 = Get Intro()
   > Branch(Condition=NOT Boolean(A=IsAnimationPlaying(InAnimation=@n1)))
   n2 = Get InputField!(self:WBP UIKit Filtered Input Field Object Reference=Get FilteredInputField() -> (FilteredInputField:WBP UIKit Filtered Input Field Object Reference)) -> (InputField:WBP UIKit Input Field Object Reference)
   > Return Node!(
    execute
    ReturnValue:Widget Object Reference=Get EditableText!(self:WBP UIKit Block Editable Text Object Reference=Get Block_EditableText!(self:WBP UIKit Input Field Base Object Reference=@n2) -> (Block_EditableText:WBP UIKit Block Editable Text Object Reference)) -> (EditableText:UIKit Editable Text Object Reference)
   )
  ]}
 EventGraph {nodes [
   n1 = "On Clicked (WBP_BottomBarHack_Back)"!() -> (OutputDelegate, then, Button:Common Button Base Object Reference) {is=title K2Node_ComponentBoundEvent}
   ClearSearch(
    execute=[@n1.then, @n6.else]
    self:Discover Search Model Object Reference=Get FortDiscoverSearchViewModel() -> (FortDiscoverSearchViewModel:Discover Search Model Object Reference)
   )
   > SetInputFieldText!(execute, self:WBP UIKit Filtered Input Field Object Reference=Get FilteredInputField(), Text:Text) -> (then)
   > SetTextFocus(self:WBP UIKit Filtered Input Field Object Reference=Get FilteredInputField())
   n2 = "On Text Changed (FilteredInputField)"!() -> (OutputDelegate, then, Text:Text) {is=title K2Node_ComponentBoundEvent}
   > Set SearchTerm!(
    execute
    SearchTerm:String="To String (Text)"(InText=@n2.Text)
    self:Discover Search Model Object Reference=Get FortDiscoverSearchViewModel()
   ) -> (then, Output_Get:String)
   n3 = "On Text Committed (FilteredInputField)"!() -> (OutputDelegate, then, CommitMethod:ETextCommit Enum=Default) {is=title K2Node_ComponentBoundEvent}
   > n4 = Switch on ETextCommit!(execute, Selection:ETextCommit Enum=@n3.CommitMethod) -> (Default:Exec, OnEnter:Exec, OnUserMovedFocus:Exec, OnCleared:Exec) {is=title K2Node_SwitchEnum}
   n5 = GetCurrentText!(execute=@n4.OnEnter, self:WBP UIKit Filtered Input Field Object Reference=Get FilteredInputField()) -> (then, Text:Text)
   > n6 = Branch(Condition=NOT Boolean(A=TextIsEmpty(InText=@n5)))
   > ActionSearch(self:Discover Search Model Object Reference=Get FortDiscoverSearchViewModel())
   "On Tab Clicked (WBP_UIKit_Tabs)"!() -> (OutputDelegate, then, TabId:Name=None) {is=title K2Node_ComponentBoundEvent}
   > n7 = Switch on ECommonInputType!(execute, Selection:ECommonInputType Enum=GetCurrentInputType(self=Get CommonInputSubsystem())) -> (MouseAndKeyboard:Exec, Gamepad:Exec, Touch:Exec, Count:Exec) {is=title K2Node_SwitchEnum}
   Branch(execute=@n7.Gamepad, Condition=Get IsTileViewDataEmpty() -> (IsTileViewDataEmpty:bool))
   n8 = Get TileView!(self:Fort Polymorphic Tile View Object Reference=Get WBP_Results_PolyTileView() -> (WBP_Results_PolyTileView:WBP Results Poly Tile View Object Reference)) -> (TileView:Fort Tile View Object Reference)
   > NavigateToIndex!(execute, self:List View Object Reference=@n8, Index:int=0) -> (then)
   SetScrollOffset!(execute=[@n7.MouseAndKeyboard, @n7.Touch], self:List View Base Object Reference=@n8, InScrollOffset:float=0) -> (then)
   > Clear Selection(self:List View Object Reference=@n8)
   "On Clicked (WBP_BottomBarHack_BackToTop)"!() -> (OutputDelegate, then, Button:Common Button Base Object Reference) {is=title K2Node_ComponentBoundEvent}
   n9 = Get TileView(self=Get WBP_Results_PolyTileView())
   > SetSelectedIndex!(execute, self:List View Object Reference=@n9, Index:int=0) -> (then)
   > ScrollToTop(self:List View Base Object Reference=@n9)
   > n10 = Set Bottom Bar Button Visibility!(execute, self, IsVisible:bool=false) -> (then) {is=title K2Node_CallFunction}
   SetTextFocus(execute=[@n10.then, @n32.then], self=Get FilteredInputField())
   n11 = "On Tab Button Creation (WBP_UIKit_Tabs)"!() -> (OutputDelegate, then, TabId:Name=None, TabButton:Common Button Base Object Reference) {is=title K2Node_ComponentBoundEvent}
   > Populate Additional Tab Content!(execute, self, TabId:Name=@n11.TabId, TabButton:Common Button Base Object Reference=@n11.TabButton) -> (then)
   n12 = "On UIKit Tab Selected (WBP_UIKit_Tabs)"!() -> (OutputDelegate, then, TabID:Name=None, TabVM:MVVM Base Viewmodel Object Reference) {is=title K2Node_ComponentBoundEvent}
   > SendSearchTabSubViewEvent!(execute, self, TabID:Name=@n12.TabID) -> (then)
   n13 = GetTabIndexByID!(self:UIKit Tab Group Object Reference=Get WBP_UIKit_Tabs() -> (WBP_UIKit_Tabs:WBP UIKit Tabs Object Reference), TabNameID:Name=@n12.TabID) -> (ReturnValue:int=0)
   > n14 = Switch on Int!(execute:Exec, Selection:int=@n13) -> (0:Exec, 1:Exec) {bHasDefaultPin=false}
   > n15 = Set with Broadcast SearchMode!(execute, SearchMode:EDiscoverSearchMode Enum=Island) -> (then, Output_Get:EDiscoverSearchMode Enum=Island)
   SetRowHeaderText!(execute=[@n16.then, @n15.then], self, TabIndex:int=@n13) -> (then)
   > Set No Result Text!(execute, self, TabIndex:int=@n13) -> (then) {is=title K2Node_CallFunction}
   n16 = Set with Broadcast SearchMode(execute=@n14.1, SearchMode=Creator) -> (Output_Get=Island)
   n17 = OnTileViewScrolled!() -> (OutputDelegate, then, ItemOffset:float=~, DistanceRemaining:float=~) {is=title K2Node_CustomEvent}
   n18 = "float > float"(A=@n17.ItemOffset)
   > n19 = Branch(Condition=@n18)
   n20 = Get Anim_ListScrollFade() -> (Anim_ListScrollFade:Widget Animation Object Reference)
   > n21 = PlayAnimationForward(InAnimation=@n20)
   n22 = Get TileView(self=Get WBP_Results_PolyTileView())
   Set Bottom Bar Button Visibility(
    execute=[@n21.then, @n23.then]
    IsVisible=AND Boolean(A=@n18, B=IsFullsizedOnScreen!(Widget:Widget Object Reference=@n22) -> (IsFullsized:bool))
   )
   n23 = PlayAnimationReverse(execute=@n19.else, InAnimation=@n20)
   n24 = TopLevelTabChanged!() -> (OutputDelegate, then, LeavingFeature:EFortUIFeature Enum=ShowHome, EnteringFeature:EFortUIFeature Enum=ShowHome) {is=title K2Node_CustomEvent}
   > n25 = Branch(Condition="Equal (Enum)"(A:EFortUIFeature Enum=@n24.LeavingFeature, B:EFortUIFeature Enum=EnableSearchTabInTopNavigationBar))
   n26 = Get FortDiscoverSearchViewModel()
   > FireAnalyticsForBrowseAndSearch!(
    execute
    self:Discover Search Model Object Reference=@n26
    BrowseViewModel:Discover Browse View Model Object Reference=Get FortDiscoverBrowseViewModel() -> (FortDiscoverBrowseViewModel:Discover Browse View Model Object Reference)
   ) -> (then)
   Branch(execute=@n25.else, Condition="Equal (Enum)"(A=@n24.EnteringFeature, B=EnableSearchTabInTopNavigationBar))
   > StartAnalyticsSession(self:Discover Search Model Object Reference=@n26)
   > n27 = Branch(Condition=IsAnySearchActiveOrInProgress!(self:Discover Search Model Object Reference=Get FortDiscoverSearchViewModel()) -> (ReturnValue:bool))
   > StartAnalyticsImpressions(self:Discover Search Model Object Reference=@n26)
   n28 = Get WBP_Results_PolyTileView()
   > n29 = For Each Loop(Array:Array of User Widget Object References=GetDisplayedEntryWidgets!(self:List View Base Object Reference=Get TileView(self=@n28)) -> (ReturnValue:Array of User Widget Object References)) -> (Array Element:User Widget Object Reference)
   > n30 = Cast To FortActivityTileDetailsDisplay!(execute, Object=@n29) -> (then, CastFailed, AsFort Activity Tile Details Display:Fort Activity Tile Details Display Object Reference) {is=title K2Node_DynamicCast}
   > OnRespondToTileViewVisibilityChange!(execute, self:Fort Activity Tile Details Display Object Reference=@n30, bIsVisible:bool=true) -> (then)
   StartAnalyticsImpressions(execute=@n27.else, self:Discover Surface View Model Object Reference=Get FortDiscoverBrowseViewModel())
   OnBackToTopClickedCategory() {is=title K2Node_CustomEvent}
   n31 = Get TileView_Categories!(self:Activity Category Tile Panel Base VM Object Reference=Get TilePanel_All_VMversion!(self:Activity Browser Browse View VM Object Reference=Get ActivityBrowserBrowseView_VM() -> (ActivityBrowserBrowseView_VM:Activity Browser Browse View VM Object Reference)) -> (TilePanel_All_VMversion:Activity Category Tile Panel Base VM Object Reference)) -> (TileView_Categories:Fort Activity Tile View Object Reference)
   > SetSelectedIndex(self=@n31, Index=0)
   > n32 = ScrollToTop(self=@n31)
   n33 = OnIntroFinished() {is=title K2Node_CustomEvent}
   n34 = GetCurrentInputType(self=Get CommonInputSubsystem())
   > Branch(Condition=NOT Boolean(A="Equal (Enum)"(A:ECommonInputType Enum=@n34, B:ECommonInputType Enum=Touch)))
   > RequestRefreshFocus(self:Common Activatable Widget Object Reference)
   n35 = OnTileViewEntryGenerated!() -> (OutputDelegate, then, Widget:User Widget Object Reference) {is=title K2Node_CustomEvent}
   > n36 = Sequence()
   > n37 = Cast To WBP_IslandTile_Minimal!(execute, Object=@n35.Widget) -> (then, CastFailed, AsWBP Island Tile Minimal:WBP Island Tile Minimal Object Reference) {is=title K2Node_DynamicCast}
   > UpdateTileSizeAndShapeThenTileView!(
    execute
    self
    TileBase:WBP Discover Tile Base Object Reference=Get WBP_Discover_Tile_Base!(self:WBP Island Tile Minimal Object Reference=@n37) -> (WBP_Discover_Tile_Base:WBP Discover Tile Base Object Reference)
   ) -> (then)
   n38 = Cast To WBP_CreatorTile!(execute=@n37.CastFailed, Object=@n35.Widget) -> (then, CastFailed, AsWBP Creator Tile:WBP Creator Tile Object Reference) {is=title K2Node_DynamicCast}
   > UpdateTileSizeAndShapeThenTileView(TileBase=Get WBP_Discover_Tile_Base(self:WBP Discover Tile Button Object Reference=@n38))
   SetNestedButtonGroupSelectionVM!(
    execute:Exec=@n36.then_1
    self:Object Reference=@n35.Widget
    InVM:Fort Nested Button Group Selection VM Object Reference=Get ButtonGroupSelectionVM() -> (ButtonGroupSelectionVM:Fort Nested Button Group Selection VM Object Reference)
   ) -> (then:Exec) {is=title K2Node_Message}
   Event PreConstruct()
   > SetCompactUIStyle() {is=title K2Node_CallFunction}
   > Set Material Instance Parameters() {is=title K2Node_CallFunction}
   > SetResultsTerm!(execute, self, ResultsTerm:String=$Search Result) -> (then)
   > SetColorAndOpacity!(
    execute
    self=Get Text_HeaderRow!(self:WBP Row Header Object Reference=Get WBP_Row_Header() -> (WBP_Row_Header:WBP Row Header Object Reference)) -> (Text_HeaderRow:Common Text Object Reference)
    InColorAndOpacity_SpecifiedColor:Linear Color Structure=(R=1,G=1,B=1,A=1)
    InColorAndOpacity_ColorUseRule:ESlateColorStylingMode Enum
   ) -> (then)
   > PopulateTabs()
   > SetRowHeaderText(TabIndex=GetSelectedTabIndex!(self:UIKit Tab Group Object Reference=Get WBP_UIKit_Tabs()) -> (ReturnValue:int=0))
   > BindToAthenaFrontendTabChange!(execute, Context:Widget Object Reference=Self-Reference(), EventToBind:"Delegate (by ref)"=@n24) -> (then)
   Event Construct()
   n39 = Get FortDiscoverBrowseViewModel()
   > n40 = Branch(Condition=ShouldQueryOnConstruct!(self:Discover Surface View Model Object Reference=@n39) -> (ReturnValue:bool))
   > n41 = Set SurfaceName!(execute, SurfaceName:Name=CreativeDiscoverySurface_Browse, self:Discover Surface View Model Object Reference=@n39) -> (then, Output_Get:Name=None)
   SetBottomButtonVisibility!(
    execute=[@n41.then, @n40.else]
    self:WBP Bottom Bar Decoy Button Object Reference=Get WBP_BottomBarHack_BackToTop() -> (WBP_BottomBarHack_BackToTop:WBP Bottom Bar Decoy Button Object Reference)
    ShouldButtonBeVisible:bool=false
   ) -> (then)
   > SetBottomButtonVisibility(self=Get WBP_BottomBarHack_Back() -> (WBP_BottomBarHack_Back:WBP Bottom Bar Decoy Button Object Reference), ShouldButtonBeVisible=false)
   n42 = Get TileView(self=Get WBP_Results_PolyTileView())
   > Bind Event to On List View Scrolled(self:List View Object Reference=@n42, Delegate=Create Event()) {is=title K2Node_AddDelegate}
   > Bind Event to On Entry Generated(self:List View Base Object Reference=@n42, Delegate=Create Event()) {is=title K2Node_AddDelegate}
   n43 = Get ActivityBrowserBrowseView_VM()
   > Bind Event to On Back to Top Clicked(self:Activity Browser Browse View VM Object Reference=@n43, Delegate=Create Event()) {is=title K2Node_AddDelegate}
   > Set with Broadcast SearchMode(SearchMode=Island) -> (Output_Get=Island)
   > n44 = Create Widget!(
    execute:Exec
    Class:User Widget Class Reference=/DiscoveryBrowser/Tiles/WBP_Discover_Tile_Base.WBP_Discover_Tile_Base_C
    OwningPlayer:Player Controller Object Reference
   ) -> (then:Exec, ReturnValue:WBP Discover Tile Base Object Reference)
   > Set DiscoverTileBaseReference(DiscoverTileBaseReference:WBP Discover Tile Base Object Reference=@n44) -> (Output_Get:WBP Discover Tile Base Object Reference)
   Event Destruct()
   > Unbind all Events from On List View Scrolled!(execute:Exec, self:List View Object Reference=@n42) -> (then:Exec) {is=title K2Node_ClearDelegate}
   > Unbind all Events from On Entry Generated!(execute:Exec, self:List View Base Object Reference=@n42) -> (then:Exec) {is=title K2Node_ClearDelegate}
   > Unbind all Events from On Back to Top Clicked!(execute:Exec, self:Activity Browser Browse View VM Object Reference=@n43) -> (then:Exec) {is=title K2Node_ClearDelegate}
   Event OnInitialized()
   > n45 = Construct Object from Class(Class=/Script/FortniteUI.FortNestedButtonGroupSelectionVM) -> (ReturnValue:Fort Nested Button Group Selection VM Object Reference)
   > Set ButtonGroupSelectionVM(ButtonGroupSelectionVM:Fort Nested Button Group Selection VM Object Reference=@n45) -> (Output_Get:Fort Nested Button Group Selection VM Object Reference)
   Event On Activated()
   n46 = Get FortDiscoverBrowseViewModel()
   > n47 = Branch(Condition=ShouldQueryOnConstruct(self=@n46))
   Fire Event Main View Opened!(
    execute=[@n54.then, @n47.then, @n53.else]
    View:Widget Object Reference=Self-Reference()
    AdditionalInfo:Flow Analytics Additional Info Structure
    bIsOverlay:bool=false
   ) -> (then)
   > SendBrowseSubViewEvent()
   > BindToAnimationFinished(Animation=Get Intro(), Delegate=@n33)
   > PlayAnimationForward(InAnimation=Get Intro())
   n48 = Get ActivityBrowserBrowseView_VM()
   > ActivateWidget(self=@n48)
   > SpawnSound2D!(
    execute
    Sound:Sound Base Object Reference=Get IntroWhooshSound() -> (IntroWhooshSound:Sound Base Object Reference)
    VolumeMultiplier:float=1
    PitchMultiplier:float=1
    StartTime:float=0
    ConcurrencySettings:Sound Concurrency Object Reference
    bPersistAcrossLevelTransition:bool=false
    bAutoDestroy:bool=true
   ) -> (then, ReturnValue:Audio Component Object Reference)
   n49 = IsAnySearchActiveOrInProgress(self=Get FortDiscoverSearchViewModel())
   > n50 = Branch(Condition=@n49)
   > n51 = StartAnalyticsVisibilityImpressions(self:Discover Search Model Object Reference=Get FortDiscoverSearchViewModel())
   SelectStartingActivity(execute=[@n52.then, @n51.then])
   n52 = StartVisibleImpressions(execute=@n50.else, self:Discover Surface View Model Object Reference=Get FortDiscoverBrowseViewModel())
   n53 = Branch(
    execute=@n47.else
    Condition="Equal (Name)"(A=Get SurfaceName!(self:Discover Surface View Model Object Reference=@n46) -> (SurfaceName:Name=None))
   )
   > n54 = Set SurfaceName(SurfaceName=CreativeDiscoverySurface_Browse, self=@n46) -> (Output_Get=None)
   Event On Deactivated()
   > UnbindFromAnimationFinished!(execute, self:User Widget Object Reference, Animation:Widget Animation Object Reference=Get Intro(), Delegate:Delegate=@n33) -> (then)
   > Check and Stop Animation!(execute:Exec, Animation:Widget Animation Object Reference=Get Intro()) -> (then:Exec) {is=title K2Node_MacroInstance}
   > DeactivateWidget(self=@n48)
   > n55 = Branch(Condition=@n49)
   > StopAnalyticsVisibilityImpressions(self:Discover Search Model Object Reference=Get FortDiscoverSearchViewModel())
   StopVisibleImpressions(execute=@n55.else, self:Discover Surface View Model Object Reference=Get FortDiscoverBrowseViewModel())
  ]}
 Fill Out TileView Data {nodes [
   Fill Out TileView Data() {is=title K2Node_FunctionEntry}
   n1 = Get TileView(self=Get WBP_Results_PolyTileView())
   > ClearListItems(self=@n1)
   n2 = Get SearchResults!(self:Discover Search Model Object Reference=Get FortDiscoverSearchViewModel()) -> (SearchResults:Array of Discover Tile Item View Model Object References)
   > n3 = Branch(Condition=Is Not Empty!(TargetArray:Array of Discover Tile Item View Model Object References=@n2) -> (ReturnValue:bool))
   > UpdateTileSizeAndShapeThenTileView(TileBase=Get DiscoverTileBaseReference() -> (DiscoverTileBaseReference:WBP Discover Tile Base Object Reference))
   > Set List Items!(execute, self:List View Object Reference=@n1, InListItems:Array of Object References=@n2) -> (then) {is=title K2Node_CallFunction}
   > Set IsTileViewDataEmpty(IsTileViewDataEmpty:bool=false) -> (Output_Get:bool)
   Set IsTileViewDataEmpty(execute=@n3.else, IsTileViewDataEmpty=true)
  ]}
 IsFullsizedOnScreen {nodes [
   n1 = IsFullsizedOnScreen!() -> (then, Widget:Widget Object Reference) {is=title K2Node_FunctionEntry}
   n2 = GetPaintSpaceGeometry!(self:Widget Object Reference=@n1) -> (ReturnValue:"Geometry Structure (by ref)")
   > Return Node!(
    execute
    IsFullsized:bool="float < float"(
     A=GetLocalSize!(Geometry:"Geometry Structure (by ref)"=@n2) -> (ReturnValue_X:double, ReturnValue_Y:double).ReturnValue_Y
     B=GetDesiredSize!(self:Widget Object Reference=@n1) -> (ReturnValue_X:double, ReturnValue_Y:double).ReturnValue_Y
    )
   )
  ]}
 Navigate Down To Results {nodes [
   Navigate Down To Results!() -> (then, Navigation:EUINavigation Enum) {is=title K2Node_FunctionEntry}
   > Branch(Condition=NOT Boolean(A=Get IsTileViewDataEmpty()))
   n1 = Get TileView(self=Get WBP_Results_PolyTileView())
   > SetSelectedIndex(self=@n1, Index=0)
   > Return Node!(execute, ReturnValue:Widget Object Reference=@n1)
  ]}
 Navigate Up from Results {nodes [
   Navigate Up from Results!() -> (then, Navigation:EUINavigation Enum) {is=title K2Node_FunctionEntry}
   n1 = Get WBP_UIKit_Tabs()
   > Return Node(ReturnValue=GetTabButtonBaseByID!(
     self:Common Tab List Widget Base Object Reference=@n1
     TabNameID:Name=GetActiveTab!(self:Common Tab List Widget Base Object Reference=@n1) -> (ReturnValue:Name=None)
    ) -> (ReturnValue:Common Button Base Object Reference))
  ]}
 NavigateDownToBrowseOrSearch {nodes [
   NavigateDownToBrowseOrSearch!() -> (then, Navigation:EUINavigation Enum) {is=title K2Node_FunctionEntry}
   n1 = IsAnySearchActiveOrInProgress(self=Get FortDiscoverSearchViewModel())
   > n2 = Branch(Condition=@n1)
   n3 = Get WBP_UIKit_Tabs()
   n4 = GetActiveTab(self=@n3) -> (ReturnValue=None)
   n5 = Get TileView_Categories(self=Get TilePanel_Featured_VMversion!(self:Activity Browser Browse View VM Object Reference=Get ActivityBrowserBrowseView_VM()) -> (TilePanel_Featured_VMversion:Activity Category Tile Panel Base VM Object Reference))
   Return Node(
    execute=[@n2.then, @n6.then]
    ReturnValue=Select(Option 0:Widget Object Reference=@n5, Option 1:Widget Object Reference=GetTabButtonBaseByID(self=@n3, TabNameID=@n4), Index=@n1) -> (ReturnValue:Widget Object Reference)
   )
   n6 = SetSelectedIndex(execute=@n2.else, self=@n5, Index=0)
  ]}
 NavigateUpToTextEntry {nodes [
   NavigateUpToTextEntry!() -> (then, Navigation:EUINavigation Enum) {is=title K2Node_FunctionEntry}
   > Return Node(ReturnValue=Get InputField(self=Get FilteredInputField()))
  ]}
 On Update Providers {nodes [
   On Update Providers!() -> (then, ProvidersVM:Array of Discover Provider View Model Object References) {is=title K2Node_FunctionEntry}
   > SetDataFromProviders(self:Discover Browse View Model Object Reference=Get FortDiscoverBrowseViewModel())
   > Set Browse View VM() {is=title K2Node_CallFunction}
  ]}
 OnQueryingMorePages {nodes [
   n1 = OnQueryingMorePages!() -> (then, LoadingState:EDiscoverLoadingState Enum=Unset) {is=title K2Node_FunctionEntry}
   > SetVisibility(
    self=Get WBP_LoadingMorePages() -> (WBP_LoadingMorePages:WBP Loading More Pages Object Reference)
    InVisibility=Select(
     Option 0:ESlateVisibility Enum=Collapsed
     Option 1:ESlateVisibility Enum=HitTestInvisible
     Index="Equal (Enum)"(A:EDiscoverLoadingState Enum=@n1, B:EDiscoverLoadingState Enum=Loading)
    ) -> (ReturnValue:ESlateVisibility Enum)
   )
  ]}
 Populate Additional Tab Content {nodes [
   Populate Additional Tab Content!() -> (then, TabId:Name, TabButton:Common Button Base Object Reference) {is=title K2Node_FunctionEntry}
   > n1 = Cast To WBP_UIKit_Button_Selectable!(execute, Object=Get TabButton() -> (TabButton:Common Button Base Object Reference)) -> (then, CastFailed, AsWBP UIKit Button Selectable:WBP UIKit Button Selectable Object Reference) {is=title K2Node_DynamicCast}
   > Set UIKitButtonSelectableRef(UIKitButtonSelectableRef:WBP UIKit Button Selectable Object Reference=@n1) -> (Output_Get:WBP UIKit Button Selectable Object Reference)
   > n2 = Construct Object from Class(Class=/FortUILibrary/Systems/ModularButton/Blocks/WBP_UIKit_Block_Text.WBP_UIKit_Block_Text_C) -> (ReturnValue:WBP UIKit Block Text Object Reference)
   > n3 = Set UIKitBlockTextRef(UIKitBlockTextRef:WBP UIKit Block Text Object Reference=@n2) -> (Output_Get:WBP UIKit Block Text Object Reference)
   > Add(
    TargetArray:Array of WBP UIKit Block Text Object References=Get TabBlockTextArray() -> (TabBlockTextArray:Array of WBP UIKit Block Text Object References)
    NewItem:"WBP UIKit Block Text Object Reference (by ref)"=@n3
   ) {is=title K2Node_CallArrayFunction}
   > n4 = AddChildToGrid!(
    execute
    self:Grid Panel Object Reference=Get Main!(self:WBP UIKit Button Generic Base Object Reference=Get UIKitButtonSelectableRef() -> (UIKitButtonSelectableRef:WBP UIKit Button Selectable Object Reference)) -> (Main:Grid Panel Object Reference)
    Content:Widget Object Reference=@n3
    InRow:int=1
    InColumn:int=2
   ) -> (then, ReturnValue:Grid Slot Object Reference)
   > SetHorizontalAlignment(self:Grid Slot Object Reference=@n4, InHorizontalAlignment=HAlign_Left)
   > SetVerticalAlignment(self:Grid Slot Object Reference=@n4, InVerticalAlignment=VAlign_Fill)
   > SetFontMaterial!(
    execute
    self:WBP UIKit Block Text Object Reference=Get UIKitBlockTextRef() -> (UIKitBlockTextRef:WBP UIKit Block Text Object Reference)
    In Material:Material Instance Object Reference=/FortUILibrary/UIKit/Buttons/TabButtons/_MaterialInstances/MI_UI_TopNav_Text
   ) -> (then)
   > RegisterModularBlock!(
    execute
    self:UIKit Modular Button Object Reference=Get UIKitButtonSelectableRef()
    Block:UIKit Block Object Reference=Get UIKitBlockTextRef()
   ) -> (then)
  ]}
 PopulateTabs {nodes [
   PopulateTabs() {is=title K2Node_FunctionEntry}
   > n1 = Construct Object from Class(Class=/Script/FortniteUI.UIKitTabGroupViewModel) -> (ReturnValue:UI Kit Tab Group View Model Object Reference)
   > n2 = Sequence!(execute) -> (then_0, then_1, then_2:Exec)
   > n3 = Construct Object from Class(Class=/Script/FortniteUI.UIKitTabGroupButtonViewModel) -> (ReturnValue:UI Kit Tab Group Button View Model Object Reference)
   > n4 = Set Label!(execute, Label:Text="LOCTEXT(\"Islands\")", self:UI Kit Selectable Button View Model Object Reference=@n3) -> (then, Output_Get:Text)
   > Set TabID!(
    execute
    TabID:Name=String To Name(InString=Get Source String from Text!(Text:Text=@n4) -> (ReturnValue:String) {is=title K2Node_CallFunction})
    self:UI Kit Tab Group Button View Model Object Reference=@n3
   ) -> (then, Output_Get:Name=None)
   n5 = Get Tabs!(self:UI Kit Tab Group View Model Object Reference=@n1) -> (Tabs:Array of UI Kit Tab Group Button View Model Object References)
   > Add(
    TargetArray:Array of UI Kit Tab Group Button View Model Object References=@n5
    NewItem:"UI Kit Tab Group Button View Model Object Reference (by ref)"=@n3
   )
   n6 = Construct Object from Class(execute=@n2.then_1, Class=&UIKitTabGroupButtonViewModel)
   > n7 = Set Label(Label="LOCTEXT(\"Creators\")", self=@n6)
   > Set TabID(TabID=String To Name(InString=Get Source String from Text(Text=@n7)), self=@n6) -> (Output_Get=None)
   > Add(TargetArray=@n5, NewItem=@n6)
   n8 = Get WBP_UIKit_Tabs()
   Set Tabs Data!(
    execute=@n2.then_2
    self:WBP UIKit Tabs Base Object Reference=@n8
    InTabs Data:Array of UI Kit Tab Group Button View Model Object References=@n5
   ) -> (then) {is=title K2Node_CallFunction}
   > SelectTabByIndex!(execute, self:UIKit Tab Group Object Reference=@n8, Index:int=0, bSuppressClickFeedback:bool=true) -> (then, ReturnValue:bool)
  ]}
 SelectStartingActivity {nodes [
   SelectStartingActivity() {is=title K2Node_FunctionEntry}
   > Branch(Condition=Get ReadyForStartingActivity() -> (ReadyForStartingActivity:bool))
   n1 = Get FortPlayerSessionContext() -> (ReturnValue:Fort Player Session Context Object Reference) {is=title K2Node_GetSubsystem}
   > Branch(Condition=HasValidStartingActivity!(self:Fort Player Session Context Object Reference=@n1) -> (ReturnValue:bool))
   > ActivateFeaturedCategory!(
    execute
    self:Activity Browser Browse View VM Object Reference=Get ActivityBrowserBrowseView_VM()
    DesiredMnemonic:String=GetBrowserStartingMnemonic!(self:Fort Player Session Context Object Reference=@n1) -> (ReturnValue:String)
   ) -> (then)
   > ClearBrowserStartingActivity(self:Fort Player Session Context Object Reference=@n1)
  ]}
 SendBrowseSubViewEvent {nodes [
   SendBrowseSubViewEvent() {is=title K2Node_FunctionEntry}
   > Branch(Condition=IsActivated!(self:Common Activatable Widget Object Reference) -> (ReturnValue:bool))
   > n1 = Cast To FortPlayerController(Object=GetOwningPlayer())
   > "Fire Event Sub View Opened (Non-Localized)"!(
    execute
    FortPC:Fort Player Controller Object Reference=@n1
    ViewName:String=Browse
    AdditionalInfo:Flow Analytics Additional Info Structure
    bIsOverlay:bool=false
   ) -> (then)
  ]}
 SendSearchTabSubViewEvent {nodes [
   n1 = SendSearchTabSubViewEvent!() -> (then, TabID:Name) {is=title K2Node_FunctionEntry}
   > Branch(Condition=IsActivated())
   > n2 = Cast To FortPlayerController(Object=GetOwningPlayer(self=Get WBP_UIKit_Tabs()))
   > "Fire Event Sub View Opened (Non-Localized)"(FortPC=@n2, ViewName="To String (Name)"(InName=@n1), bIsOverlay=false)
  ]}
 Set Bottom Bar Button Visibility {nodes [
   n1 = Set Bottom Bar Button Visibility!() -> (then, IsVisible:bool=~) {is=title K2Node_FunctionEntry}
   > SetBottomButtonVisibility(self=Get WBP_BottomBarHack_BackToTop(), ShouldButtonBeVisible=@n1)
   > SetBottomButtonVisibility(self=Get WBP_BottomBarHack_Back(), ShouldButtonBeVisible=NOT Boolean(A=@n1))
  ]}
 Set Browse View VM {nodes [
   Set Browse View VM() {is=title K2Node_FunctionEntry}
   n1 = Get FortDiscoverBrowseViewModel()
   > Set ViewModel!(
    execute
    self:Activity Browser Browse View VM Object Reference=Get ActivityBrowserBrowseView_VM()
    In Discover Browse VM:Discover Browse View Model Object Reference=@n1
   ) -> (then) {is=title K2Node_CallFunction}
   > Set ReadyForStartingActivity(ReadyForStartingActivity:bool="Not Equal (Name)"(A=Get SurfaceName(self=@n1) -> (SurfaceName=None)) -> (ReturnValue=~)) -> (Output_Get:bool)
   > SelectStartingActivity() {is=title K2Node_CallFunction}
  ]}
 Set Creator Result Count {nodes [
   n1 = Set Creator Result Count!() -> (then, Creator Result Count:int) {is=title K2Node_FunctionEntry}
   > SetText!(
    execute
    self:WBP UIKit Block Text Object Reference="Get (a copy)"(Array:Array of WBP UIKit Block Text Object References=Get TabBlockTextArray(), Dimension 1=1) -> (Output:WBP UIKit Block Text Object Reference)
    Text:Text=Format Text!(Format="LOCTEXT(\"({CreatorsCount})\")", CreatorsCount:int=@n1) -> (Result)
   ) -> (then)
  ]}
 Set Island Result Count {nodes [
   n1 = Set Island Result Count!() -> (then, Island Result Count:int) {is=title K2Node_FunctionEntry}
   > SetText(
    self="Get (a copy)"(Array=Get TabBlockTextArray())
    Text=Format Text!(Format="LOCTEXT(\"({IslandsCount})\")", IslandsCount:int=@n1) -> (Result)
   )
  ]}
 Set Material Instance Parameters {nodes [
   Set Material Instance Parameters() {is=title K2Node_FunctionEntry}
   > n1 = GetDynamicMaterial(self=Get Image_SoftEdgeGradient_Islands() -> (Image_SoftEdgeGradient_Islands:Image Object Reference))
   > SetScalarParameterValue(self=@n1, ParameterName=Overlay Rotation, Value=0.5)
  ]}
 Set No Result Text {nodes [
   n1 = Set No Result Text!() -> (then, TabIndex:int) {is=title K2Node_FunctionEntry}
   n2 = Get FortDiscoverSearchViewModel()
   > Branch(Condition=IsValid(Object=@n2))
   > n3 = Switch on Int(Selection=@n1) {bHasDefaultPin=false}
   n4 = Get WBP_NoResults_SearchResults() -> (WBP_NoResults_SearchResults:WBP Discovery No Results Message Object Reference)
   n5 = Get VerticalBox_IslandBox() -> (VerticalBox_IslandBox:Vertical Box Object Reference)
   n6 = "Equal (Integer)"(A=GetIslandResultCount!(self:Discover Search Model Object Reference=@n2) -> (ReturnValue:int=0))
   > SetActiveWidget(
    self:Widget Switcher Object Reference=Get CommonWidgetSwitcher_Results() -> (CommonWidgetSwitcher_Results:Common Widget Switcher Legacy Object Reference)
    Widget=Select(Option 0:Widget Object Reference=@n5, Option 1:Widget Object Reference=@n4, Index=@n6) -> (ReturnValue:Widget Object Reference)
   )
   n7 = Get WBP_Row_Header()
   > SetVisibility(
    self=@n7
    InVisibility=Select(Option 0:ESlateVisibility Enum=SelfHitTestInvisible, Option 1:ESlateVisibility Enum=Hidden, Index=@n6) -> (ReturnValue:ESlateVisibility Enum)
   )
   n8 = "Equal (Integer)"(A=GetCreatorResultCount!(self:Discover Search Model Object Reference=@n2) -> (ReturnValue:int=0))
   SetActiveWidget(
    execute=@n3.1
    self=Get CommonWidgetSwitcher_Results()
    Widget=Select(Option 0:Widget Object Reference=@n5, Option 1:Widget Object Reference=@n4, Index=@n8) -> (ReturnValue:Widget Object Reference)
   )
   > SetVisibility(
    self=@n7
    InVisibility=Select(Option 0:ESlateVisibility Enum=SelfHitTestInvisible, Option 1:ESlateVisibility Enum=Hidden, Index=@n8) -> (ReturnValue:ESlateVisibility Enum)
   )
  ]}
 SetCompactUIStyle {nodes [
   SetCompactUIStyle() {is=title K2Node_FunctionEntry}
   > Branch(Condition=IsMobileGame())
   > SetSidesToPad!(
    execute
    self:Safe Zone Object Reference=Get SafeZone() -> (SafeZone:Safe Zone Object Reference)
    InPadLeft:bool=false
    InPadRight:bool=false
    InPadTop:bool=true
    InPadBottom:bool=false
   ) -> (then)
  ]}
 SetCreatorSearchInProgress {nodes [
   n1 = SetCreatorSearchInProgress!() -> (then, SearchInProgress:bool=~) {is=title K2Node_FunctionEntry}
   > SetSearchInProgress!(execute, self, SearchInProgress:bool=@n1) -> (then)
  ]}
 SetIslandSearchInProgress {nodes [
   n1 = SetIslandSearchInProgress!() -> (then, SearchInProgress:bool=~) {is=title K2Node_FunctionEntry}
   > SetSearchInProgress(SearchInProgress=@n1)
  ]}
 SetResultsTerm {nodes [
   n1 = SetResultsTerm!() -> (then, ResultsTerm:String) {is=title K2Node_FunctionEntry}
   > SetSubHeaderText!(
    execute
    self:WBP Row Header Object Reference=Get WBP_Row_Header()
    SubHeaderText:Text=Format Text!(Format="LOCTEXT(\"Matching \\\"{ResultsTerm}\\\"\")", ResultsTerm:String=@n1) -> (Result)
   ) -> (then)
  ]}
 SetRowHeaderText {nodes [
   n1 = SetRowHeaderText!() -> (then, TabIndex:int) {is=title K2Node_FunctionEntry}
   n2 = "Get (a copy)"(
    Array:Array of UI Kit Tab Group Button View Model Object References=Get LocalTabsViewModels!(self:UIKit Tab Group Object Reference=Get WBP_UIKit_Tabs()) -> (LocalTabsViewModels:Array of UI Kit Tab Group Button View Model Object References)
    Dimension 1=@n1
   ) -> (Output:UI Kit Tab Group Button View Model Object Reference)
   > SetHeaderText!(
    execute
    self:WBP Row Header Object Reference=Get WBP_Row_Header()
    HeaderText:Text=Format Text!(Format="LOCTEXT(\"{Label} Results\")", Label:Text=Get Label!(self:UI Kit Selectable Button View Model Object Reference=@n2) -> (Label:Text)) -> (Result)
   ) -> (then)
  ]}
 SetSearchInProgress {nodes [
   n1 = SetSearchInProgress!() -> (then, SearchInProgress:bool=~) {is=title K2Node_FunctionEntry}
   > n2 = Branch(Condition=@n1)
   > SetActiveWidget(
    self=Get SearchResult_Switcher() -> (SearchResult_Switcher:Common Widget Switcher Legacy Object Reference)
    Widget=Get SearchSpinner() -> (SearchSpinner:Grid Panel Object Reference)
   )
   n3 = Switch on ELinkEntryQueryResult!(
    execute=@n2.else
    Selection:ELinkEntryQueryResult Enum=Get LastLinkEntryQueryResult!(self:Discover Search Model Object Reference=Get FortDiscoverSearchViewModel()) -> (LastLinkEntryQueryResult:ELinkEntryQueryResult Enum=Success)
   ) -> (
    Success:Exec
    PartialSuccess:Exec
    FailedQuery:Exec
    FailedPlayerDataQuery:Exec
    FailedParse:Exec
    NotAllowedOnPlatform:Exec
    CabinModeBlocked:Exec
    UnableToFind:Exec
    NotVisible:Exec
    TournamentDataUnavailable:Exec
    FailedStringSanitization:Exec
    FailedSubModePopulation:Exec
    FailedSubModeSurfacePopulation:Exec
    FailedSubModeParsing:Exec
    FilteredOut:Exec
    RatingUnrated:Exec
    RatingRestricted:Exec
    ModeratorOnly:Exec
    FailedPlayerOrientedDataHydration:Exec
    InvalidQuery:Exec
    FailedReferenceLinkPopulation:Exec
   ) {is=title K2Node_SwitchEnum}
   SetActiveWidget(
    execute=[
     @n3.PartialSuccess
     @n3.Success
     @n3.CabinModeBlocked
     @n3.FailedParse
     @n3.FailedQuery
     @n3.FailedStringSanitization
     @n3.FailedSubModeParsing
     @n3.FailedSubModePopulation
     @n3.FailedSubModeSurfacePopulation
     @n3.FilteredOut
     @n3.NotAllowedOnPlatform
     @n3.NotVisible
     @n3.UnableToFind
     @n3.RatingRestricted
     @n3.RatingUnrated
    ]
    self=Get SearchResult_Switcher()
    Widget=Get ResultsGrid() -> (ResultsGrid:Grid Panel Object Reference)
   )
  ]}
 SetSearchResults {nodes [
   n1 = SetSearchResults!() -> (then, SearchResultsArray:Array of Discover Tile Item View Model Object References) {is=title K2Node_FunctionEntry}
   > Fill Out TileView Data() {is=title K2Node_CallFunction}
   > Set No Result Text(TabIndex=GetSelectedTabIndex(self=Get WBP_UIKit_Tabs()) -> (ReturnValue=0)) {is=title K2Node_CallFunction}
   Reroute Node(InputPin:Array of Discover Tile Item View Model Object References=@n1) -> (OutputPin:Array of Discover Tile Item View Model Object References)
  ]}
 Switch Between Browse And Search Views {nodes [
   n1 = Switch Between Browse And Search Views!() -> (then, IsSearchActive:bool=~) {is=title K2Node_FunctionEntry}
   > SetActiveWidget(
    self=Get Switcher_CategoriesOrSearch() -> (Switcher_CategoriesOrSearch:Common Widget Switcher Legacy Object Reference)
    Widget=Select(Option 0:Widget Object Reference=Get ActivityBrowserBrowseView_VM(), Option 1:Widget Object Reference=Get SearchResult_Switcher(), Index=@n1) -> (ReturnValue:Widget Object Reference)
   )
   > n2 = Branch(Condition=@n1)
   > StartAnalyticsImpressions(self:Discover Search Model Object Reference=Get FortDiscoverSearchViewModel())
   > n3 = SendSearchTabSubViewEvent(TabID=GetSelectedTabId!(self:Common Tab List Widget Base Object Reference=Get WBP_UIKit_Tabs()) -> (ReturnValue:Name=None)) {is=title K2Node_CallFunction}
   n4 = GetInputAction!(self:Common Button Base Object Reference=Get WBP_BottomBarHack_BackToTop()) -> (InputActionRow:Data Table Row Handle Structure, ReturnValue:bool)
   SetBottomButtonVisibility(
    execute=[@n5.then, @n3.then]
    self=Get WBP_BottomBarHack_Back()
    ShouldButtonBeVisible=AND Boolean(A="Equal (Boolean)"(A:bool=@n4.ReturnValue, B:bool) -> (ReturnValue:bool=~) {is=title K2Node_PromotableOperator}, B=@n1)
   )
   StartAnalyticsImpressions(execute=@n2.else, self:Discover Surface View Model Object Reference=Get FortDiscoverBrowseViewModel())
   > n5 = SendBrowseSubViewEvent() {is=title K2Node_CallFunction}
  ]}
 UpdateTileSizeAndShapeThenTileView {nodes [
   n1 = UpdateTileSizeAndShapeThenTileView!() -> (then, TileBase:WBP Discover Tile Base Object Reference) {is=title K2Node_FunctionEntry}
   n2 = "Equal (Enum)"(A:EDiscoverSearchMode Enum=Get SearchMode() -> (SearchMode:EDiscoverSearchMode Enum=Island), B:EDiscoverSearchMode Enum=Creator)
   > Set Tile Size!(
    execute
    self:WBP Discover Tile Base Object Reference=@n1
    InSize:Vector 2D Structure=ColumnSizeToTileSize!(self:WBP Discover Tile Base Object Reference=@n1, NewColumnSize:int=SelectInt(A=3, B=5, bPickA=@n2), IsSquareTiles:bool=@n2) -> (NewTileSize:Vector 2D Structure)
   ) -> (then) {is=title K2Node_CallFunction}
   n3 = BreakVector2D(InVec=Get Dynamic Tile Size!(self:WBP Discover Tile Base Object Reference=@n1) -> (DynamicTileSize:Vector 2D Structure) {is=title K2Node_CallFunction})
   n4 = Get TileView(self=Get WBP_Results_PolyTileView())
   > SetEntryWidth!(execute, self:Tile View Object Reference=@n4, NewWidth:float=@n3) -> (then)
   > SetEntryHeight!(execute, self:Tile View Object Reference=@n4, NewHeight:float=@n3.Y) -> (then)
  ]}
]
variables [
 ButtonGroupSelectionVM {type=FortNestedButtonGroupSelectionVM, default=null}
 DiscoverTileBaseReference {type=WBP_Discover_Tile_Base_C, default=null}
 IntroWhooshSound {type=SoundBase, default=/Game/Sounds/Fort_UI_Sounds/FrontEnd/Tabs/UI_Whoosh_TopBar_Tab_Search_Cue}
 IsSearchActive {type=boolean}
 IsTileViewDataEmpty {type=boolean}
 ReadyForStartingActivity {type=boolean}
 SearchMode {type=EDiscoverSearchMode, default=Island}
 TabBlockTextArray {type="array<WBP_UIKit_Block_Text_C>", default []}
]
widgets {
 info {inheritedWidgetCount=0, parentClass=&CommonActivatableWidget, rootWidgetClass=&GridPanel, widgetCount=30}
 widgets [
  RootGrid_Search {widgetClassPath=&GridPanel, widget=RootGrid_Search, properties {columnFill [1], rowFill [1]}}
  NebulousBackground {
   widgetClassPath=&Image
   bIsVariable=true
   parent=RootGrid_Search
   widget=NebulousBackground
   properties {brush {
     bIsDynamicallyLoaded=false
     drawAs=Image
     imageType=NoImage
     mirroring=NoMirror
     resourceName=None
     resourceObject=/FortUILibrary/MaterialInstances/Background/Nebulous/MI_UI_Background_Nebulous
     tiling=NoTile
     imageSize=<32 32>
     margin=<ltrb 0 0 0 0>
     outlineSettings {
      bUseBrushTransparency=false
      roundingType=HalfHeightRadius
      width=0
      color {colorUseRule=UseColor_Specified, specifiedColor=<rgba 0 0 0 0>}
      cornerRadii=<0 0 0 0>
     }
     tintColor {colorUseRule=UseColor_Specified, specifiedColor=<rgba 1 1 1 1>}
     uVRegion {bIsValid=false, max=<0 0>, min=<0 0>}
    }}
   slot {class=GridSlot}
  }
  CaptureForPostBufferUpdate {
   widgetClassPath=&WBP_CaptureForPostBufferUpdate
   bIsVariable=true
   parent=RootGrid_Search
   widget=CaptureForPostBufferUpdate
   slot {class=GridSlot, properties {layer=1}}
  }
  SafeZone {
   widgetClassPath=&SafeZone
   bIsVariable=true
   parent=RootGrid_Search
   widget=SafeZone
   properties {padBottom=false, padLeft=false, padRight=false, padTop=false}
   slot {class=GridSlot, properties {layer=10, padding=<ltrb 0 160 0 0>}}
  }
  VerticalBox {widgetClassPath=&VerticalBox, parent=SafeZone, widget=VerticalBox, slot {class=SafeZoneSlot}}
  Wrapper_CustomNavEvents_UMGBug {
   widgetClassPath=&GridPanel
   parent=VerticalBox
   widget=Wrapper_CustomNavEvents_UMGBug
   properties {columnFill [1], rowFill [1]}
   slot {class=VerticalBoxSlot, properties {horizontalAlignment=HAlign_Center, verticalAlignment=VAlign_Top}}
  }
  FilteredInputField {
   widgetClassPath=/FortUILibrary/UIKit/InputField/WBP_UIKit_FilteredInputField
   bIsVariable=true
   parent=Wrapper_CustomNavEvents_UMGBug
   widget=FilteredInputField
   properties {
    bIsFocusable=true
    hideValidationInfo=true
    hintText=Search for Islands or Creators...
    isSearchField=true
    navigation=&self:WidgetTree.FilteredInputField.WidgetNavigation_0
    useHintText=true
    clearButtonPadding=<ltrb 0 0 12 0>
    desiredImageSize=<56 56>
    height Override {enable=true, value=80}
    scaleOverride {enable=true, value=1.2}
    width Override {enable=true, value=800}
   }
   slot {class=GridSlot}
  }
  Switcher_CategoriesOrSearch {
   widgetClassPath=&CommonWidgetSwitcherLegacy
   bIsVariable=true
   parent=VerticalBox
   widget=Switcher_CategoriesOrSearch
   properties {navigation=&self:WidgetTree.Switcher_CategoriesOrSearch.WidgetNavigation_0}
   slot {class=VerticalBoxSlot, properties {padding=<ltrb 0 80 0 0>, size {sizeRule=Fill, value=1}}}
  }
  ActivityBrowserBrowseView_VM {
   widgetClassPath=/DiscoveryBrowser/ActivityBrowserBrowseView_VM
   bIsVariable=true
   parent=Switcher_CategoriesOrSearch
   widget=ActivityBrowserBrowseView_VM
   properties {bAutoRestoreFocus=false, bSupportsActivationFocus=false, clipping=ClipToBounds, desiredFocusWidget {widgetName=None}}
   slot {class=WidgetSwitcherSlot}
  }
  SearchResult_Switcher {
   widgetClassPath=&CommonWidgetSwitcherLegacy
   bIsVariable=true
   parent=Switcher_CategoriesOrSearch
   widget=SearchResult_Switcher
   properties {
    activeWidgetIndex=2
    navigation=&self:WidgetTree.SearchResult_Switcher.WidgetNavigation_0
   }
   slot {class=WidgetSwitcherSlot}
  }
  SearchSpinner {
   widgetClassPath=&GridPanel
   bIsVariable=true
   parent=SearchResult_Switcher
   widget=SearchSpinner
   properties {columnFill [1], rowFill [1]}
   slot {class=WidgetSwitcherSlot, properties {padding=<ltrb 0 0 0 80>}}
  }
  WBP_UIKit_Throbber {
   widgetClassPath=/FortUILibrary/UIKit/Throbber/WBP_UIKit_Throbber
   bIsVariable=true
   parent=SearchSpinner
   widget=WBP_UIKit_Throbber
   properties {displayBackground=false, throbberSize=Large}
   slot {class=GridSlot, properties {horizontalAlignment=HAlign_Center, verticalAlignment=VAlign_Center}}
  }
  ResultsGrid {
   widgetClassPath=&GridPanel
   bIsVariable=true
   parent=SearchResult_Switcher
   widget=ResultsGrid
   properties {columnFill [1], rowFill [0,1]}
   slot {class=WidgetSwitcherSlot, properties {padding=<ltrb 0 -32 0 0>}}
  }
  WBP_UIKit_Tabs {
   widgetClassPath=/FortUILibrary/UIKit/Tabs/WBP_UIKit_Tabs
   bIsVariable=true
   parent=ResultsGrid
   widget=WBP_UIKit_Tabs
   properties {
    bIsFocusable=true
    backplateCornerRadius=Button
    buttonHorizontalAlignment=HAlign_Left
    initialSelectedTabIndex=0
    inputActions=None
    isNaturalNavigation=true
    minButtonHeight=30
    minButtonWidth=50
    navigation=&self:WidgetTree.WBP_UIKit_Tabs.WidgetNavigation_0
    useBackplate=true
    tabGroupPadding=<ltrb 4 4 4 4>
   }
   slot {class=GridSlot, properties {horizontalAlignment=HAlign_Left, layer=1, padding=<ltrb 132 0 0 48>}}
  }
  CommonWidgetSwitcher_Results {
   widgetClassPath=&CommonWidgetSwitcherLegacy
   bIsVariable=true
   parent=ResultsGrid
   widget=CommonWidgetSwitcher_Results
   slot {class=GridSlot, properties {layer=2, row=1}}
  }
  VerticalBox_IslandBox {
   widgetClassPath=&VerticalBox
   bIsVariable=true
   parent=CommonWidgetSwitcher_Results
   widget=VerticalBox_IslandBox
   properties {clipping=ClipToBoundsWithoutIntersecting}
   slot {class=WidgetSwitcherSlot}
  }
  WBP_Row_Header {
   widgetClassPath=/DiscoveryBrowser/Rows/WBP_Row_Header
   bIsVariable=true
   parent=VerticalBox_IslandBox
   widget=WBP_Row_Header
   slot {class=VerticalBoxSlot, properties {padding=<ltrb 52 0 0 0>}}
  }
  Overlay {
   widgetClassPath=&Overlay
   parent=VerticalBox_IslandBox
   widget=Overlay
   slot {class=VerticalBoxSlot, properties {size {sizeRule=Fill, value=1}}}
  }
  CLIPPING_AND_PADDING_FIX {
   widgetClassPath=&VerticalBox
   parent=Overlay
   widget=CLIPPING_AND_PADDING_FIX
   properties {clipping=ClipToBounds}
   slot {class=OverlaySlot, properties {horizontalAlignment=HAlign_Fill, verticalAlignment=VAlign_Fill, padding=<ltrb 108 0 80 0>}}
  }
  WBP_Results_PolyTileView {
   widgetClassPath=/DiscoveryBrowser/WBP_Results_PolyTileView
   bIsVariable=true
   parent=CLIPPING_AND_PADDING_FIX
   widget=WBP_Results_PolyTileView
   properties {
    navigation=&self:WidgetTree.WBP_Results_PolyTileView.WidgetNavigation_0
    desiredFocusWidget {widgetName=None}
   }
   slot {class=VerticalBoxSlot, properties {verticalAlignment=VAlign_Top, padding=<ltrb 12 0 0 0>, size {sizeRule=Fill, value=1}}}
  }
  Spacer_99 {
   widgetClassPath=&Spacer
   parent=CLIPPING_AND_PADDING_FIX
   widget=Spacer_99
   properties {size=<1 50>}
   slot {class=VerticalBoxSlot, properties {verticalAlignment=VAlign_Bottom}}
  }
  Image_SoftEdgeGradient_Islands {
   widgetClassPath=&Image
   bIsVariable=true
   parent=Overlay
   widget=Image_SoftEdgeGradient_Islands
   properties {
    renderOpacity=0
    visibility=SelfHitTestInvisible
    brush {
     bIsDynamicallyLoaded=false
     drawAs=Image
     imageType=NoImage
     mirroring=NoMirror
     resourceName=None
     resourceObject=/DiscoveryBrowser/TabScreens/DiscoverHome/Materials/MI_UI_Discovery_NebulousBackgroundWithBottomMask
     tiling=NoTile
     imageSize=<32 50>
     margin=<ltrb 0 0 0 0>
     outlineSettings {
      bUseBrushTransparency=false
      roundingType=HalfHeightRadius
      width=0
      color {colorUseRule=UseColor_Specified, specifiedColor=<rgba 0 0 0 0>}
      cornerRadii=<0 0 0 0>
     }
     tintColor {colorUseRule=UseColor_Specified, specifiedColor=<rgba 1 1 1 1>}
     uVRegion {bIsValid=false, max=<0 0>, min=<0 0>}
    }
   }
   slot {class=OverlaySlot, properties {horizontalAlignment=HAlign_Fill}}
  }
  WBP_NoResults_SearchResults {
   widgetClassPath=/DiscoveryBrowser/WBP_Discovery_MessageNoResults.WBP_Discovery_NoResultsMessage
   bIsVariable=true
   parent=CommonWidgetSwitcher_Results
   widget=WBP_NoResults_SearchResults
   properties {flavorMessageBase="Please search for a different word or phrase ", flavorTitleBase=No Results Found}
   slot {class=WidgetSwitcherSlot, properties {verticalAlignment=VAlign_Center, padding=<ltrb 0 0 0 80>}}
  }
  SZBottomButtons {
   widgetClassPath=&SafeZone
   parent=RootGrid_Search
   widget=SZBottomButtons
   properties {padTop=false}
   slot {class=GridSlot, properties {horizontalAlignment=HAlign_Right, layer=20, verticalAlignment=VAlign_Bottom, padding=<ltrb 0 0 80 32>}}
  }
  Overlay_1 {widgetClassPath=&Overlay, parent=SZBottomButtons, widget=Overlay_1, slot {class=SafeZoneSlot}}
  WBP_CaptureForPostBufferUpdate {
   widgetClassPath=&WBP_CaptureForPostBufferUpdate
   bIsVariable=true
   parent=Overlay_1
   slot {class=OverlaySlot, properties {horizontalAlignment=HAlign_Fill, verticalAlignment=VAlign_Fill}}
  }
  HBBottomButtons {
   widgetClassPath=&HorizontalBox
   parent=Overlay_1
   widget=HBBottomButtons
   slot {class=OverlaySlot, properties {horizontalAlignment=HAlign_Fill, verticalAlignment=VAlign_Fill}}
  }
  WBP_BottomBarHack_Back {
   widgetClassPath=/Game/UI/Frontend/BottomBar/WBP_BottomBarDecoyButton
   bIsVariable=true
   parent=HBBottomButtons
   widget=WBP_BottomBarHack_Back
   properties {minWidth=120, triggeringInputAction {dataTable=&InputActionDataTable, rowName=Back_SearchResult}}
   slot {class=HorizontalBoxSlot}
  }
  WBP_BottomBarHack_BackToTop {
   widgetClassPath=&WBP_BottomBarDecoyButton
   bIsVariable=true
   parent=HBBottomButtons
   widget=WBP_BottomBarHack_BackToTop
   properties {minWidth=120, triggeringInputAction {dataTable=&InputActionDataTable, rowName=BackToTop}}
   slot {class=HorizontalBoxSlot}
  }
  WBP_LoadingMorePages {
   widgetClassPath=/DiscoveryBrowser/TabScreens/WBP_LoadingMorePages
   bIsVariable=true
   parent=RootGrid_Search
   widget=WBP_LoadingMorePages
   properties {visibility=Collapsed}
   slot {class=GridSlot, properties {horizontalAlignment=HAlign_Center, layer=30, verticalAlignment=VAlign_Bottom, padding=<ltrb 8 8 8 8>}}
  }
 ]
}
