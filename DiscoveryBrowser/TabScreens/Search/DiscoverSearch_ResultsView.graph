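# Widget for Discovery search results, with Islands and Creators tabs, a loading spinner, a headed tile grid and a no-results message.
# Tabs are built from view models with injected count labels; selecting one broadcasts SearchMode, fires analytics and refreshes header and switcher.
# Incoming result view models fill a polymorphic tile view, resized by search mode, with square tiles for creators.
# Generated entries receive a shared nested button-group selection view model.
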WidgetBlueprint /DiscoveryBrowser/TabScreens/Search/DiscoverSearch_ResultsView
animations [Anim_ListScrollFade {
  binding_count=1
  length_seconds=5
  object=&self:Anim_ListScrollFade
  sequence {bindings [Image_SoftEdgeGradient_Islands {
     id=#1
     possessed_class=Image
     tracks [{
       type=MovieSceneFloatTrack
       display_name=Render Opacity
       sections [{type=MovieSceneFloatSection, end_frame=6, start_frame=0, channels [None {type=MovieSceneScriptingFloatChannel, keys [[0,0],[6,1]]}]}]
      }]
    }]}
 }]
graphs [
 EventGraph {nodes [
   n1 = "On UIKit Tab Selected (WBP_UIKit_Tabs)"!() -> (OutputDelegate, then, TabID:Name=None, TabVM:MVVM Base Viewmodel Object Reference) {is=title K2Node_ComponentBoundEvent}
   > SendSearchTabSubViewEvent!(execute, self, TabID:Name=@n1.TabID) -> (then)
   n2 = Get WBP_UIKit_Tabs() -> (WBP_UIKit_Tabs:WBP UIKit Tabs Object Reference)
   > Set with Broadcast SearchMode!(
    execute
    SearchMode:EDiscoverSearchMode Enum=Select(
     Option 0:EDiscoverSearchMode Enum=Island
     Option 1:EDiscoverSearchMode Enum=Creator
     Index:int=GetTabIndexByID!(self:UIKit Tab Group Object Reference=@n2, TabNameID:Name=@n1.TabID) -> (ReturnValue:int=0)
    ) -> (ReturnValue:EDiscoverSearchMode Enum)
   ) -> (then, Output_Get:EDiscoverSearchMode Enum=Island)
   > SetRowHeaderText()
   > SetResultSwitcherState()
   n3 = "On Tab Button Creation (WBP_UIKit_Tabs)"!() -> (OutputDelegate, then, TabId:Name=None, TabButton:Common Button Base Object Reference) {is=title K2Node_ComponentBoundEvent}
   > Populate Additional Tab Content!(execute, self, TabId:Name=@n3.TabId, TabButton:Common Button Base Object Reference=@n3.TabButton) -> (then)
   n4 = "On Tab Clicked (WBP_UIKit_Tabs)"!() -> (OutputDelegate, then, TabId:Name=None) {is=title K2Node_ComponentBoundEvent}
   > Call On UIKit Tab Clicked!(execute:Exec, self:Discover Search Results View Object Reference, TabId:Name=@n4.TabId) -> (then:Exec) {is=title K2Node_CallDelegate}
   n5 = OnTileViewEntryGenerated!() -> (OutputDelegate, then, Widget:User Widget Object Reference) {is=title K2Node_CustomEvent}
   > n6 = Sequence()
   > n7 = Cast To WBP_IslandTile_Minimal!(execute, Object=@n5.Widget) -> (then, CastFailed, AsWBP Island Tile Minimal:WBP Island Tile Minimal Object Reference) {is=title K2Node_DynamicCast}
   > UpdateTileSizeAndShapeThenTileView!(
    execute
    self
    TileBase:WBP Discover Tile Base Object Reference=Get WBP_Discover_Tile_Base!(self:WBP Island Tile Minimal Object Reference=@n7) -> (WBP_Discover_Tile_Base:WBP Discover Tile Base Object Reference)
   ) -> (then)
   n8 = Cast To WBP_Discover_Tile_Button!(execute=@n7.CastFailed, Object=@n5.Widget) -> (then, CastFailed, AsWBP Discover Tile Button:WBP Discover Tile Button Object Reference) {is=title K2Node_DynamicCast}
   > UpdateTileSizeAndShapeThenTileView(TileBase=Get WBP_Discover_Tile_Base(self:WBP Discover Tile Button Object Reference=@n8))
   SetNestedButtonGroupSelectionVM!(
    execute:Exec=@n6.then_1
    self:Object Reference=@n5.Widget
    InVM:Fort Nested Button Group Selection VM Object Reference=Get ButtonGroupSelectionVM() -> (ButtonGroupSelectionVM:Fort Nested Button Group Selection VM Object Reference)
   ) -> (then:Exec) {is=title K2Node_Message}
   Event OnInitialized()
   > n9 = Construct Object from Class(Class=/Script/FortniteUI.FortNestedButtonGroupSelectionVM) -> (ReturnValue:Fort Nested Button Group Selection VM Object Reference)
   > Set ButtonGroupSelectionVM(ButtonGroupSelectionVM:Fort Nested Button Group Selection VM Object Reference=@n9) -> (Output_Get:Fort Nested Button Group Selection VM Object Reference)
   Event Construct()
   > n10 = Create Widget!(
    execute:Exec
    Class:User Widget Class Reference=/DiscoveryBrowser/Tiles/WBP_Discover_Tile_Base.WBP_Discover_Tile_Base_C
    OwningPlayer:Player Controller Object Reference
   ) -> (then:Exec, ReturnValue:WBP Discover Tile Base Object Reference)
   > Set DiscoverTileBaseReference(DiscoverTileBaseReference:WBP Discover Tile Base Object Reference=@n10) -> (Output_Get:WBP Discover Tile Base Object Reference)
   n11 = Create Event()
   n12 = Get TileView!(self:Fort Polymorphic Tile View Object Reference=Get WBP_Results_PolyTileView() -> (WBP_Results_PolyTileView:WBP Results Poly Tile View Object Reference)) -> (TileView:Fort Tile View Object Reference)
   > Bind Event to On Entry Generated(self:List View Base Object Reference=@n12, Delegate=@n11) {is=title K2Node_AddDelegate}
   Event Destruct()
   > Unbind Event from On Entry Generated(self:List View Base Object Reference=@n12, Delegate=@n11) {is=title K2Node_RemoveDelegate}
   Event PreConstruct()
   > PopulateTabs()
   > SetResultsTerm!(execute, self, ResultTerm:String=$Search Result) -> (then)
   > SetColorAndOpacity!(
    execute
    self=Get Text_HeaderRow!(self:WBP Row Header Object Reference=Get WBP_Row_Header() -> (WBP_Row_Header:WBP Row Header Object Reference)) -> (Text_HeaderRow:Common Text Object Reference)
    InColorAndOpacity_SpecifiedColor:Linear Color Structure=(R=1,G=1,B=1,A=1)
    InColorAndOpacity_ColorUseRule:ESlateColorStylingMode Enum
   ) -> (then)
   > Set with Broadcast SearchMode(SearchMode=Island) -> (Output_Get=Island)
   > SetResultSwitcherState()
   > SetTileViewEntrySpacing()
  ]}
 Fill Out TileView Data {nodes [
   n1 = Fill Out TileView Data!() -> (then, InputPin:Array of Discover Tile Item View Model Object References) {is=title K2Node_FunctionEntry}
   n2 = Get TileView(self=Get WBP_Results_PolyTileView())
   > ClearListItems(self=@n2)
   > n3 = Branch(Condition=Is Not Empty!(TargetArray:Array of Discover Tile Item View Model Object References=@n1) -> (ReturnValue:bool))
   > UpdateTileSizeAndShapeThenTileView(TileBase=Get DiscoverTileBaseReference() -> (DiscoverTileBaseReference:WBP Discover Tile Base Object Reference))
   > Set List Items!(execute, self:List View Object Reference=@n2, InListItems:Array of Object References=@n1) -> (then) {is=title K2Node_CallFunction}
   > Set IsTileViewDataEmpty(IsTileViewDataEmpty:bool=false) -> (Output_Get:bool)
   Set IsTileViewDataEmpty(execute=@n3.else, IsTileViewDataEmpty=true)
  ]}
 Navigate Down To Results {nodes [
   Navigate Down To Results!() -> (then, Navigation:EUINavigation Enum) {is=title K2Node_FunctionEntry}
   > Branch(Condition=NOT Boolean(A=Get IsTileViewDataEmpty() -> (IsTileViewDataEmpty:bool)))
   n1 = Get TileView(self=Get WBP_Results_PolyTileView())
   > SetSelectedIndex!(execute, self:List View Object Reference=@n1, Index:int=0) -> (then)
   > Return Node!(execute, ReturnValue:Widget Object Reference=@n1)
  ]}
 Navigate Up from Results {nodes [
   Navigate Up from Results!() -> (then, Navigation:EUINavigation Enum) {is=title K2Node_FunctionEntry}
   n1 = Get WBP_UIKit_Tabs()
   > Return Node(ReturnValue=GetTabButtonBaseByID!(
     self:Common Tab List Widget Base Object Reference=@n1
     TabNameID:Name=GetActiveTab!(self:Common Tab List Widget Base Object Reference=@n1) -> (ReturnValue:Name=None)
    ) -> (ReturnValue:Common Button Base Object Reference))
  ]}
 OnUIKitTabClicked {nodes [OnUIKitTabClicked!() -> (then, TabId:Name) {is=title K2Node_FunctionEntry}]}
 PlayFadeAnimation {nodes [
   n1 = PlayFadeAnimation!() -> (then, PlayForward:bool=~) {is=title K2Node_FunctionEntry}
   > n2 = Branch(Condition=@n1)
   n3 = Get Anim_ListScrollFade() -> (Anim_ListScrollFade:Widget Animation Object Reference)
   > PlayAnimationForward(InAnimation=@n3)
   PlayAnimationReverse(execute=@n2.else, InAnimation=@n3)
  ]}
 Populate Additional Tab Content {nodes [
   Populate Additional Tab Content!() -> (then, TabId:Name, TabButton:Common Button Base Object Reference) {is=title K2Node_FunctionEntry}
   > n1 = Cast To WBP_UIKit_Button_Selectable!(execute, Object=Get TabButton() -> (TabButton:Common Button Base Object Reference)) -> (then, CastFailed, AsWBP UIKit Button Selectable:WBP UIKit Button Selectable Object Reference) {is=title K2Node_DynamicCast}
   > Set UIKitButtonSelectableRef(UIKitButtonSelectableRef:WBP UIKit Button Selectable Object Reference=@n1) -> (Output_Get:WBP UIKit Button Selectable Object Reference)
   > n2 = Construct Object from Class(Class=/FortUILibrary/Systems/ModularButton/Blocks/WBP_UIKit_Block_Text.WBP_UIKit_Block_Text_C) -> (ReturnValue:WBP UIKit Block Text Object Reference)
   > n3 = Set UIKitBlockTextRef(UIKitBlockTextRef:WBP UIKit Block Text Object Reference=@n2) -> (Output_Get:WBP UIKit Block Text Object Reference)
   > Add(
    TargetArray:Array of WBP UIKit Block Text Object References=Get TabBlockTextArray() -> (TabBlockTextArray:Array of WBP UIKit Block Text Object References)
    NewItem:"WBP UIKit Block Text Object Reference (by ref)"=@n3
   ) {is=title K2Node_CallArrayFunction}
   > n4 = AddChildToGrid!(
    execute
    self:Grid Panel Object Reference=Get Main!(self:WBP UIKit Button Generic Base Object Reference=Get UIKitButtonSelectableRef() -> (UIKitButtonSelectableRef:WBP UIKit Button Selectable Object Reference)) -> (Main:Grid Panel Object Reference)
    Content:Widget Object Reference=@n3
    InRow:int=1
    InColumn:int=2
   ) -> (then, ReturnValue:Grid Slot Object Reference)
   > SetHorizontalAlignment(self:Grid Slot Object Reference=@n4, InHorizontalAlignment=HAlign_Left)
   > SetVerticalAlignment(self:Grid Slot Object Reference=@n4, InVerticalAlignment=VAlign_Fill)
   > SetFontMaterial!(
    execute
    self:WBP UIKit Block Text Object Reference=Get UIKitBlockTextRef() -> (UIKitBlockTextRef:WBP UIKit Block Text Object Reference)
    In Material:Material Instance Object Reference=/FortUILibrary/UIKit/Buttons/TabButtons/_MaterialInstances/MI_UI_TopNav_Text
   ) -> (then)
   > RegisterModularBlock!(
    execute
    self:UIKit Modular Button Object Reference=Get UIKitButtonSelectableRef()
    Block:UIKit Block Object Reference=Get UIKitBlockTextRef()
   ) -> (then)
  ]}
 PopulateTabs {nodes [
   PopulateTabs() {is=title K2Node_FunctionEntry}
   > n1 = Construct Object from Class(Class=/Script/FortniteUI.UIKitTabGroupViewModel) -> (ReturnValue:UI Kit Tab Group View Model Object Reference)
   > n2 = Sequence!(execute) -> (then_0, then_1, then_2:Exec)
   > n3 = Construct Object from Class(Class=/Script/FortniteUI.UIKitTabGroupButtonViewModel) -> (ReturnValue:UI Kit Tab Group Button View Model Object Reference)
   > n4 = Set Label!(execute, Label:Text="LOCTEXT(\"Islands\")", self:UI Kit Selectable Button View Model Object Reference=@n3) -> (then, Output_Get:Text)
   > Set TabID!(
    execute
    TabID:Name=String To Name(InString=Get Source String from Text!(Text:Text=@n4) -> (ReturnValue:String) {is=title K2Node_CallFunction})
    self:UI Kit Tab Group Button View Model Object Reference=@n3
   ) -> (then, Output_Get:Name=None)
   n5 = Get Tabs!(self:UI Kit Tab Group View Model Object Reference=@n1) -> (Tabs:Array of UI Kit Tab Group Button View Model Object References)
   > Add(
    TargetArray:Array of UI Kit Tab Group Button View Model Object References=@n5
    NewItem:"UI Kit Tab Group Button View Model Object Reference (by ref)"=@n3
   )
   n6 = Construct Object from Class(execute=@n2.then_1, Class=&UIKitTabGroupButtonViewModel)
   > n7 = Set Label(Label="LOCTEXT(\"Creators\")", self=@n6)
   > Set TabID(TabID=String To Name(InString=Get Source String from Text(Text=@n7)), self=@n6) -> (Output_Get=None)
   > Add(TargetArray=@n5, NewItem=@n6)
   n8 = Get WBP_UIKit_Tabs()
   Set Tabs Data!(
    execute=@n2.then_2
    self:WBP UIKit Tabs Base Object Reference=@n8
    InTabs Data:Array of UI Kit Tab Group Button View Model Object References=@n5
   ) -> (then) {is=title K2Node_CallFunction}
   > SelectTabByIndex!(execute, self:UIKit Tab Group Object Reference=@n8, Index:int=0, bSuppressClickFeedback:bool=true) -> (then, ReturnValue:bool)
  ]}
 SendSearchTabSubViewEvent {nodes [
   n1 = SendSearchTabSubViewEvent!() -> (then, TabID:Name) {is=title K2Node_FunctionEntry}
   > n2 = Cast To FortPlayerController(Object=GetOwningPlayer(self=Get WBP_UIKit_Tabs()))
   > "Fire Event Sub View Opened (Non-Localized)"!(
    execute
    FortPC:Fort Player Controller Object Reference=@n2
    ViewName:String="To String (Name)"(InName=@n1)
    AdditionalInfo:Flow Analytics Additional Info Structure
    bIsOverlay:bool=false
   ) -> (then)
  ]}
 SendSearchTabSubViewEventID {nodes [
   SendSearchTabSubViewEventID() {is=title K2Node_FunctionEntry}
   > SendSearchTabSubViewEvent(TabID=GetSelectedTabId!(self:Common Tab List Widget Base Object Reference=Get WBP_UIKit_Tabs()) -> (ReturnValue:Name=None)) {is=title K2Node_CallFunction}
  ]}
 Set Creator Result Count {nodes [
   n1 = Set Creator Result Count!() -> (then, CreatorResultCount:int) {is=title K2Node_FunctionEntry}
   > n2 = Set CreatorResultCount(CreatorResultCount:int=@n1) -> (Output_Get:int=0)
   > SetText!(
    execute
    self:WBP UIKit Block Text Object Reference="Get (a copy)"(Array:Array of WBP UIKit Block Text Object References=Get TabBlockTextArray(), Dimension 1=1) -> (Output:WBP UIKit Block Text Object Reference)
    Text:Text=Format Text!(Format="LOCTEXT(\"({CreatorsCount})\")", CreatorsCount:int=@n2) -> (Result)
   ) -> (then)
   > SetResultSwitcherState()
  ]}
 Set Island Result Count {nodes [
   n1 = Set Island Result Count!() -> (then, Island Result Count:int) {is=title K2Node_FunctionEntry}
   > n2 = Set IslandResultCount(IslandResultCount:int=@n1) -> (Output_Get:int=0)
   > SetText(
    self="Get (a copy)"(Array=Get TabBlockTextArray())
    Text=Format Text!(Format="LOCTEXT(\"({IslandsCount})\")", IslandsCount:int=@n2) -> (Result)
   )
   > SetResultSwitcherState()
  ]}
 SetCreatorSearchInProgress {nodes [
   n1 = SetCreatorSearchInProgress!() -> (then, SearchInProgress:bool=~) {is=title K2Node_FunctionEntry}
   > SetSearchInProgress!(execute, self, SearchInProgress:bool=@n1) -> (then)
  ]}
 SetIslandSearchInProgress {nodes [
   n1 = SetIslandSearchInProgress!() -> (then, SearchInProgress:bool=~) {is=title K2Node_FunctionEntry}
   > SetSearchInProgress(SearchInProgress=@n1)
  ]}
 SetResultSwitcherState {nodes [
   SetResultSwitcherState() {is=title K2Node_FunctionEntry}
   > n1 = Switch on Int!(execute:Exec, Selection:int=GetSelectedTabIndex!(self:UIKit Tab Group Object Reference=Get WBP_UIKit_Tabs()) -> (ReturnValue:int=0)) -> (0:Exec, 1:Exec) {bHasDefaultPin=false}
   n2 = Get WBP_NoResults_SearchResults() -> (WBP_NoResults_SearchResults:WBP Discovery No Results Message Object Reference)
   n3 = Get VerticalBox_IslandBox() -> (VerticalBox_IslandBox:Vertical Box Object Reference)
   n4 = "Equal (Integer)"(A=Get IslandResultCount() -> (IslandResultCount:int=0))
   > SetActiveWidget(
    self:Widget Switcher Object Reference=Get CommonWidgetSwitcher_Results() -> (CommonWidgetSwitcher_Results:Common Widget Switcher Legacy Object Reference)
    Widget=Select(Option 0:Widget Object Reference=@n3, Option 1:Widget Object Reference=@n2, Index:bool=@n4) -> (ReturnValue:Widget Object Reference)
   )
   n5 = Get WBP_Row_Header()
   > SetVisibility(
    self=@n5
    InVisibility=Select(Option 0:ESlateVisibility Enum=SelfHitTestInvisible, Option 1:ESlateVisibility Enum=Hidden, Index=@n4) -> (ReturnValue:ESlateVisibility Enum)
   )
   n6 = "Equal (Integer)"(A=Get CreatorResultCount() -> (CreatorResultCount:int=0))
   SetActiveWidget(
    execute=@n1.1
    self=Get CommonWidgetSwitcher_Results()
    Widget=Select(Option 0:Widget Object Reference=@n3, Option 1:Widget Object Reference=@n2, Index=@n6) -> (ReturnValue:Widget Object Reference)
   )
   > SetVisibility(
    self=@n5
    InVisibility=Select(Option 0:ESlateVisibility Enum=SelfHitTestInvisible, Option 1:ESlateVisibility Enum=Hidden, Index=@n6) -> (ReturnValue:ESlateVisibility Enum)
   )
  ]}
 SetResultsTerm {nodes [
   n1 = SetResultsTerm!() -> (then, ResultTerm:String) {is=title K2Node_FunctionEntry}
   > SetSubHeaderText!(
    execute
    self:WBP Row Header Object Reference=Get WBP_Row_Header()
    SubHeaderText:Text=Format Text!(Format="LOCTEXT(\"Matching \\\"{ResultsTerm}\\\"\")", ResultsTerm:String=@n1) -> (Result)
   ) -> (then)
  ]}
 SetRowHeaderText {nodes [
   SetRowHeaderText() {is=title K2Node_FunctionEntry}
   n1 = Get WBP_UIKit_Tabs()
   n2 = "Get (a copy)"(
    Array:Array of UI Kit Tab Group Button View Model Object References=Get LocalTabsViewModels!(self:UIKit Tab Group Object Reference=@n1) -> (LocalTabsViewModels:Array of UI Kit Tab Group Button View Model Object References)
    Dimension 1=GetSelectedTabIndex(self=@n1) -> (ReturnValue=0)
   ) -> (Output:UI Kit Tab Group Button View Model Object Reference)
   > SetHeaderText!(
    execute
    self:WBP Row Header Object Reference=Get WBP_Row_Header()
    HeaderText:Text=Format Text!(Format="LOCTEXT(\"{Label} Results\")", Label:Text=Get Label!(self:UI Kit Selectable Button View Model Object Reference=@n2) -> (Label:Text)) -> (Result)
   ) -> (then)
  ]}
 SetSearchInProgress {nodes [
   n1 = SetSearchInProgress!() -> (then, SearchInProgress:bool=~) {is=title K2Node_FunctionEntry}
   > SetActiveWidget(
    self:Common Visibility Switcher Object Reference=Get SearchResult_Switcher() -> (SearchResult_Switcher:Common Visibility Switcher Object Reference)
    Widget=Select(
     Option 0:Widget Object Reference=Get ResultsGrid() -> (ResultsGrid:Grid Panel Object Reference)
     Option 1:Widget Object Reference=Get SearchSpinner() -> (SearchSpinner:Grid Panel Object Reference)
     Index=@n1
    ) -> (ReturnValue:Widget Object Reference)
   )
  ]}
 SetSearchResults {nodes [
   n1 = SetSearchResults!() -> (then, SearchResultsArray:Array of Discover Tile Item View Model Object References) {is=title K2Node_FunctionEntry}
   > Fill Out TileView Data!(execute, self, InputPin:Array of Discover Tile Item View Model Object References=@n1) -> (then) {is=title K2Node_CallFunction}
   > SetResultSwitcherState() {is=title K2Node_CallFunction}
  ]}
 SetTileViewEntrySpacing {nodes [
   SetTileViewEntrySpacing() {is=title K2Node_FunctionEntry}
   n1 = Get TileView(self=Get WBP_Results_PolyTileView())
   > Set HorizontalEntrySpacing!(execute, HorizontalEntrySpacing:float=24, self:List View Object Reference=@n1) -> (then, Output_Get:float)
   > Set VerticalEntrySpacing!(execute, VerticalEntrySpacing:float=16, self:List View Object Reference=@n1) -> (then, Output_Get:float)
  ]}
 SetTileViewResultsHeight {nodes [
   n1 = SetTileViewResultsHeight!() -> (then, Height:double=~) {is=title K2Node_FunctionEntry}
   > SetHeightOverride(self=Get SB_ContentLimiter() -> (SB_ContentLimiter:Size Box Object Reference), InHeightOverride=@n1)
  ]}
 UpdateTileSizeAndShapeThenTileView {nodes [
   n1 = UpdateTileSizeAndShapeThenTileView!() -> (then, TileBase:WBP Discover Tile Base Object Reference) {is=title K2Node_FunctionEntry}
   n2 = SelectInt(A=8, B=3, bPickA=UseNumberOfColumns!() -> (ReturnValue:bool))
   n3 = "Equal (Enum)"(A:EDiscoverSearchMode Enum=Get SearchMode() -> (SearchMode:EDiscoverSearchMode Enum=Island), B:EDiscoverSearchMode Enum=Creator)
   > Set Tile Size!(
    execute
    self:WBP Discover Tile Base Object Reference=Get TileBase() -> (TileBase:WBP Discover Tile Base Object Reference)
    InSize:Vector 2D Structure=ColumnSizeToTileSize!(self:WBP Discover Tile Base Object Reference=@n1, NewColumnSize:int=SelectInt(A=@n2, B=6, bPickA=@n3), IsSquareTiles:bool=@n3) -> (NewTileSize:Vector 2D Structure)
   ) -> (then) {is=title K2Node_CallFunction}
   n4 = BreakVector2D(InVec=Get Dynamic Tile Size!(self:WBP Discover Tile Base Object Reference=Get TileBase()) -> (DynamicTileSize:Vector 2D Structure) {is=title K2Node_CallFunction})
   n5 = Get TileView(self=Get WBP_Results_PolyTileView())
   > SetEntryWidth!(execute, self:Tile View Object Reference=@n5, NewWidth:float=@n4) -> (then)
   > SetEntryHeight!(execute, self:Tile View Object Reference=@n5, NewHeight:float=@n4.Y) -> (then)
  ]}
]
variables [
 ButtonGroupSelectionVM {type=FortNestedButtonGroupSelectionVM, default=null}
 CreatorResultCount {type=integer}
 DiscoverTileBaseReference {type=WBP_Discover_Tile_Base_C, default=null}
 IslandResultCount {type=integer}
 IsTileViewDataEmpty {type=boolean}
 OnUIKitTabClicked {}
 SearchMode {type=EDiscoverSearchMode, default=Island}
 TabBlockTextArray {type="array<WBP_UIKit_Block_Text_C>", default []}
]
widgets {
 info {
  inheritedWidgetCount=0
  parentClass=&CommonUserWidget
  rootWidgetClass=&CommonVisibilitySwitcher
  widgetCount=15
 }
 widgets [
  SearchResult_Switcher {widgetClassPath=&CommonVisibilitySwitcher, bIsVariable=true, widget=SearchResult_Switcher, properties {activeWidgetIndex=1}}
  SearchSpinner {
   widgetClassPath=&GridPanel
   bIsVariable=true
   parent=SearchResult_Switcher
   widget=SearchSpinner
   properties {columnFill [1], rowFill [1]}
   slot {class=CommonVisibilitySwitcherSlot, properties {padding=<ltrb 0 0 0 80>}}
  }
  WBP_UIKit_Throbber {
   widgetClassPath=/FortUILibrary/UIKit/Throbber/WBP_UIKit_Throbber
   bIsVariable=true
   parent=SearchSpinner
   widget=WBP_UIKit_Throbber
   properties {displayBackground=false, throbberSize=Large}
   slot {class=GridSlot, properties {horizontalAlignment=HAlign_Center, verticalAlignment=VAlign_Center}}
  }
  ResultsGrid {
   widgetClassPath=&GridPanel
   bIsVariable=true
   parent=SearchResult_Switcher
   widget=ResultsGrid
   properties {columnFill [1], rowFill [0,1]}
   slot {class=CommonVisibilitySwitcherSlot}
  }
  WBP_UIKit_Tabs {
   widgetClassPath=/FortUILibrary/UIKit/Tabs/WBP_UIKit_Tabs
   bIsVariable=true
   parent=ResultsGrid
   widget=WBP_UIKit_Tabs
   properties {
    bIsFocusable=true
    backplateCornerRadius=Button
    buttonHorizontalAlignment=HAlign_Left
    initialSelectedTabIndex=0
    inputActions=None
    isNaturalNavigation=true
    minButtonHeight=30
    minButtonWidth=50
    navigation=&self:WidgetTree.WBP_UIKit_Tabs.WidgetNavigation_0
    useBackplate=true
    tabGroupPadding=<ltrb 4 4 4 4>
   }
   slot {class=GridSlot, properties {horizontalAlignment=HAlign_Left, layer=1, padding=<ltrb 80 0 0 48>}}
  }
  CommonWidgetSwitcher_Results {
   widgetClassPath=&CommonWidgetSwitcherLegacy
   bIsVariable=true
   parent=ResultsGrid
   widget=CommonWidgetSwitcher_Results
   slot {class=GridSlot, properties {layer=2, row=1}}
  }
  VerticalBox_IslandBox {
   widgetClassPath=&VerticalBox
   bIsVariable=true
   parent=CommonWidgetSwitcher_Results
   widget=VerticalBox_IslandBox
   slot {class=WidgetSwitcherSlot}
  }
  WBP_Row_Header {
   widgetClassPath=/DiscoveryBrowser/Rows/WBP_Row_Header
   bIsVariable=true
   parent=VerticalBox_IslandBox
   widget=WBP_Row_Header
   slot {class=VerticalBoxSlot}
  }
  Overlay {
   widgetClassPath=&Overlay
   parent=VerticalBox_IslandBox
   widget=Overlay
   properties {clipping=ClipToBounds}
   slot {class=VerticalBoxSlot, properties {size {sizeRule=Fill, value=1}}}
  }
  SB_ContentLimiter {
   widgetClassPath=&SizeBox
   bIsVariable=true
   parent=Overlay
   widget=SB_ContentLimiter
   properties {bOverride_HeightOverride=true, heightOverride=1080}
   slot {class=OverlaySlot, properties {horizontalAlignment=HAlign_Fill, padding=<ltrb 64 8 0 0>}}
  }
  CLIPPING_AND_PADDING_FIX {widgetClassPath=&VerticalBox, parent=SB_ContentLimiter, widget=CLIPPING_AND_PADDING_FIX, slot {class=SizeBoxSlot}}
  WBP_Results_PolyTileView {
   widgetClassPath=/DiscoveryBrowser/WBP_Results_PolyTileView
   bIsVariable=true
   parent=CLIPPING_AND_PADDING_FIX
   widget=WBP_Results_PolyTileView
   properties {
    navigation=&self:WidgetTree.WBP_Results_PolyTileView.WidgetNavigation_0
    desiredFocusWidget {widgetName=None}
   }
   slot {class=VerticalBoxSlot, properties {verticalAlignment=VAlign_Top, size {sizeRule=Fill, value=1}}}
  }
  Spacer_99 {
   widgetClassPath=&Spacer
   parent=CLIPPING_AND_PADDING_FIX
   widget=Spacer_99
   properties {size=<1 50>}
   slot {class=VerticalBoxSlot, properties {verticalAlignment=VAlign_Bottom}}
  }
  Image_SoftEdgeGradient_Islands {
   widgetClassPath=&Image
   bIsVariable=true
   parent=Overlay
   widget=Image_SoftEdgeGradient_Islands
   properties {
    renderOpacity=0
    visibility=SelfHitTestInvisible
    brush {
     bIsDynamicallyLoaded=false
     drawAs=Image
     imageType=NoImage
     mirroring=NoMirror
     resourceName=None
     resourceObject=/DiscoveryBrowser/TabScreens/DiscoverHome/Materials/MI_UI_Discovery_NebulousBackgroundWithBottomMask
     tiling=NoTile
     imageSize=<32 50>
     margin=<ltrb 0 0 0 0>
     outlineSettings {
      bUseBrushTransparency=false
      roundingType=HalfHeightRadius
      width=0
      color {colorUseRule=UseColor_Specified, specifiedColor=<rgba 0 0 0 0>}
      cornerRadii=<0 0 0 0>
     }
     tintColor {colorUseRule=UseColor_Specified, specifiedColor=<rgba 1 1 1 1>}
     uVRegion {bIsValid=false, max=<0 0>, min=<0 0>}
    }
   }
   slot {class=OverlaySlot, properties {horizontalAlignment=HAlign_Fill}}
  }
  WBP_NoResults_SearchResults {
   widgetClassPath=/DiscoveryBrowser/WBP_Discovery_MessageNoResults.WBP_Discovery_NoResultsMessage
   bIsVariable=true
   parent=CommonWidgetSwitcher_Results
   widget=WBP_NoResults_SearchResults
   properties {flavorMessageBase="Please search for a different word or phrase ", flavorTitleBase=No Results Found}
   slot {class=WidgetSwitcherSlot, properties {verticalAlignment=VAlign_Center, padding=<ltrb 0 0 0 80>}}
  }
 ]
}
