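# Discover search tab screen: an activatable widget with a search field above a switcher between the browse mega list and search results.
# The field drives the search view model, while browse providers populate the mega list.
# Scroll position toggles decoy Back and Back To Top buttons; input type and a stored last-focused widget steer focus.
# It also plays an intro animation, adjusts layout for mobile and reports impression analytics on tab and activation changes.
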WidgetBlueprint /DiscoveryBrowser/TabScreens/Search/DiscoverSearchView_CreatorV2
animations [
 Intro {
  binding_count=2
  length_seconds=0.6
  object=&self:Intro
  sequence {
   bindings [
    Switcher_CategoriesOrSearch {
     id=#1
     possessed_class=CommonWidgetSwitcherLegacy
     tracks [
      {
       type=MovieSceneFloatTrack
       display_name=Render Opacity
       sections [{type=MovieSceneFloatSection, channels [None {type=MovieSceneScriptingFloatChannel, keys [[0,0],[6,0],[24,1]]}]}]
      }
      {
       type=MovieScene2DTransformTrack
       sections [{type=MovieScene2DTransformSection, channels [Translation.Y {type=MovieSceneScriptingFloatChannel, keys [[0,40],[6,40],[24,0]]}]}]
      }
     ]
    }
    FilteredInputField {
     id=#2
     possessed_class=WBP_UIKit_FilteredInputField_C
     tracks [
      {
       type=MovieSceneFloatTrack
       display_name=Render Opacity
       sections [{type=MovieSceneFloatSection, channels [None {type=MovieSceneScriptingFloatChannel, keys [[0,0],[18,1]]}]}]
      }
      {
       type=MovieScene2DTransformTrack
       sections [{type=MovieScene2DTransformSection, channels [Translation.Y {type=MovieSceneScriptingFloatChannel, keys [[0,40],[18,0]]}]}]
      }
     ]
    }
   ]
   display_rate {numerator=60}
  }
 }
 Anim_MegaListScrollFade {
  binding_count=2
  length_seconds=0.3
  object=&self:Anim_MegaListScrollFade
  sequence {bindings [Image_SoftEdgeGradient_MegaList {
     id=#3
     possessed_class=Image
     tracks [{
       type=MovieSceneFloatTrack
       display_name=Render Opacity
       sections [{type=MovieSceneFloatSection, end_frame=6, start_frame=0, channels [None {type=MovieSceneScriptingFloatChannel, keys [[0,0],[6,1]]}]}]
      }]
    }]}
 }
 Anim_HeaderScrolling {
  binding_count=2
  length_seconds=0.2
  object=&self:Anim_HeaderScrolling
  sequence {
   bindings [
    GridSlot_2 {
     id=#4
     parent_id=#5
     possessed_class=GridSlot
     tracks [{
       type=MovieSceneMarginTrack
       display_name=Padding
       sections [{
         type=MovieSceneMarginSection
         end_frame=12
         start_frame=0
         channels [
          Left {type=MovieSceneScriptingFloatChannel, const [0,1]}
          Top {type=MovieSceneScriptingFloatChannel, keys [[0,150],[12,0]]}
          Right {type=MovieSceneScriptingFloatChannel, const [0,1]}
          Bottom {type=MovieSceneScriptingFloatChannel, const [0,1]}
         ]
        }]
      }]
    }
    SafeZone {id=#5, possessed_class=SafeZone, tracks []}
   ]
   display_rate {numerator=60}
  }
 }
]
graphs [
 BP_GetDesiredFocusTarget {nodes [
   Get Desired Focus Target()
   > n1 = Get!(execute:Exec) -> (then:Exec, else:Exec, LastFocusedWidget:Widget Object Reference)
   > Return Node!(execute, ReturnValue:Widget Object Reference=Get LastFocusedWidget() -> (LastFocusedWidget:Widget Object Reference))
   SetBottomButtonVisibility!(
    execute=@n1.else
    self:WBP Bottom Bar Decoy Button Object Reference=[
     Get BottomBarHack_Back() -> (BottomBarHack_Back:WBP Bottom Bar Decoy Button Object Reference)
     Get BottomBarHack_BackToTop() -> (BottomBarHack_BackToTop:WBP Bottom Bar Decoy Button Object Reference)
    ]
    ShouldButtonBeVisible:bool=false
   ) -> (then)
   n2 = Get InputField!(self:WBP UIKit Filtered Input Field Object Reference=Get FilteredInputField() -> (FilteredInputField:WBP UIKit Filtered Input Field Object Reference)) -> (InputField:WBP UIKit Input Field Object Reference)
   > Return Node(ReturnValue=Get EditableText!(self:WBP UIKit Block Editable Text Object Reference=Get Block_EditableText!(self:WBP UIKit Input Field Base Object Reference=@n2) -> (Block_EditableText:WBP UIKit Block Editable Text Object Reference)) -> (EditableText:UIKit Editable Text Object Reference))
  ]}
 EventGraph {nodes [
   n1 = "On Clicked (BottomBarHack_Back)"!() -> (OutputDelegate, then, Button:Common Button Base Object Reference) {is=title K2Node_ComponentBoundEvent}
   ClearSearch(
    execute=[@n8.else, @n1.then, @n31.then]
    self:Discover Search Model Object Reference=Get FortDiscoverSearchViewModel() -> (FortDiscoverSearchViewModel:Discover Search Model Object Reference)
   )
   > SetInputFieldText!(execute, self:WBP UIKit Filtered Input Field Object Reference=Get FilteredInputField(), Text:Text) -> (then)
   > UpdateFocusToSearchField()
   > UpdateClearButtonVisibility(self:WBP UIKit Filtered Input Field Object Reference=Get FilteredInputField())
   "On UIKit Tab Clicked (ResultsView)"!() -> (OutputDelegate, then, TabId:Name=None) {is=title K2Node_ComponentBoundEvent}
   > n2 = Switch on ECommonInputType!(execute, Selection:ECommonInputType Enum=Get Input Type() -> (Input Type:ECommonInputType Enum)) -> (MouseAndKeyboard:Exec, Gamepad:Exec, Touch:Exec, Count:Exec) {is=title K2Node_SwitchEnum}
   Branch(
    execute=@n2.Gamepad
    Condition=Get IsTileViewDataEmpty!(self:Discover Search Results View Object Reference=Get ResultsView() -> (ResultsView:Discover Search Results View Object Reference)) -> (IsTileViewDataEmpty:bool)
   )
   n3 = Get TileView!(self:Fort Polymorphic Tile View Object Reference=Get WBP_Results_PolyTileView!(self:Discover Search Results View Object Reference=Get ResultsView()) -> (WBP_Results_PolyTileView:WBP Results Poly Tile View Object Reference)) -> (TileView:Fort Tile View Object Reference)
   > NavigateToIndex!(execute, self:List View Object Reference=@n3, Index:int=0) -> (then)
   SetScrollOffset!(execute=[@n2.MouseAndKeyboard, @n2.Touch], self:List View Base Object Reference=@n3, InScrollOffset:float=0) -> (then)
   > Clear Selection(self:List View Object Reference=@n3)
   "Navigate Up Out Of Mega List (WBP_Poly_MegaList)"() {is=title K2Node_ComponentBoundEvent}
   > UpdateFocusToSearchField()
   n4 = "On Text Changed (FilteredInputField)"!() -> (OutputDelegate, then, Text:Text) {is=title K2Node_ComponentBoundEvent}
   > Set SearchTerm!(
    execute
    SearchTerm:String="To String (Text)"(InText=@n4.Text)
    self:Discover Search Model Object Reference=Get FortDiscoverSearchViewModel()
   ) -> (then, Output_Get:String)
   n5 = "On Text Committed (FilteredInputField)"!() -> (OutputDelegate, then, CommitMethod:ETextCommit Enum=Default) {is=title K2Node_ComponentBoundEvent}
   > n6 = Switch on ETextCommit!(execute, Selection:ETextCommit Enum=@n5.CommitMethod) -> (Default:Exec, OnEnter:Exec, OnUserMovedFocus:Exec, OnCleared:Exec) {is=title K2Node_SwitchEnum}
   n7 = GetCurrentText!(execute=@n6.OnEnter, self:WBP UIKit Filtered Input Field Object Reference=Get FilteredInputField()) -> (then, Text:Text)
   > n8 = Branch(Condition=NOT Boolean(A=TextIsEmpty(InText=@n7)))
   > ActionSearch(self:Discover Search Model Object Reference=Get FortDiscoverSearchViewModel())
   n9 = "On User Scrolled (ScrollBox)"!() -> (OutputDelegate, then, CurrentOffset:float) {is=title K2Node_ComponentBoundEvent}
   n10 = "Nearly Equal (Float)"(A=@n9.CurrentOffset, B=0, ErrorTolerance=0)
   n11 = Get ScrollBox() -> (ScrollBox:Scroll Box Object Reference)
   > SetBottomButtonVisibility(
    self=Get BottomBarHack_BackToTop()
    ShouldButtonBeVisible=AND Boolean(A=NOT Boolean(A=@n10), B=IsFullsizedOnScreen!(Widget:Widget Object Reference=@n11) -> (IsFullsized:bool))
   )
   "On Clicked (BottomBarHack_BackToTop)"!() -> (OutputDelegate, then, Button:Common Button Base Object Reference) {is=title K2Node_ComponentBoundEvent}
   > n12 = Branch(Condition=IsAnySearchActiveOrInProgress!(self:Discover Search Model Object Reference=Get FortDiscoverSearchViewModel()) -> (ReturnValue:bool))
   n13 = Get TileView(self=Get WBP_Results_PolyTileView(self=Get ResultsView()))
   > SetSelectedIndex!(execute, self:List View Object Reference=@n13, Index:int=0) -> (then)
   > SetScrollOffset(self=@n13, InScrollOffset=0)
   > ScrollToStart(self:Scroll Box Object Reference=Get ScrollBox())
   > n14 = Switch Back and Back To Top buttons!(execute, self, IsVisible:bool=false) -> (then)
   UpdateFocusToSearchField(execute=[@n14.then, @n16.then])
   n15 = Get ListView!(self:Fort Polymorphic List View Object Reference=Get WBP_Poly_MegaList() -> (WBP_Poly_MegaList:WBP Poly Mega List Object Reference)) -> (ListView:Common List View Object Reference)
   SetScrollOffset(execute=@n12.else, self=@n15, InScrollOffset=0)
   > SetSelectedIndex(self=@n15, Index=0)
   > ScrollToStart(self=Get ScrollBox())
   > n16 = SetBottomButtonVisibility(self=[Get BottomBarHack_Back(), Get BottomBarHack_BackToTop()], ShouldButtonBeVisible=false)
   n17 = OnTileViewResultsScrolled!() -> (OutputDelegate, then, ItemOffset:float=~, DistanceRemaining:float=~) {is=title K2Node_CustomEvent}
   n18 = "float > float"(A=@n17.ItemOffset)
   > n19 = Branch(Condition=@n18)
   n20 = Get ResultsView()
   > n21 = PlayFadeAnimation!(execute, self:Discover Search Results View Object Reference=@n20, PlayForward:bool=true) -> (then)
   n22 = Get TileView(self=Get WBP_Results_PolyTileView(self=Get ResultsView()))
   Switch Back and Back To Top buttons(execute=[@n23.then, @n21.then], IsVisible=AND Boolean(A=@n18, B=IsFullsizedOnScreen(Widget=@n22)))
   n23 = PlayFadeAnimation(execute=@n19.else, self=@n20, PlayForward=false)
   n24 = TopLevelTabChanged!() -> (OutputDelegate, then, LeavingFeature:EFortUIFeature Enum=ShowHome, EnteringFeature:EFortUIFeature Enum=ShowHome) {is=title K2Node_CustomEvent}
   > n25 = Branch(Condition="Equal (Enum)"(A:EFortUIFeature Enum=@n24.LeavingFeature, B:EFortUIFeature Enum=EnableSearchTabInTopNavigationBar))
   n26 = Get FortDiscoverSearchViewModel()
   > FireAnalyticsForBrowseAndSearch!(
    execute
    self:Discover Search Model Object Reference=@n26
    BrowseViewModel:Discover Browse View Model Object Reference=Get FortDiscoverBrowseViewModel() -> (FortDiscoverBrowseViewModel:Discover Browse View Model Object Reference)
   ) -> (then)
   Branch(execute=@n25.else, Condition="Equal (Enum)"(A=@n24.EnteringFeature, B=EnableSearchTabInTopNavigationBar))
   > StartAnalyticsSession(self:Discover Search Model Object Reference=@n26)
   > n27 = Branch(Condition=IsAnySearchActiveOrInProgress(self=Get FortDiscoverSearchViewModel()))
   > StartAnalyticsImpressions(self:Discover Search Model Object Reference=Get FortDiscoverSearchViewModel())
   n28 = Get WBP_Results_PolyTileView(self=Get ResultsView())
   > n29 = For Each Loop(Array:Array of User Widget Object References=GetDisplayedEntryWidgets!(self:List View Base Object Reference=Get TileView(self=@n28)) -> (ReturnValue:Array of User Widget Object References)) -> (Array Element:User Widget Object Reference)
   > n30 = Cast To FortActivityTileDetailsDisplay!(execute, Object=@n29) -> (then, CastFailed, AsFort Activity Tile Details Display:Fort Activity Tile Details Display Object Reference) {is=title K2Node_DynamicCast}
   > OnRespondToTileViewVisibilityChange!(execute, self:Fort Activity Tile Details Display Object Reference=@n30, bIsVisible:bool=true) -> (then)
   StartAnalyticsImpressions(execute=@n27.else, self:Discover Surface View Model Object Reference=Get FortDiscoverBrowseViewModel())
   n31 = OnInputFieldButtonClearClicked!() -> (OutputDelegate, then, Button:Common Button Base Object Reference) {is=title K2Node_CustomEvent}
   n32 = OnIntroFinished() {is=title K2Node_CustomEvent}
   n33 = Get Input Type()
   > Branch(Condition=NOT Boolean(A="Equal (Enum)"(A:ECommonInputType Enum=@n33, B:ECommonInputType Enum=Touch)))
   > RequestRefreshFocus(self:Common Activatable Widget Object Reference)
   n34 = OnInputMethodChanged!() -> (OutputDelegate, then, InputType:ECommonInputType Enum) {is=title K2Node_CustomEvent}
   > Set with Broadcast Input Type!(execute, Input Type:ECommonInputType Enum=@n34.InputType) -> (then, Output_Get:ECommonInputType Enum)
   > RequestRefreshFocus()
   Event PreConstruct()
   > SetCompactUIStyle() {is=title K2Node_CallFunction}
   > Set Material Instance Parameters() {is=title K2Node_CallFunction}
   Event Construct()
   n35 = Get FortDiscoverBrowseViewModel()
   > n36 = Branch(Condition=ShouldQueryOnConstruct!(self:Discover Surface View Model Object Reference=@n35) -> (ReturnValue:bool))
   > n37 = Set SurfaceName!(
    execute
    SurfaceName:Name=GetBrowseSurfaceName!(self:Discover Surface View Model Object Reference=@n35) -> (ReturnValue:"Name (by ref)"=None)
    self:Discover Surface View Model Object Reference=@n35
   ) -> (then, Output_Get:Name=None)
   SetWheelScrollMultiplier!(
    execute=[@n37.then, @n36.else]
    self:List View Base Object Reference=Get ListView(self=Get WBP_Poly_MegaList())
    NewWheelScrollMultiplier:float=0
   ) -> (then)
   > SetBottomButtonVisibility(self=[Get BottomBarHack_BackToTop(), Get BottomBarHack_Back()], ShouldButtonBeVisible=false)
   n38 = Create Event()
   n39 = Get TileView(self=Get WBP_Results_PolyTileView(self=Get ResultsView()))
   > Bind Event to On List View Scrolled(self:List View Object Reference=@n39, Delegate=@n38) {is=title K2Node_AddDelegate}
   n40 = Create Event()
   n41 = Get CommonInputSubsystem()
   > Bind Event to On Input Method Changed(self=@n41, Delegate=@n40)
   > OnInputMethodChanged!(execute, self, InputType:ECommonInputType Enum=GetCurrentInputType(self=Get CommonInputSubsystem())) -> (then) {is=title K2Node_CallFunction}
   n42 = Create Event()
   n43 = Get Button_Clear!(self:WBP UIKit Filtered Input Field Object Reference=Get FilteredInputField()) -> (Button_Clear:WBP UIKit Button Regular Object Reference)
   > Bind Event to On Clicked(self:Common Button Base Object Reference=@n43, Delegate=@n42) {is=title K2Node_AddDelegate}
   > BindToAthenaFrontendTabChange!(execute, Context:Widget Object Reference=Self-Reference(), EventToBind:"Delegate (by ref)"=Create Event()) -> (then)
   > Set LastFocusedWidget(LastFocusedWidget:Widget Object Reference) -> (Output_Get:Widget Object Reference)
   Event Destruct()
   > Unbind Event from On List View Scrolled(self:List View Object Reference=@n39, Delegate=@n38) {is=title K2Node_RemoveDelegate}
   > Unbind Event from On Input Method Changed(self=@n41, Delegate=@n40)
   > Unbind Event from On Clicked(self:Common Button Base Object Reference=@n43, Delegate=@n42) {is=title K2Node_RemoveDelegate}
   > UnbindToAthenaFrontendTabChange!(execute, Context:Widget Object Reference=Self-Reference()) -> (then)
   Event On Activated()
   > SetScrollWidget!(execute, ToScroll:Widget Object Reference=Get ScrollBox()) -> (then)
   > n44 = Branch(Condition=ShouldQueryOnConstruct(self=Get FortDiscoverBrowseViewModel()))
   Fire Event Main View Opened!(
    execute=[@n49.then, @n44.then, @n48.else]
    View:Widget Object Reference=Self-Reference()
    AdditionalInfo:Flow Analytics Additional Info Structure
    bIsOverlay:bool=false
   ) -> (then)
   > SendBrowseSubViewEvent()
   > BindToAnimationFinished(Animation=Get Intro(), Delegate=@n32)
   > PlayAnimationForward(InAnimation=Get Intro())
   > SpawnSound2D!(
    execute
    Sound:Sound Base Object Reference=Get IntroWhooshSound() -> (IntroWhooshSound:Sound Base Object Reference)
    VolumeMultiplier:float=1
    PitchMultiplier:float=1
    StartTime:float=0
    ConcurrencySettings:Sound Concurrency Object Reference
    bPersistAcrossLevelTransition:bool=false
    bAutoDestroy:bool=true
   ) -> (then, ReturnValue:Audio Component Object Reference)
   n45 = IsAnySearchActiveOrInProgress(self=Get FortDiscoverSearchViewModel())
   > n46 = Branch(Condition=@n45)
   > StartAnalyticsVisibilityImpressions(self:Discover Search Model Object Reference=Get FortDiscoverSearchViewModel())
   StartVisibleImpressions(execute=@n46.else, self:Discover Surface View Model Object Reference=Get FortDiscoverBrowseViewModel())
   n47 = Get FortDiscoverBrowseViewModel()
   n48 = Branch(
    execute=@n44.else
    Condition="Equal (Name)"(A=Get SurfaceName!(self:Discover Surface View Model Object Reference=@n47) -> (SurfaceName:Name=None))
   )
   > n49 = Set SurfaceName(SurfaceName=GetBrowseSurfaceName(self=@n47) -> (ReturnValue=None), self=@n47) -> (Output_Get=None)
   Event On Deactivated()
   > UnbindFromAnimationFinished!(execute, self:User Widget Object Reference, Animation:Widget Animation Object Reference=Get Intro(), Delegate:Delegate=@n32) -> (then)
   > Check and Stop Animation!(execute:Exec, Animation:Widget Animation Object Reference=Get Intro()) -> (then:Exec) {is=title K2Node_MacroInstance}
   > n50 = Branch(Condition=@n45)
   > StopAnalyticsVisibilityImpressions(self:Discover Search Model Object Reference=Get FortDiscoverSearchViewModel())
   StopVisibleImpressions(execute=@n50.else, self:Discover Surface View Model Object Reference=Get FortDiscoverBrowseViewModel())
  ]}
 IsFullsizedOnScreen {nodes [
   n1 = IsFullsizedOnScreen!() -> (then, Widget:Widget Object Reference) {is=title K2Node_FunctionEntry}
   n2 = GetPaintSpaceGeometry!(self:Widget Object Reference=@n1) -> (ReturnValue:"Geometry Structure (by ref)")
   > Return Node!(
    execute
    IsFullsized:bool="float < float"(
     A=GetLocalSize!(Geometry:"Geometry Structure (by ref)"=@n2) -> (ReturnValue_X:double, ReturnValue_Y:double).ReturnValue_Y
     B=GetDesiredSize!(self:Widget Object Reference=@n1) -> (ReturnValue_X:double, ReturnValue_Y:double).ReturnValue_Y
    )
   )
  ]}
 NavigateDownToBrowseOrSearch {nodes [
   NavigateDownToBrowseOrSearch!() -> (then, Navigation:EUINavigation Enum) {is=title K2Node_FunctionEntry}
   > n1 = Branch(Condition=IsAnySearchActiveOrInProgress(self=Get FortDiscoverSearchViewModel()))
   n2 = Get WBP_UIKit_Tabs!(self:Discover Search Results View Object Reference=Get ResultsView()) -> (WBP_UIKit_Tabs:WBP UIKit Tabs Object Reference)
   > n3 = Set LastFocusedWidget(LastFocusedWidget=GetTabButtonBaseByID!(
     self:Common Tab List Widget Base Object Reference=@n2
     TabNameID:Name=GetActiveTab!(self:Common Tab List Widget Base Object Reference=@n2) -> (ReturnValue:Name=None)
    ) -> (ReturnValue:Common Button Base Object Reference))
   > Return Node!(execute, ReturnValue:Widget Object Reference=@n3)
   n4 = Get WBP_Poly_MegaList()
   Set bFirstRowSnapIsZeroIndex!(execute=@n1.else, bFirstRowSnapIsZeroIndex:bool=true, self:WBP Poly Mega List Object Reference=@n4) -> (then, Output_Get:bool)
   n5 = Get ListView(self=@n4)
   > SetSelectedIndex(self=@n5, Index=0)
   > Navigate To Item!(
    execute
    self:List View Object Reference=@n5
    Item:Object Reference=Get Selected Item!(self:List View Object Reference=@n5) -> (ReturnValue:Object Reference) {is=title K2Node_CallFunction}
   ) -> (then)
   > Set LastFocusedWidget(LastFocusedWidget=@n5)
  ]}
 NavigateUpToTextEntry {nodes [
   NavigateUpToTextEntry!() -> (then, Navigation:EUINavigation Enum) {is=title K2Node_FunctionEntry}
   > n1 = Set LastFocusedWidget(LastFocusedWidget=Get InputField(self=Get FilteredInputField()))
   > Return Node(ReturnValue=@n1)
  ]}
 On Update Providers {nodes [
   n1 = On Update Providers!() -> (then, ProvidersVM:Array of Discover Provider View Model Object References) {is=title K2Node_FunctionEntry}
   > SetDataFromProviders(self:Discover Browse View Model Object Reference=Get FortDiscoverBrowseViewModel())
   n2 = Get ListView(self=Get WBP_Poly_MegaList())
   > ClearListItems(self=@n2)
   > Branch(Condition=Is Not Empty!(TargetArray:Array of Discover Provider View Model Object References=@n1) -> (ReturnValue:bool))
   > Set List Items!(execute, self:List View Object Reference=@n2, InListItems:Array of Object References=@n1) -> (then) {is=title K2Node_CallFunction}
  ]}
 OnQueryingMorePages {nodes [
   n1 = OnQueryingMorePages!() -> (then, LoadingState:EDiscoverLoadingState Enum=Unset) {is=title K2Node_FunctionEntry}
   > SetVisibility(
    self=Get WBP_LoadingMorePages() -> (WBP_LoadingMorePages:WBP Loading More Pages Object Reference)
    InVisibility=Select(Option 0=Collapsed, Option 1=HitTestInvisible, Index="Equal (Enum)"(A:EDiscoverLoadingState Enum=@n1, B:EDiscoverLoadingState Enum=Loading))
   )
  ]}
 SendBrowseSubViewEvent {nodes [
   SendBrowseSubViewEvent() {is=title K2Node_FunctionEntry}
   > Branch(Condition=IsActivated!(self:Common Activatable Widget Object Reference) -> (ReturnValue:bool))
   > n1 = Cast To FortPlayerController(Object=GetOwningPlayer())
   > "Fire Event Sub View Opened (Non-Localized)"!(
    execute
    FortPC:Fort Player Controller Object Reference=@n1
    ViewName:String=Browse
    AdditionalInfo:Flow Analytics Additional Info Structure
    bIsOverlay:bool=false
   ) -> (then)
  ]}
 Set Material Instance Parameters {nodes [
   Set Material Instance Parameters() {is=title K2Node_FunctionEntry}
   > n1 = GetDynamicMaterial(self=Get Image_SoftEdgeGradient_Islands!(self:Discover Search Results View Object Reference=Get ResultsView()) -> (Image_SoftEdgeGradient_Islands:Image Object Reference))
   > SetScalarParameterValue(self=@n1, ParameterName=Overlay Rotation, Value=0.5)
  ]}
 SetCompactUIStyle {nodes [
   n1 = Get Wrapper_CustomNavEvents_UMGBug() -> (Wrapper_CustomNavEvents_UMGBug:Grid Panel Object Reference)
   AddChild(self=Get CompactUI_SafeZone() -> (CompactUI_SafeZone:Safe Zone Object Reference), Content=@n1)
   n2 = SlotAsVerticalBoxSlot(Widget=@n1)
   > SetHorizontalAlignment(self:Vertical Box Slot Object Reference=@n2, InHorizontalAlignment=HAlign_Fill)
   > SetVerticalAlignment(self:Vertical Box Slot Object Reference=@n2, InVerticalAlignment=VAlign_Fill)
   SetCompactUIStyle() {is=title K2Node_FunctionEntry}
   > Branch(Condition=IsMobileGame())
   > AddChild(
    self=Get RootGrid_Search() -> (RootGrid_Search:Grid Panel Object Reference)
    Content=Get CaptureForPostBufferUpdate() -> (CaptureForPostBufferUpdate:WBP Capture For Post Buffer Update Object Reference)
   )
   > SetSidesToPad!(
    execute
    self:Safe Zone Object Reference=Get SafeZone() -> (SafeZone:Safe Zone Object Reference)
    InPadLeft:bool=true
    InPadRight:bool=true
    InPadTop:bool=true
    InPadBottom:bool=false
   ) -> (then)
  ]}
 Switch Back and Back To Top buttons {nodes [
   n1 = Switch Back and Back To Top buttons!() -> (then, IsVisible:bool=~) {is=title K2Node_FunctionEntry}
   > SetBottomButtonVisibility(self=Get BottomBarHack_BackToTop(), ShouldButtonBeVisible=@n1)
   > SetBottomButtonVisibility(self=Get BottomBarHack_Back(), ShouldButtonBeVisible=NOT Boolean(A=@n1))
  ]}
 Switch Between Browse And Search Views {nodes [
   n1 = Switch Between Browse And Search Views!() -> (then, IsSearchActive:bool=~) {is=title K2Node_FunctionEntry}
   > SetConsumeMouseWheel!(
    execute
    self:Scroll Box Object Reference=Get ScrollBox()
    NewConsumeMouseWheel:EConsumeMouseWheel Enum=Select(Option 0:EConsumeMouseWheel Enum=WhenScrollingPossible, Option 1:EConsumeMouseWheel Enum=Never, Index=@n1) -> (ReturnValue:EConsumeMouseWheel Enum)
   ) -> (then)
   > SetActiveWidget(
    self:Widget Switcher Object Reference=Get Switcher_CategoriesOrSearch() -> (Switcher_CategoriesOrSearch:Common Widget Switcher Legacy Object Reference)
    Widget=Select(Option 0:Widget Object Reference=Get WBP_Poly_MegaList(), Option 1:Widget Object Reference=Get ResultsView(), Index=@n1) -> (ReturnValue:Widget Object Reference)
   )
   > n2 = Branch(Condition=@n1)
   > StartAnalyticsImpressions(self:Discover Search Model Object Reference=Get FortDiscoverSearchViewModel())
   n3 = Get Image_SoftEdgeGradient_MegaList() -> (Image_SoftEdgeGradient_MegaList:Image Object Reference)
   > SetRenderOpacity(self=@n3, InOpacity=0)
   > SendSearchTabSubViewEventID(self:Discover Search Results View Object Reference=Get ResultsView())
   > UpdateTileViewSize()
   > n4 = Switch Back and Back To Top buttons(IsVisible=false) {is=title K2Node_CallFunction}
   n5 = GetInputAction!(self:Common Button Base Object Reference=Get BottomBarHack_BackToTop()) -> (InputActionRow:Data Table Row Handle Structure, ReturnValue:bool)
   SetBottomButtonVisibility(
    execute=[@n7.then, @n4.then]
    self=Get BottomBarHack_Back()
    ShouldButtonBeVisible=AND Boolean(A="Equal (Boolean)"(A:bool=@n5.ReturnValue, B:bool) -> (ReturnValue:bool=~) {is=title K2Node_PromotableOperator}, B=@n1)
   )
   StartAnalyticsImpressions(execute=@n2.else, self:Discover Surface View Model Object Reference=Get FortDiscoverBrowseViewModel())
   > SetRenderOpacity(self=@n3, InOpacity=1)
   > SendBrowseSubViewEvent() {is=title K2Node_CallFunction}
   n6 = GetScrollOffset!(self:Scroll Box Object Reference=Get ScrollBox()) -> (ReturnValue:float)
   > n7 = SetBottomButtonVisibility(self=Get BottomBarHack_BackToTop(), ShouldButtonBeVisible=NOT Boolean(A="Nearly Equal (Float)"(A=@n6, B=0, ErrorTolerance=0)))
  ]}
 UpdateFocusToSearchField {nodes [
   UpdateFocusToSearchField() {is=title K2Node_FunctionEntry}
   n1 = Get InputField(self=Get FilteredInputField())
   > Set LastFocusedWidget(LastFocusedWidget=Get EditableText(self=Get Block_EditableText(self=@n1)))
   > RequestRefreshFocus()
  ]}
 UpdateTileViewSize {nodes [
   UpdateTileViewSize() {is=title K2Node_FunctionEntry}
   n1 = GetViewportSize!() -> (ReturnValue_X:double, ReturnValue_Y:double)
   n2 = GetViewportScale!() -> (ReturnValue:float)
   > SetTileViewResultsHeight!(
    execute
    self:Discover Search Results View Object Reference=Get ResultsView()
    Height:double=float * float(A=float / float(A=@n1.ReturnValue_Y, B=@n2), B=0.6) {is=title K2Node_PromotableOperator}
   ) -> (then)
  ]}
]
variables [
 Input Type {type=ECommonInputType, default=MouseAndKeyboard}
 IntroWhooshSound {type=SoundBase, default=/Game/Sounds/Fort_UI_Sounds/FrontEnd/Tabs/UI_Whoosh_TopBar_Tab_Search_Cue}
 IsSearchActive {type=boolean}
 LastFocusedWidget {type=Widget, default=null}
]
widgets {
 info {inheritedWidgetCount=0, parentClass=&CommonActivatableWidget, rootWidgetClass=&GridPanel, widgetCount=21}
 widgets [
  RootGrid_Search {widgetClassPath=&GridPanel, bIsVariable=true, widget=RootGrid_Search, properties {columnFill [1], rowFill [1]}}
  NebulousBackground {
   widgetClassPath=&Image
   bIsVariable=true
   parent=RootGrid_Search
   widget=NebulousBackground
   properties {brush {
     bIsDynamicallyLoaded=false
     drawAs=Image
     imageType=NoImage
     mirroring=NoMirror
     resourceName=None
     resourceObject=/FortUILibrary/MaterialInstances/Background/Nebulous/MI_UI_Background_Nebulous
     tiling=NoTile
     imageSize=<32 32>
     margin=<ltrb 0 0 0 0>
     outlineSettings {
      bUseBrushTransparency=false
      roundingType=HalfHeightRadius
      width=0
      color {colorUseRule=UseColor_Specified, specifiedColor=<rgba 0 0 0 0>}
      cornerRadii=<0 0 0 0>
     }
     tintColor {colorUseRule=UseColor_Specified, specifiedColor=<rgba 1 1 1 1>}
     uVRegion {bIsValid=false, max=<0 0>, min=<0 0>}
    }}
   slot {class=GridSlot}
  }
  Image_SoftEdgeGradient_MegaList {
   widgetClassPath=&Image
   bIsVariable=true
   parent=RootGrid_Search
   widget=Image_SoftEdgeGradient_MegaList
   properties {
    visibility=SelfHitTestInvisible
    brush {
     bIsDynamicallyLoaded=false
     drawAs=Image
     imageType=FullColor
     mirroring=NoMirror
     resourceName=None
     resourceObject=/DiscoveryBrowser/TabScreens/Search/MaterialInstances/MI_BackgroundSoftEdge_Nebulous
     tiling=NoTile
     imageSize=<32 150>
     margin=<ltrb 0 0 0 0>
     outlineSettings {
      bUseBrushTransparency=false
      roundingType=HalfHeightRadius
      width=0
      color {colorUseRule=UseColor_Specified, specifiedColor=<rgba 0 0 0 0>}
      cornerRadii=<0 0 0 0>
     }
     tintColor {colorUseRule=UseColor_Specified, specifiedColor=<rgba 1 1 1 1>}
     uVRegion {bIsValid=false, max=<0 0>, min=<0 0>}
    }
   }
   slot {class=GridSlot, properties {layer=30, verticalAlignment=VAlign_Top}}
  }
  CaptureForPostBufferUpdate {
   widgetClassPath=&WBP_CaptureForPostBufferUpdate
   bIsVariable=true
   parent=RootGrid_Search
   widget=CaptureForPostBufferUpdate
   slot {class=GridSlot, properties {layer=1}}
  }
  SafeZone {
   widgetClassPath=&SafeZone
   bIsVariable=true
   parent=RootGrid_Search
   widget=SafeZone
   properties {padBottom=false, padLeft=false, padRight=false, padTop=false}
   slot {class=GridSlot, properties {layer=10, padding=<ltrb 0 150 0 0>}}
  }
  FortSwipePanel {widgetClassPath=/Script/FortniteUI.FortSwipePanel, bIsVariable=true, parent=SafeZone, widget=FortSwipePanel, slot {class=SafeZoneSlot}}
  ScrollBox {
   widgetClassPath=&ScrollBox
   bIsVariable=true
   parent=FortSwipePanel
   widget=ScrollBox
   properties {
    allowOverscroll=false
    bAllowRightClickDragScrolling=false
    bIsFocusable=true
    clipping=Inherit
    navigationScrollPadding=120
    scrollBarVisibility=Collapsed
    scrollWhenFocusChanges=AnimatedScroll
    wheelScrollMultiplier=3
    scrollbarPadding=<ltrb 0 0 0 0>
    scrollbarThickness=<0 0>
    widgetStyle {
     barThickness=0
     bottomShadowBrush {
      bIsDynamicallyLoaded=false
      drawAs=NoDrawType
      imageType=FullColor
      mirroring=NoMirror
      resourceName=../../../Engine/Content/Slate/Common/ScrollBoxShadowBottom.png
      resourceObject=null
      tiling=NoTile
      imageSize=<16 8>
      margin=<ltrb 0.5 0 0.5 1>
      outlineSettings {
       bUseBrushTransparency=false
       roundingType=HalfHeightRadius
       width=0
       color {colorUseRule=UseColor_Specified, specifiedColor=<rgba 0 0 0 0>}
       cornerRadii=<0 0 0 0>
      }
      tintColor {colorUseRule=UseColor_Specified, specifiedColor=<rgba 1 1 1 1>}
      uVRegion {bIsValid=false, max=<0 0>, min=<0 0>}
     }
     horizontalScrolledContentPadding=<ltrb 0 0 0 0>
     leftShadowBrush {
      bIsDynamicallyLoaded=false
      drawAs=NoDrawType
      imageType=FullColor
      mirroring=NoMirror
      resourceName=../../../Engine/Content/Slate/Common/ScrollBoxShadowLeft.png
      resourceObject=null
      tiling=NoTile
      imageSize=<8 16>
      margin=<ltrb 1 0.5 0 0.5>
      outlineSettings {
       bUseBrushTransparency=false
       roundingType=HalfHeightRadius
       width=0
       color {colorUseRule=UseColor_Specified, specifiedColor=<rgba 0 0 0 0>}
       cornerRadii=<0 0 0 0>
      }
      tintColor {colorUseRule=UseColor_Specified, specifiedColor=<rgba 1 1 1 1>}
      uVRegion {bIsValid=false, max=<0 0>, min=<0 0>}
     }
     rightShadowBrush {
      bIsDynamicallyLoaded=false
      drawAs=NoDrawType
      imageType=FullColor
      mirroring=NoMirror
      resourceName=../../../Engine/Content/Slate/Common/ScrollBoxShadowRight.png
      resourceObject=null
      tiling=NoTile
      imageSize=<8 16>
      margin=<ltrb 0 0.5 1 0.5>
      outlineSettings {
       bUseBrushTransparency=false
       roundingType=HalfHeightRadius
       width=0
       color {colorUseRule=UseColor_Specified, specifiedColor=<rgba 0 0 0 0>}
       cornerRadii=<0 0 0 0>
      }
      tintColor {colorUseRule=UseColor_Specified, specifiedColor=<rgba 1 1 1 1>}
      uVRegion {bIsValid=false, max=<0 0>, min=<0 0>}
     }
     topShadowBrush {
      bIsDynamicallyLoaded=false
      drawAs=NoDrawType
      imageType=FullColor
      mirroring=NoMirror
      resourceName=../../../Engine/Content/Slate/Common/ScrollBoxShadowTop.png
      resourceObject=null
      tiling=NoTile
      imageSize=<16 8>
      margin=<ltrb 0.5 1 0.5 0>
      outlineSettings {
       bUseBrushTransparency=false
       roundingType=HalfHeightRadius
       width=0
       color {colorUseRule=UseColor_Specified, specifiedColor=<rgba 0 0 0 0>}
       cornerRadii=<0 0 0 0>
      }
      tintColor {colorUseRule=UseColor_Specified, specifiedColor=<rgba 1 1 1 1>}
      uVRegion {bIsValid=false, max=<0 0>, min=<0 0>}
     }
     verticalScrolledContentPadding=<ltrb 0 0 0 0>
    }
   }
   slot {class=FortSwipePanelSlot}
  }
  VerticalBox {
   widgetClassPath=&VerticalBox
   parent=ScrollBox
   widget=VerticalBox
   slot {class=ScrollBoxSlot, properties {size {sizeRule=Fill, value=1}}}
  }
  Wrapper_CustomNavEvents_UMGBug {
   widgetClassPath=&GridPanel
   bIsVariable=true
   parent=VerticalBox
   widget=Wrapper_CustomNavEvents_UMGBug
   properties {
    navigation=&self:WidgetTree.Wrapper_CustomNavEvents_UMGBug.WidgetNavigation_0
    columnFill [1]
    rowFill [1]
   }
   slot {class=VerticalBoxSlot, properties {verticalAlignment=VAlign_Top, padding=<ltrb 80 10 80 0>}}
  }
  FilteredInputField {
   widgetClassPath=/FortUILibrary/UIKit/InputField/WBP_UIKit_FilteredInputField
   bIsVariable=true
   parent=Wrapper_CustomNavEvents_UMGBug
   widget=FilteredInputField
   properties {
    bIsFocusable=true
    hideValidationInfo=true
    hintText=Search Discover...
    isSearchField=true
    useHintText=true
    clearButtonPadding=<ltrb 0 0 12 0>
    desiredImageSize=<56 56>
    height Override {enable=true, value=72}
    scaleOverride {enable=true, value=1.2}
    width Override {enable=false, value=800}
   }
   slot {class=GridSlot}
  }
  Switcher_CategoriesOrSearch {
   widgetClassPath=&CommonWidgetSwitcherLegacy
   bIsVariable=true
   parent=VerticalBox
   widget=Switcher_CategoriesOrSearch
   properties {navigation=&self:WidgetTree.Switcher_CategoriesOrSearch.WidgetNavigation_0}
   slot {class=VerticalBoxSlot, properties {padding=<ltrb 0 60 0 0>, size {sizeRule=Fill, value=1}}}
  }
  WBP_Poly_MegaList {
   widgetClassPath=/DiscoveryBrowser/WBP_Poly_MegaList
   bIsVariable=true
   parent=Switcher_CategoriesOrSearch
   widget=WBP_Poly_MegaList
   slot {class=WidgetSwitcherSlot}
  }
  ResultsView {
   widgetClassPath=&/DiscoverSearch_ResultsView
   bIsVariable=true
   parent=Switcher_CategoriesOrSearch
   widget=ResultsView
   slot {class=WidgetSwitcherSlot}
  }
  SZBottomButtons {
   widgetClassPath=&SafeZone
   parent=RootGrid_Search
   widget=SZBottomButtons
   properties {padTop=false}
   slot {class=GridSlot, properties {horizontalAlignment=HAlign_Right, layer=20, verticalAlignment=VAlign_Bottom, padding=<ltrb 0 0 80 32>}}
  }
  Overlay_1 {widgetClassPath=&Overlay, parent=SZBottomButtons, widget=Overlay_1, slot {class=SafeZoneSlot}}
  WBP_CaptureForPostBufferUpdate {
   widgetClassPath=&WBP_CaptureForPostBufferUpdate
   bIsVariable=true
   parent=Overlay_1
   slot {class=OverlaySlot, properties {horizontalAlignment=HAlign_Fill, verticalAlignment=VAlign_Fill}}
  }
  HBBottomButtons {
   widgetClassPath=&HorizontalBox
   parent=Overlay_1
   widget=HBBottomButtons
   slot {class=OverlaySlot, properties {horizontalAlignment=HAlign_Fill, verticalAlignment=VAlign_Fill}}
  }
  BottomBarHack_Back {
   widgetClassPath=/Game/UI/Frontend/BottomBar/WBP_BottomBarDecoyButton
   bIsVariable=true
   parent=HBBottomButtons
   widget=BottomBarHack_Back
   properties {minWidth=120, triggeringInputAction {dataTable=&InputActionDataTable, rowName=Back_SearchResult}}
   slot {class=HorizontalBoxSlot}
  }
  BottomBarHack_BackToTop {
   widgetClassPath=&WBP_BottomBarDecoyButton
   bIsVariable=true
   parent=HBBottomButtons
   widget=BottomBarHack_BackToTop
   properties {minWidth=120, triggeringInputAction {dataTable=&InputActionDataTable, rowName=BackToTop}}
   slot {class=HorizontalBoxSlot}
  }
  CompactUI_SafeZone {
   widgetClassPath=&SafeZone
   bIsVariable=true
   parent=RootGrid_Search
   widget=CompactUI_SafeZone
   properties {padBottom=false, padLeft=false, padRight=false}
   slot {class=GridSlot, properties {horizontalAlignment=HAlign_Center, verticalAlignment=VAlign_Top, padding=<ltrb 0 150 0 0>}}
  }
  WBP_LoadingMorePages {
   widgetClassPath=/DiscoveryBrowser/TabScreens/WBP_LoadingMorePages
   bIsVariable=true
   parent=RootGrid_Search
   widget=WBP_LoadingMorePages
   properties {visibility=Collapsed}
   slot {class=GridSlot, properties {horizontalAlignment=HAlign_Center, layer=30, verticalAlignment=VAlign_Bottom, padding=<ltrb 8 8 8 8>}}
  }
 ]
}
