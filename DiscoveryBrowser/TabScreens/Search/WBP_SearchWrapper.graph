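# WBP_SearchWrapper is an activatable wrapper widget that hosts the search view in the Discovery browser's tab screens.
# On Construct, it checks the runtime option for the new browse surface and creates either DiscoverSearchView_CreatorV2 or the older DiscoverSearchView_Creator, storing it as SearchWidget.
# UpdateContent clears the root overlay and adds the chosen view with fill alignment.
# Activation and deactivation of the wrapper are forwarded to SearchWidget.
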WidgetBlueprint /DiscoveryBrowser/TabScreens/Search/WBP_SearchWrapper
graphs [
 EventGraph {nodes [
   Event Construct()
   > n1 = Branch(Condition=GetIsNewBrowseSurfaceEnabled!(self:Fort Runtime Options Object Reference=GetRuntimeOptions()) -> (ReturnValue:bool))
   > n2 = Create Widget!(
    execute:Exec
    Class:User Widget Class Reference=&/DiscoverSearchView_CreatorV2.DiscoverSearchView_CreatorV2_C
    OwningPlayer:Player Controller Object Reference
   ) -> (then:Exec, ReturnValue:Discover Search View Creator V2 Object Reference)
   > n3 = Set SearchWidget(SearchWidget:Common Activatable Widget Object Reference=@n2) -> (Output_Get:Common Activatable Widget Object Reference)
   > UpdateContent!(execute, self, ConstructRef:Common Activatable Widget Object Reference=@n3) -> (then)
   n4 = Create Widget(execute=@n1.else, Class=&/DiscoverSearchView_Creator.DiscoverSearchView_Creator_C) -> (ReturnValue:Discover Search View Creator Object Reference)
   > n5 = Set SearchWidget(SearchWidget=@n4)
   > UpdateContent(ConstructRef=@n5)
   Event On Activated()
   n6 = Get SearchWidget() -> (SearchWidget:Common Activatable Widget Object Reference)
   > ActivateWidget(self=@n6)
   Event On Deactivated()
   > DeactivateWidget(self=@n6)
  ]}
 UpdateContent {nodes [
   n1 = UpdateContent!() -> (then, ConstructRef:Common Activatable Widget Object Reference) {is=title K2Node_FunctionEntry}
   > ClearChildren(self=Get Overlay_Content() -> (Overlay_Content:Overlay Object Reference))
   > n2 = AddChildToOverlay!(execute, self:Overlay Object Reference=Get Overlay_Content(), Content:Widget Object Reference=@n1) -> (then, ReturnValue:Overlay Slot Object Reference)
   > SetHorizontalAlignment(self=@n2, InHorizontalAlignment=HAlign_Fill)
   > SetVerticalAlignment(self=@n2, InVerticalAlignment=VAlign_Fill)
  ]}
]
variables [SearchWidget {type=CommonActivatableWidget, default=null}]
widgets {
 info {inheritedWidgetCount=0, parentClass=&CommonActivatableWidget, rootWidgetClass=&Overlay, widgetCount=1}
 widgets [Overlay_Content {widgetClassPath=&Overlay, bIsVariable=true, widget=Overlay_Content}]
}
